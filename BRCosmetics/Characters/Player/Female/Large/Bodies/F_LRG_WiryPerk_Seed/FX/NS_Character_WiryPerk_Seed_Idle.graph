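# Idle cosmetic Niagara effect for the WiryPerk Seed player character body, with variant-selected colors.
# FakeTrail holds burst-spawned meshes on skeletal mesh sockets, while Squares sprites and curl-noise Voxel meshes spawn continuously on rings around sampled sockets.
# PartitionParticles assigns sockets and masks per-socket adjustments.
# User bools shift offsets, rotation and material for ADS and zero alpha when the hand is hidden; front-end modifiers adjust FakeTrail, and Squares and Voxel brighten at night.
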NiagaraSystem /BRCosmetics/Characters/Player/Female/Large/Bodies/F_LRG_WiryPerk_Seed/FX/NS_Character_WiryPerk_Seed_Idle
dependencies {
 usedDataInterfaces [&NiagaraDataInterfaceCurve, &NiagaraDataInterfaceSkeletalMesh]
 usedDynamicInputs [&RotateVectorByQuaternion]
 usedModules [
  &self:SystemSpawnScript.SystemScriptSource.SystemScriptGraph.NiagaraNodeAssignment_0.SetVariables_#1
  &SystemState
  &FN_VariantManager
 ]
}
emitters {
 FakeTrail {
  emitterUpdateScript [
   &EmitterState(
    Inactive Response=::NewEnumerator1
    Life Cycle Mode=::NewEnumerator1
    Loop Behavior=::NewEnumerator1
    Loop Duration Mode=::NewEnumerator1
    UseLoopDelay=true
   )
   &SpawnBurst_Instantaneous(Spawn Count=2, Use Spawn Probability=false, UseLoopCountLimit=false)
   &SpawnBurst_Instantaneous(Spawn Count=2, Use Spawn Probability=false, UseLoopCountLimit=false) {module=SpawnBurst_Instantaneous001}
  ]
  particleSpawnScript [
   &PartitionParticles()
   &InitializeParticle(
    Color=&VariantColorSelector(Variant Amount=1, Original Color=<rgba 1.3 0 0.4011 1>, Variant 1 Color=<rgba 0 1.3 0.6892 1>, Variant 2 Color=<rgba 1.3 0.7898 0 1>)
    Mesh Scale Mode=::NewEnumerator3
    Mesh Uniform Scale=2.4
    Sprite Rotation Mode=::NewEnumerator2
    Sprite Size Mode=::NewEnumerator3
    Uniform Sprite Size=1
    UsePositionOffset=false
    Write Material Random=true
   )
   /SetVariables(Particles.RandomCore:NiagaraFloat=&FloatFromCurve(
     FloatCurve=DataInterface() {propertyValues {Curve {keys [
         {arriveTangent=0.0559, interpMode=RCIM_Cubic, leaveTangent=0.0559, tangentMode=RCTM_User, time=0, value=0.2}
         {arriveTangent=0.1895, interpMode=RCIM_Cubic, leaveTangent=0.1895, tangentMode=RCTM_User, time=0.9, value=0.25}
         {arriveTangent=1.1808, interpMode=RCIM_Cubic, leaveTangent=1.1808, tangentMode=RCTM_User, time=1, value=0.3}
        ]}}}
     CurveIndex=&RandomRangeFloat(Minimum=0, Maximum=1)
     Scale Curve=2
    )) {isSetParameters=true, module=SetVariables_#2}
   /SetVariables(
    Particles.CameraOffset:NiagaraFloat=&FN_FrontEndFloatModifier(Gameplay=0.2, Front End=0)
    Particles.GlowColor=<rgba 0.2115 0 2 1>
    Particles.GlowSize=&Multiply_Vector2D_ByFloat(Float=&MakeFloatFromVector(Vector=$Particles.Scale, Channel=::NewEnumerator0), Vector2D=<60 60>)
   ) {isSetParameters=true, module=SetVariables_#3}
  ]
  particleUpdateScript [
   &ScaleColor(
    Scale Alpha=&RandomRangeFloat(Minimum=1, Maximum=1)
    Scale RGB=&Lerp_Vector(A=<1 1 1>, B=<1.5 1 1.5>, Alpha=&RandomRangeFloat(Minimum=0, Maximum=1))
    ScaleA=true
    ScaleRGB=true
   )
   &ParticleState(Kill Particles When Lifetime Has Elapsed=false)
   &Drag(Drag=2, Use Linear Drag=true, Use Rotational Drag=false)
   &SolveForcesAndVelocity(Clamp Velocity=false, Limit Acceleration=false)
   &SkeletalMeshLocation(
    Bone / Socket Index=$Particles.PartitionParticles.Partition
    MaskSpawnGroup=false
    Mesh Sampling Type=::NewEnumerator5
    Orientation Sampling=::NewEnumerator1
    Sampled Position Offset=&FN_FrontEndVector3Modifier(
     Gameplay=<0 0 0>
     Front End=&MaskVector(Mask Source=0, Mask Parameter=?, Mask=$Particles.PartitionParticles.Partition, Masked Vector=<0 0 0>, Passthrough Vector=<0 0 0>)
    )
    Skeletal Mesh=$System.Skeletal Mesh
    Socket Sampling Mode=::NewEnumerator5
   )
   /SetVariables(Particles.MeshOrientation=/Niagara/DynamicInputs/Helpers/MaskQuaternion(
     Mask Source:NiagaraInt32=0
     Mask Parameter:NiagaraInt32=?
     Mask:NiagaraInt32=$Particles.PartitionParticles.Partition
     Masked Quaternion:Quat4f=&LerpQuaternion(
      Quaternion A:Quat4f=$Particles.SkeletalMeshLocation.SampledRotation
      Quaternion B:Quat4f=&MultiplyQuaternion(
       Quaternion A=$Particles.SkeletalMeshLocation.SampledRotation
       Quaternion B=&MakeQuaternion(Quaternion From=1, Axis=<1 0 0>, Angle=0, Yaw=0.1, From=?, To=?, Angle Type=::NewEnumerator1, Normalize Vectors=false)
      )
      Lerp Factor:NiagaraFloat=&self:LerpedAlpha(rateOfChange:NiagaraFloat=0.1, isADS:NiagaraBool=$User.bisADS)
      Enable Spherical Linear Interpolation:NiagaraBool=false
     )
     Passthrough Quaternion:Quat4f=$Particles.SkeletalMeshLocation.SampledRotation
    )) {isSetParameters=true}
   /SetVariables(Particles.Vector:Vector3f=&Multiply_Vector(
     A=&Multiply_Vector(
      A=&ReturnMeshOrientationAxis(Mesh Orientation Axis=::ZAxis, Mesh Orientation Quat=$Particles.SkeletalMeshLocation.SampledRotation)
      B=&VectorFromFloat(Value=&RandomRangeFloat(Minimum=0, Maximum=0))
     )
     B=&VectorFromFloat(Value=$Particles.NormalizedAge)
    )) {isSetParameters=true, module=SetVariables_#4}
   /SetVariables(Particles.Position=&AddVectorToPosition(Position=$Particles.Position, Vector=$Particles.Vector)) {isSetParameters=true, module=SetVariables_#5}
   /SetVariables(Particles.Position=&AddVectorToPosition(
     Position=$Particles.Position
     Vector=&Multiply_Vector(
      A=&ReturnMeshOrientationAxis(Mesh Orientation Axis=::XAxis, Mesh Orientation Quat=$Particles.SkeletalMeshLocation.SampledRotation)
      B=&VectorFromFloat(Value=&MakeCustomFloatFromBool(Bool=$User.bisADS, True Float=0, False Float=-0.2))
     )
    )) {isSetParameters=true}
   &ScaleColor(
    Scale Alpha=&MakeCustomFloatFromBool(Bool=$User.bisHandHidden, True Float=0, False Float=1)
    ScaleA=true
    ScaleRGB=true
   ) {module=ScaleColor001}
   &ScaleMeshSize(Scale Factor=&FN_FrontEndVector3Modifier(
     Gameplay=<1 1 1>
     Front End=&MaskVector(Mask Source=0, Mask Parameter=?, Mask=$Particles.PartitionParticles.Partition, Masked Vector=<1.25 0.888 1>, Passthrough Vector=<1 1 1>)
    )) {module=ScaleMeshSize001}
  ]
  rendererClasses [&NiagaraMeshRendererProperties]
  renderers [{rendererClass=&NiagaraMeshRendererProperties, rendererIndex=0}]
 }
 Squares {
  emitterUpdateScript [
   &EmitterState(
    Inactive Response=::NewEnumerator1
    Life Cycle Mode=::NewEnumerator1
    Loop Behavior=::NewEnumerator1
    Loop Duration Mode=::NewEnumerator1
    UseLoopDelay=true
   )
   &SpawnRate(SpawnRate=18, Use Spawn Probability=false)
  ]
  particleSpawnScript [
   &PartitionParticles()
   &InitializeParticle(
    Color=&VariantColorSelector(Variant Amount=1, Original Color=<rgba 1.3 0 0.4011 1>, Variant 1 Color=<rgba 0 1.3 0.6892 1>, Variant 2 Color=<rgba 1.3 0.7466 0 1>)
    Lifetime=&RandomRangeFloat(Minimum=0.4, Maximum=0.8)
    Mesh Scale Mode=::NewEnumerator1
    Mesh Uniform Scale Max=0.15
    Mesh Uniform Scale Min=0.03
    Sprite Rotation Angle=90
    Sprite Rotation Mode=::NewEnumerator0
    Sprite Size Mode=::NewEnumerator1
    Uniform Sprite Size Max=1.85
    Uniform Sprite Size Min=0.5
    UsePositionOffset=false
    Write Material Random=true
   )
   /SetVariables(Particles.CameraOffset=1) {isSetParameters=true}
   /SetVariables() {isSetParameters=true}
  ]
  particleUpdateScript [
   &ScaleColor(Scale RGB=&RandomRangeVector(Minimum=<1 1 1>, Maximum=<0.8 0.8 0.8>), ScaleA=true, ScaleRGB=true) {module=ScaleColor003}
   &ScaleColor(
    Scale RGB=&FN_Select_Vector_from_Time_Of_Day(Morning=<1 1 1>, Day=<1 1 1>, Evening=<1 1 1>, Night=<3 3 3>)
    ScaleA=true
    ScaleRGB=true
   ) {module=ScaleColor002}
   &ScaleColor(
    Scale Alpha=&RandomRangeFloat(Minimum=0.5, Maximum=1)
    Scale RGB=&Lerp_Vector(A=<1 1 1>, B=<1.5 1 1.5>, Alpha=&RandomRangeFloat(Minimum=0, Maximum=1))
    ScaleA=true
    ScaleRGB=true
   )
   &ParticleState()
   &ScaleSpriteSize(
    Uniform Curve Scale=&RandomRangeFloat(Minimum=0.75, Maximum=1)
    Uniform Curve Sprite Scale=DataInterface() {propertyValues {Curve {
       defaultValue=1
       keys [
        {arriveTangent=2.6372, interpMode=RCIM_Cubic, leaveTangent=2.6372, tangentMode=RCTM_User, time=0, value=0.732}
        {arriveTangent=0.1046, interpMode=RCIM_Cubic, leaveTangent=0.1046, tangentMode=RCTM_User, time=0.1854, value=0.9498}
        {arriveTangent=-2.8989, interpMode=RCIM_Cubic, leaveTangent=-2.8989, tangentMode=RCTM_User, time=1, value=0}
       ]
      }}}
   )
   &Drag(Drag=2, Use Linear Drag=true, Use Rotational Drag=false)
   &SolveForcesAndVelocity(Clamp Velocity=false, Limit Acceleration=false)
   /SetVariables(Particles.CurrentADS:NiagaraFloat=&self:OffsetLerpedAlpha(rateOfChange:NiagaraFloat=10, isADS:NiagaraBool=$User.bisADS, CurrentADS:NiagaraFloat=$Particles.CurrentADS)) {isSetParameters=true}
   &SkeletalMeshLocation(
    Bone / Socket Index=$Particles.PartitionParticles.Partition
    MaskSpawnGroup=false
    Mesh Sampling Type=::NewEnumerator5
    Orientation Sampling=::NewEnumerator1
    Sampled Position Offset Coordinate Space:ENiagaraSkelMeshTransforms=::NewEnumerator2
    Skeletal Mesh=$System.Skeletal Mesh
    Socket Sampling Mode=::NewEnumerator5
   )
   &ShapeLocation(
    Additional Quaternion Rotation=true
    Additional Yaw / Pitch / Roll=&MaskVectorBySpawnGroup(
     Masked Vector=&MakeCustomVectorFromBool(Bool=$User.bisADS, True Vector=<0.15 0 0>, False Vector=<0.25 0 0>)
     Passthrough Vector=<0.25 0 0>
     Spawn Group Mask=$Particles.PartitionParticles.Partition
    )
    Disc Coverage=1
    MaskSpawnGroup=false
    Override Seed=false
    Ring Radius=6
    Rotation Mode=::NewEnumerator3
    Rotation Quaternion=$Particles.SkeletalMeshLocation.SampledRotation
    Shape Origin=$Particles.Position
    Shape Primitive=::NewEnumerator4
    Transform Order:ENiagara_TransformOrder=::NewEnumerator1
    U Distribution:NiagaraFloat=&MaskFloatBySpawnGroup(
     Masked Float=&MakeCustomFloatFromBool(Bool=$User.bisADS, True Float=1, False Float=0.5)
     Passthrough Float=0.5
     Spawn Group Mask=$Particles.PartitionParticles.Partition
    )
    UseOwnerScale=true
   ) {module=ShapeLocation002}
   /SetVariables(Particles.Vector:Vector3f=&Multiply_Vector(
     A=&Multiply_Vector(
      A=&ReturnMeshOrientationAxis(Mesh Orientation Axis=::ZAxis, Mesh Orientation Quat=$Particles.SkeletalMeshLocation.SampledRotation)
      B=&VectorFromFloat(Value=&RandomRangeFloat(Minimum=-10, Maximum=-10))
     )
     B=&VectorFromFloat(Value=$Particles.NormalizedAge)
    )) {isSetParameters=true, module=SetVariables_#4}
   /SetVariables(Particles.Position=&AddVectorToPosition(
     Position=&AddVectorToPosition(Position=$Particles.Position, Vector=$Particles.Vector)
     Vector=&Multiply_Vector(
      A=&ReturnMeshOrientationAxis(Mesh Orientation Axis=::ZAxis, Mesh Orientation Quat=$Particles.SkeletalMeshLocation.SampledRotation)
      B=&VectorFromFloat(Value=&RandomRangeFloat(Minimum=0, Maximum=10))
     )
    )) {isSetParameters=true, module=SetVariables_#5}
   /SetVariables(Particles.Position=&AddVectorToPosition(
     Position=$Particles.Position
     Vector=&Multiply_Vector(
      A=&ReturnMeshOrientationAxis(Mesh Orientation Axis=::XAxis, Mesh Orientation Quat=$Particles.SkeletalMeshLocation.SampledRotation)
      B=&VectorFromFloat(Value=&MakeCustomFloatFromBool(Bool=$User.bisADS, True Float=2, False Float=2))
     )
    )) {isSetParameters=true}
   &DynamicMaterialParameters(
    Index 0 Param 1=&MakeCustomFloatFromBool(Bool=$User.bisADS, True Float=50, False Float=1)
    Module.Index0Float0Write=true
    Module.Index0Float1Write=true
    Module.Index0Float2Write=true
    Module.Index0Float3Write=true
   )
   &ScaleColor(
    Scale Alpha=&MakeCustomFloatFromBool(Bool=$User.bisHandHidden, True Float=0, False Float=1)
    ScaleA=true
    ScaleRGB=true
   ) {module=ScaleColor001}
  ]
  rendererClasses [&NiagaraSpriteRendererProperties]
  renderers [{rendererClass=&NiagaraSpriteRendererProperties}]
 }
 Voxel {
  emitterUpdateScript [
   &EmitterState(
    Inactive Response=::NewEnumerator1
    Life Cycle Mode=::NewEnumerator1
    Loop Behavior=::NewEnumerator1
    Loop Duration Mode=::NewEnumerator1
    UseLoopDelay=true
   )
   &SpawnRate(SpawnRate=15, Use Spawn Probability=false)
  ]
  particleSpawnScript [
   &PartitionParticles()
   &InitializeParticle(
    Color=&VariantColorSelector(Variant Amount=1, Original Color=<rgba 1.3 0 0.4011 1>, Variant 1 Color=<rgba 0 1.3 0.6892 1>, Variant 2 Color=<rgba 1.3 0.7466 0 1>)
    Lifetime=&RandomRangeFloat(Minimum=0.4, Maximum=0.8)
    Mesh Scale Mode=::NewEnumerator1
    Mesh Uniform Scale Max=0.1
    Mesh Uniform Scale Min=0.08
    Sprite Rotation Mode=::NewEnumerator2
    Sprite Size Mode=::NewEnumerator3
    UsePositionOffset=false
    Write Material Random=true
   )
   /SetVariables(Particles.RandomCore:NiagaraFloat=&FloatFromCurve(
     FloatCurve=DataInterface() {propertyValues {Curve {keys [
         {arriveTangent=0.0559, interpMode=RCIM_Cubic, leaveTangent=0.0559, tangentMode=RCTM_User, time=0, value=0.2}
         {arriveTangent=0.1895, interpMode=RCIM_Cubic, leaveTangent=0.1895, tangentMode=RCTM_User, time=0.9, value=0.25}
         {arriveTangent=1.1808, interpMode=RCIM_Cubic, leaveTangent=1.1808, tangentMode=RCTM_User, time=1, value=0.3}
        ]}}}
     CurveIndex=&RandomRangeFloat(Minimum=0, Maximum=1)
     Scale Curve=2
    )) {isSetParameters=true, module=SetVariables_#2}
   /SetVariables(
    Particles.CameraOffset=1
    Particles.GlowColor=<rgba 0.2115 0 2 1>
    Particles.GlowSize=&Multiply_Vector2D_ByFloat(Float=&MakeFloatFromVector(Vector=$Particles.Scale, Channel=::NewEnumerator0), Vector2D=<60 60>)
   ) {isSetParameters=true, module=SetVariables_#3}
   /SetVariables() {isSetParameters=true}
  ]
  particleUpdateScript [
   &ScaleColor(Scale RGB=&RandomRangeVector(Minimum=<1 1 1>, Maximum=<0.8 0.8 0.8>), ScaleA=true, ScaleRGB=true) {module=ScaleColor003}
   &ScaleColor(
    Scale RGB=&FN_Select_Vector_from_Time_Of_Day(Morning=<1 1 1>, Day=<1 1 1>, Evening=<1 1 1>, Night=<3 3 3>)
    ScaleA=true
    ScaleRGB=true
   ) {module=ScaleColor002}
   &ScaleColor(
    Scale Alpha=&RandomRangeFloat(Minimum=0.5, Maximum=1)
    Scale RGB=&Lerp_Vector(A=<1 1 1>, B=<1.5 1 1.5>, Alpha=&RandomRangeFloat(Minimum=0, Maximum=1))
    ScaleA=true
    ScaleRGB=true
   )
   &ParticleState()
   &CurlNoiseForce(
    Noise Frequency=4
    Noise Strength=&Multiply_Float(A=75, B=$Particles.RandomCore)
    Pan Noise=true
    Pan Noise Field=<0.5 0.5 0>
   )
   &ScaleMeshSize(Scale Factor=&VectorFromFloat(Value=&FloatFromCurve(
      FloatCurve=DataInterface() {propertyValues {Curve {
         defaultValue=1
         keys [
          {arriveTangent=2.6372, interpMode=RCIM_Cubic, leaveTangent=2.6372, tangentMode=RCTM_User, time=0, value=0.732}
          {arriveTangent=0.1046, interpMode=RCIM_Cubic, leaveTangent=0.1046, tangentMode=RCTM_User, time=0.1854, value=0.9498}
          {arriveTangent=-2.8989, interpMode=RCIM_Cubic, leaveTangent=-2.8989, tangentMode=RCTM_User, time=1, value=0}
         ]
        }}}
      Scale Curve=&RandomRangeFloat(Minimum=0.75, Maximum=1)
     )))
   &Drag(Drag=2, Use Linear Drag=true, Use Rotational Drag=false)
   &SolveForcesAndVelocity(Clamp Velocity=false, Limit Acceleration=false)
   /SetVariables(Particles.CurrentADS:NiagaraFloat=&self:OffsetLerpedAlpha(rateOfChange:NiagaraFloat=10, isADS:NiagaraBool=$User.bisADS, CurrentADS:NiagaraFloat=$Particles.CurrentADS)) {isSetParameters=true}
   &SkeletalMeshLocation(
    Bone / Socket Index=$Particles.PartitionParticles.Partition
    MaskSpawnGroup=false
    Mesh Sampling Type=::NewEnumerator5
    Orientation Sampling=::NewEnumerator1
    Sampled Position Offset Coordinate Space:ENiagaraSkelMeshTransforms=::NewEnumerator2
    Skeletal Mesh=$System.Skeletal Mesh
    Socket Sampling Mode=::NewEnumerator5
   )
   &ShapeLocation(
    Additional Quaternion Rotation=true
    Additional Yaw / Pitch / Roll=&MaskVectorBySpawnGroup(
     Masked Vector=&MakeCustomVectorFromBool(Bool=$User.bisADS, True Vector=<0.15 0 0>, False Vector=<0.3 0 0>)
     Passthrough Vector=<0.3 0 0>
     Spawn Group Mask=$Particles.PartitionParticles.Partition
    )
    Disc Coverage=1
    MaskSpawnGroup=false
    Override Seed=false
    Ring Radius=6
    Rotation Mode=::NewEnumerator3
    Rotation Quaternion=$Particles.SkeletalMeshLocation.SampledRotation
    Shape Origin=$Particles.Position
    Shape Primitive=::NewEnumerator4
    Transform Order:ENiagara_TransformOrder=::NewEnumerator1
    U Distribution:NiagaraFloat=&MaskFloatBySpawnGroup(
     Masked Float=&MakeCustomFloatFromBool(Bool=$User.bisADS, True Float=1, False Float=0.5)
     Passthrough Float=0.5
     Spawn Group Mask=$Particles.PartitionParticles.Partition
    )
    UseOwnerScale=true
   ) {module=ShapeLocation002}
   /SetVariables(Particles.Vector:Vector3f=&Multiply_Vector(
     A=&Multiply_Vector(
      A=&ReturnMeshOrientationAxis(Mesh Orientation Axis=::ZAxis, Mesh Orientation Quat=$Particles.SkeletalMeshLocation.SampledRotation)
      B=&VectorFromFloat(Value=&RandomRangeFloat(Minimum=-10, Maximum=-10))
     )
     B=&VectorFromFloat(Value=$Particles.NormalizedAge)
    )) {isSetParameters=true, module=SetVariables_#4}
   /SetVariables(Particles.Position=&AddVectorToPosition(
     Position=&AddVectorToPosition(Position=$Particles.Position, Vector=$Particles.Vector)
     Vector=&Multiply_Vector(
      A=&ReturnMeshOrientationAxis(Mesh Orientation Axis=::ZAxis, Mesh Orientation Quat=$Particles.SkeletalMeshLocation.SampledRotation)
      B=&VectorFromFloat(Value=&RandomRangeFloat(Minimum=0, Maximum=15))
     )
    )) {isSetParameters=true, module=SetVariables_#5}
   /SetVariables(Particles.Position=&AddVectorToPosition(
     Position=$Particles.Position
     Vector=&Multiply_Vector(
      A=&ReturnMeshOrientationAxis(Mesh Orientation Axis=::XAxis, Mesh Orientation Quat=$Particles.SkeletalMeshLocation.SampledRotation)
      B=&VectorFromFloat(Value=&MakeCustomFloatFromBool(Bool=$User.bisADS, True Float=2, False Float=2))
     )
    )) {isSetParameters=true}
   &DynamicMaterialParameters(
    Index 0 Param 1=&MakeCustomFloatFromBool(Bool=$User.bisADS, True Float=50, False Float=1)
    Module.Index0Float0Write=true
    Module.Index0Float1Write=true
    Module.Index0Float2Write=true
    Module.Index0Float3Write=true
   )
   &ScaleColor(
    Scale Alpha=&MakeCustomFloatFromBool(Bool=$User.bisHandHidden, True Float=0, False Float=1)
    ScaleA=true
    ScaleRGB=true
   ) {module=ScaleColor001}
  ]
  rendererClasses [&NiagaraMeshRendererProperties]
  renderers [{rendererClass=&NiagaraMeshRendererProperties, rendererIndex=0}]
 }
}
summary {
 systemName=NS_Character_WiryPerk_Seed_Idle
 emitters [
  {emitterName=Voxel, rendererClasses [&NiagaraMeshRendererProperties]}
  {emitterName=FakeTrail, rendererClasses [&NiagaraMeshRendererProperties]}
  {emitterName=Squares, rendererClasses [&NiagaraSpriteRendererProperties]}
 ]
 userVariables [
  User.bisADS {type=NiagaraBool}
  User.bIsFrontEnd {type=NiagaraBool}
  User.bIsFrontEndPreview {type=NiagaraBool}
  User.bisHandHidden {type=NiagaraBool}
  User.ForwardDot {type=NiagaraFloat, defaultValue=0, description=""}
  User.Variant {type=NiagaraFloat, defaultValue=0, description=""}
 ]
}
