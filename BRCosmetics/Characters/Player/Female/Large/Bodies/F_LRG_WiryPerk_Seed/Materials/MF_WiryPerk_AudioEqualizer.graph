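# Material function for the WiryPerk Seed character body that draws an animated audio-equalizer effect from a scalar input.
# The input blends with a texture mask and an offset to set a smoothed threshold on panning noise.
# The result is gated by input and mask, then scaled by brightness that a pre-skinned local-height SphereMask suppresses.
# A time-driven sine, phase-shifted by mask blue, defines a "Rhythmic flow" reroute.
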MaterialFunction /BRCosmetics/Characters/Player/Female/Large/Bodies/F_LRG_WiryPerk_Seed/Materials/MF_WiryPerk_AudioEqualizer
graph {nodes [
  n1 = TextureSample(UVs=TextureCoordinate()) {
   samplerSource=SSM_Wrap_WorldGroupSettings
   samplerType=SAMPLERTYPE_Masks
   texture=/BRCosmetics/Characters/Player/Female/Large/Bodies/F_LRG_WiryPerk_Seed/Textures/T_WiryPerkSeed_TA_Mask3
  }
  n2 = FunctionInput() {bUsePreviewValueAsDefault=true, inputType=FunctionInput_Scalar}
  n3 = LinearInterpolate(A=@n1.R, B=@n2)
  n4 = NamedRerouteUsage()
  n5 = InverseLinearInterpolate(A=ScalarParameter() {defaultValue=0.4, group=Wiryperk_CH, parameterName=AudioMinThreshold}, Value=Saturate(None=Add(A=@n3, B=@n4))) {constA=0.5}
  n6 = ScalarParameter() {defaultValue=0.05, group=Wiryperk_CH, parameterName=Smoothing}
  n7 = Subtract(A=@n5, B=@n6) {constB=0.1}
  n8 = Add(A=@n5, B=@n6)
  n9 = TextureCoordinate()
  n10 = Constant() {r=0.3}
  n11 = TextureSample(UVs=Panner(
    Coordinate=Multiply(A=@n9, B=@n10) {constB=3}
    Speed=ScalarParameter() {defaultValue=0.7, group=Wiryperk_CH, parameterName=AudioEqualizer_PannerSpeed}
   ) {speedY=10.2}) {samplerSource=SSM_Wrap_WorldGroupSettings, samplerType=SAMPLERTYPE_Grayscale, texture=&LowResBlurredNoise}
  n12 = Multiply(A=@n11.R) {constB=2}
  n13 = Multiply(A=SmoothStep(Min=Saturate(None=@n7), Max=Saturate(None=@n8), Value=Saturate(None=@n12)), B=Saturate(None=Multiply(A=@n2) {constB=10}))
  n14 = ScalarParameter() {defaultValue=8, group=Wiryperk_CH, parameterName=AudioEqualizer_Brightness}
  n15 = SphereMask(
   A=VertexInterpolator(VS=LocalPosition() {localOrigin=InstancePreSkinning}.Z).PS
   B=Constant() {r=-16.8304}
   Radius=Constant() {r=57.7797}
   Hardness=Constant() {r=0.72}
  )
  FunctionOutput(None=Multiply(A=Multiply(A=@n13, B=@n1.G), B=LinearInterpolate(A=@n14, Alpha=@n15) {constB=0}))
  n16 = Time()
  n17 = ScalarParameter() {defaultValue=0.7, group=Wiryperk_CH, parameterName=AudioEqualizer_RhythmSpeed}
  n18 = Sine(None=Add(A=Multiply(A=@n16, B=@n17) {constB=0.9}, B=NamedRerouteDeclaration(None=@n1.B) {name=HorizontalWindowMask, nodeColor=<rgba 0 1 0.0625 1>}))
  NamedRerouteDeclaration(None=Multiply(A=Saturate(None=@n18)) {constB=0.4}) {name=Rhythmic flow, nodeColor=<rgba 1 0 0.6797 1>}
 ]}
