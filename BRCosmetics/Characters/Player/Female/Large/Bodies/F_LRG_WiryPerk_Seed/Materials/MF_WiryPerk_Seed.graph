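# Material function for the WiryPerk Seed character body that overrides base color and emissive on incoming material attributes with animated glow effects.
# Two packed mask textures drive named reroutes for hue shift, water cubes, hands and equalizer regions.
# Emissive blends time-based hue shift, a WaterCooling subfunction and fresnel-tinted hand glow with pulsing grid.
# Leg animation combines a switchable AudioEqualizer subfunction with pre-skinning height sphere masks, while fresnel darkens base color.
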MaterialFunction /BRCosmetics/Characters/Player/Female/Large/Bodies/F_LRG_WiryPerk_Seed/Materials/MF_WiryPerk_Seed
graph {nodes [
  n1 = GetMaterialAttributes(None=FunctionInput() {inputName=MA, inputType=FunctionInput_MaterialAttributes}) {attributeGetTypes [#1, #2]}
  n2 = ScalarParameter() {defaultValue=1, group=Wiryperk_CH, parameterName=LegbaseEmissiveOffset}
  n3 = ScalarParameter() {defaultValue=8, group=Wiryperk_CH, parameterName=BrightnessEqualizer}
  n4 = VertexInterpolator(VS=LocalPosition() {localOrigin=InstancePreSkinning}.Z)
  n5 = Constant() {r=-16.8304}
  n6 = Constant() {r=52.8146}
  n7 = Constant() {r=0.8038}
  n8 = Multiply(A=NamedRerouteUsage(), B=OneMinus(None=SphereMask(A=@n4.PS, B=@n5, Radius=@n6, Hardness=@n7)))
  n9 = LinearInterpolate(
   A=NamedRerouteUsage()
   B=Multiply(A=NamedRerouteUsage(), B=LinearInterpolate(A=@n2, B=@n3, Alpha=@n8) {constA=0.9}) {constB=2}
   Alpha=NamedRerouteUsage()
  )
  n10 = Multiply(A=@n9, B=NamedRerouteUsage())
  n11 = NamedRerouteUsage()
  FunctionOutput(None=SetMaterialAttributes!(MaterialAttributes=@n1.MaterialAttributes, EmissiveColor=LinearInterpolate(A=@n9, B=@n10, Alpha=@n11), BaseColor=NamedRerouteUsage()) {attributeSetTypes [#1, #2]})
  n12 = TextureSampleParameter2D() {
   group=Wiryperk_CH
   parameterName=TA_Mask_A
   samplerSource=SSM_Wrap_WorldGroupSettings
   texture=/BRCosmetics/Characters/Player/Female/Large/Bodies/F_LRG_WiryPerk_Seed/Textures/T_WiryPerkSeed_TA_Mask1
  }
  NamedRerouteDeclaration(None=@n12.B) {name=WaterCubes, nodeColor=<rgba 1 0 0.3281 1>}
  NamedRerouteDeclaration(None=@n12.G) {name=HandMask, nodeColor=<rgba 0.6172 0 1 1>}
  n13 = MaterialFunctionCall(
   Hue Shift Percentage=Multiply(A=Time(), B=ScalarParameter() {defaultValue=0.3, group=Wiryperk_CH, parameterName=HueShiftSpeed}) {constB=0.3}
   Texture=NamedRerouteUsage()
  ) {materialFunction=&HueShift}
  n14 = LinearInterpolate(A=NamedRerouteUsage(), B=@n13.Result, Alpha=NamedRerouteUsage()) {constB=0.3}
  n15 = NamedRerouteUsage()
  n16 = MaterialFunctionCall!(baseColor=@n13.Result, S_map, input_fxmask, WaterCooling Mask, AO) {materialFunction=&/MF_WiryPerk_Seed_WaterCooling}
  NamedRerouteDeclaration(None=LinearInterpolate(A=LinearInterpolate(A=@n14, B=@n16.Slurp_emiss, Alpha=@n15), B=NamedRerouteUsage(), Alpha=NamedRerouteUsage())) {name=MainEmissive, nodeColor=<rgba 0.1953 1 0 1>}
  n17 = NamedRerouteUsage()
  n18 = LinearInterpolate(A=@n17, B=Multiply(A=@n17, B=Constant() {r=0.488}) {constB=0.2}, Alpha=NamedRerouteUsage())
  n19 = LinearInterpolate(A=@n18, B=Multiply(A=@n18, B=Constant() {r=0.8}) {constB=0.5}, Alpha=Saturate(None=NamedRerouteUsage()))
  n20 = OneMinus(None=Fresnel(ExponentIn=Constant() {r=1}, BaseReflectFractionIn=Constant() {r=-0.0837}))
  n21 = LinearInterpolate(
   A=ScalarParameter() {defaultValue=0.6, group=Wiryperk_CH, parameterName=FresnelBrightnessMax_ BaseColor}
   B=ScalarParameter() {defaultValue=0.3, group=Wiryperk_CH, parameterName=FresnelBrightnessMin_BaseColor}
   Alpha=Saturate(None=Power(Base=@n20))
  ) {constA=0.5, constB=0.9}
  NamedRerouteDeclaration(None=LinearInterpolate(A=@n19, B=Multiply(A=@n19, B=@n21), Alpha=NamedRerouteUsage())) {name=OutputColor, nodeColor=<rgba 1 0 0.5156 1>}
  n22 = TextureSampleParameter2D() {
   group=Wiryperk_CH
   parameterName=TA_Mask_B
   samplerSource=SSM_Wrap_WorldGroupSettings
   texture=/BRCosmetics/Characters/Player/Female/Large/Bodies/F_LRG_WiryPerk_Seed/Textures/T_WiryPerkSeed_TA_Mask2
  }
  n23 = SphereMask(A=TextureCoordinate(), B=Constant2Vector() {r=1}, Radius=Constant() {r=0.4})
  NamedRerouteDeclaration(None=Multiply(A=@n22.B, B=OneMinus(None=@n23))) {name=LightPoint, nodeColor=<rgba 0 0.8359 1 1>}
  n24 = TextureCoordinate()
  n25 = TextureSample(UVs=Panner(Coordinate=Multiply(A=@n24), Speed=Constant() {r=0.6}) {speedX=0.2, speedY=0.2}) {samplerSource=SSM_Wrap_WorldGroupSettings, samplerType=SAMPLERTYPE_Grayscale, texture=&LowResBlurredNoise}
  n26 = Power(Base=Multiply(A=@n25.R) {constB=5}) {constExponent=5}
  NamedRerouteDeclaration(None=LinearInterpolate(Alpha=Saturate(None=@n26)) {constB=5}) {name=Noise, nodeColor=<rgba 0 1 1 1>}
  NamedRerouteDeclaration(None=Multiply(A=@n22.B, B=@n23)) {name=HandGrid, nodeColor=<rgba 0.6641 1 0 1>}
  NamedRerouteDeclaration(None=@n22.G) {name=EqualizerGradation_G, nodeColor=<rgba 0.3594 1 0 1>}
  NamedRerouteDeclaration(None=@n1.BaseColor) {name=InputBaseColor, nodeColor=<rgba 0 1 0.2734 1>}
  NamedRerouteDeclaration(None=@n1.EmissiveColor) {name=InputEmissiveColor, nodeColor=<rgba 1 0 0.0469 1>}
  n27 = Multiply(
   A=VectorParameter() {group=Wiryperk_CH, parameterName=HandFresnelEmissvieColor_Max, defaultValue=<rgba 1 1 1 1>}.RGB
   B=ScalarParameter() {defaultValue=0.3, group=Wiryperk_CH, parameterName=HandFresnelEmissvie_Max}
  )
  n28 = Multiply(
   A=VectorParameter() {group=Wiryperk_CH, parameterName=HandFresnelEmissvieColor_Min, defaultValue=<rgba 1 1 1 1>}.RGB
   B=ScalarParameter() {defaultValue=1.4, group=Wiryperk_CH, parameterName=HandFresnelEmissvie_Min}
  )
  n29 = Fresnel(
   ExponentIn=ScalarParameter() {defaultValue=0.8, group=Wiryperk_CH, parameterName=HandExponentln}
   BaseReflectFractionIn=ScalarParameter() {group=Wiryperk_CH, parameterName=HandBaseReflectFractionln}
  )
  n30 = TextureSampleParameter2D() {
   group=06_Emissive
   parameterName=Emissive
   samplerSource=SSM_Wrap_WorldGroupSettings
   sortPriority=1
   texture=&T_BlacksRGB
  }
  n31 = TextureSampleParameter2D() {
   group=00_Texture Inputs
   parameterName=Mask
   samplerSource=SSM_Wrap_WorldGroupSettings
   samplerType=SAMPLERTYPE_Masks
   sortPriority=3
   texture=&T_BlackMask
  }
  n32 = Multiply(
   A=Multiply(A=LinearInterpolate(A=@n27, B=@n28, Alpha=@n29), B=@n30.RGB)
   B=LinearInterpolate(
    A=ScalarParameter() {defaultValue=1, group=Wiryperk_CH, parameterName=HandEmissvie_Max}
    B=ScalarParameter() {defaultValue=0.4, group=Wiryperk_CH, parameterName=HandEmissvie_Min}
    Alpha=Power(Base=@n31.G)
   )
  )
  n33 = ScalarParameter() {defaultValue=2, group=Wiryperk_CH, parameterName=Global_Hand_Emissive}
  n34 = Multiply(A=Time(), B=Constant() {r=1.3}) {constB=0.01}
  n35 = NamedRerouteUsage()
  n36 = Saturate(None=Sine(None=Add(A=@n34, B=@n35)))
  n37 = Constant() {r=1}
  n38 = LinearInterpolate(
   B=LinearInterpolate(
    A=VectorParameter() {group=Wiryperk_CH, parameterName=HandGridColorMIn, defaultValue=<rgba 2.7679 0.2887 3 1>}.RGB
    B=VectorParameter() {group=Wiryperk_CH, parameterName=HandGridColorMax, defaultValue=<rgba 0.7758 0 0.85 1>}.RGB
    Alpha=Multiply(A=@n36, B=@n37) {constB=0.1}
   ) {constB=-0.2}
   Alpha=NamedRerouteUsage()
  )
  n39 = ScalarParameter() {defaultValue=0.7, group=Wiryperk_CH, parameterName=Gridbrightness}
  NamedRerouteDeclaration(None=Add(A=Multiply(A=@n32, B=@n33), B=Multiply(A=@n38, B=@n39) {constB=0.5})) {name=HandEmissvie, nodeColor=<rgba 0 0.9297 1 1>}
  n40 = LocalPosition() {localOrigin=InstancePreSkinning}
  n41 = Constant2Vector()
  n42 = Constant() {r=131.2953}
  NamedRerouteDeclaration(None=SphereMask(A=VertexInterpolator(VS=@n40.XYZ).PS, B=AppendVector(A=@n41, B=@n42), Radius=Constant() {r=52.9723}, Hardness=Constant() {r=-0.3})) {name=HandGridMask, nodeColor=<rgba 0.1719 1 0 1>}
  n43 = NamedRerouteUsage()
  n44 = StaticSwitchParameter(
   True=Multiply(
    A=MaterialFunctionCall!(In=@n43) {materialFunction=&/MF_WiryPerk_AudioEqualizer}.Result
    B=ScalarParameter() {defaultValue=1, group=Wiryperk_CH, parameterName=EqualizerMask_Weight}
   )
   False=Constant()
  ) {defaultValue=true, group=Wiryperk_CH, parameterName=Use AudioEqualizer}
  n45 = SphereMask(
   A=VertexInterpolator(VS=LocalPosition() {localOrigin=InstancePreSkinning}.Z).PS
   B=Constant() {r=-16.8304}
   Radius=ScalarParameter() {defaultValue=55, group=Wiryperk_CH, parameterName=LegEqualizerRadius}
   Hardness=Constant() {r=0.8038}
  )
  NamedRerouteDeclaration(None=Multiply(A=Max(A=@n44, B=@n45), B=ScalarParameter() {defaultValue=1, group=Wiryperk_CH, parameterName=LegAnimationMask_Blend})) {name=LegAnimation Mask, nodeColor=<rgba 0.0078 1 0 1>}
  n46 = ScalarParameter() {defaultValue=1, group=Wiryperk_CH, parameterName=EqualizerMask_Weight}
  n47 = Multiply(A=@n46, B=Multiply(A=@n22.R, B=@n22.R))
  NamedRerouteDeclaration(None=Saturate(None=Multiply(A=@n47) {constB=10})) {name=EqualizerGradation, nodeColor=<rgba 0 1 0.7188 1>}
  n48 = Multiply(A=@n22.R, B=@n46)
  NamedRerouteDeclaration(None=Saturate(None=Multiply(A=@n48) {constB=30})) {name=EqualizerMask, nodeColor=<rgba 0 1 0.7188 1>}
  NamedRerouteDeclaration(None=@n12.R) {name=HueShiftMask, nodeColor=<rgba 0 0.3437 1 1>}
 ]}
