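# Control Rig for the WiryPerk Seed large female body cosmetic, procedurally adjusting arms, joystick accessory and hood.
# Elbow-to-spine distance offsets pole vectors for two-bone arm IK.
# Hand proximity shifts the joystick target, driving extra arm bones through spring and aim constraints, with clavicle-driven elbows.
# Hood bones blend spine and clavicles; deform-rig attributes, head/neck angles, pose readers and targeting state weight corrective poses, including extra thumbs.
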ControlRigBlueprint /BRCosmetics/Characters/Player/Female/Large/Bodies/F_LRG_WiryPerk_Seed/Meshes/F_LRG_WiryPerk_Seed_CtrlRig
graphs [
 Add_1_ContainedGraph {nodes [
   n1 = RigVMFunctionEntryNode()
   Add = &RigVMFunction_MathDoubleAdd(A=@n1.A, B=@n1.B)
   RigVMFunctionReturnNode(Result=&RigVMFunction_MathDoubleAdd(A=@Add.Result, B=@n1.C).Result)
  ]}
 Add_4_ContainedGraph {nodes [
   n1 = RigVMFunctionEntryNode()
   RigVMFunctionReturnNode(Result=&RigVMFunction_MathDoubleAdd(A=&RigVMFunction_MathDoubleAdd(A=@n1.A, B=@n1.B).Result, B=@n1.C).Result)
  ]}
 CollapseNode_ContainedGraph {nodes [
   n1 = RigVMFunctionEntryNode()
   n2 = &RigUnit_GetTransform(Item=(Type=Control,Name="main_joystick_ctrl"), bInitial=False)
   RigVMFunction_MathTransformTransformVector = /Script/RigVM.RigVMFunction_MathTransformTransformVector(Location=(X=20,Y=0,Z=0), Transform=@n2.Transform) {is=struct RigVMUnitNode}
   RigVMFunction_DebugTransformMutableNoSpace = /Script/RigVM.RigVMFunction_DebugTransformMutableNoSpace(
    Color=(R=1,G=0,B=0,A=1)
    DebugDrawSettings=(DepthPriority=SDPG_Foreground,Lifetime=-1)
    Mode=Axes
    Scale=10
    Thickness=0
    Transform=(Rotation=(X=0,Y=0,Z=0,W=1),Translation=(X=0,Y=0,Z=0),Scale3D=(X=1,Y=1,Z=1))
    WorldOffset=(Rotation=(X=0,Y=0,Z=0,W=1),Translation=(X=0,Y=0,Z=0),Scale3D=(X=1,Y=1,Z=1))
    bEnabled=True
    ExecutePin=@n1.ExecutePin
    Transform.Translation=@RigVMFunction_MathTransformTransformVector.Result
   ) {is=struct RigVMUnitNode}
   n3 = RigVMDispatchNode(False=1, True=0.5, Condition=RigVMVariableNode() {variable=isMoving}.Value) {template="DISPATCH_RigVMDispatch_If(in Condition,in True,in False,out Result)"}
   n4 = &RigUnit_GetTransform(Item=(Type=Bone,Name="hand_l"), bInitial=False)
   RigVMFunction_MathVectorDistance = &RigVMFunction_MathVectorDistance(B=@n2.Transform.Translation, A=@n4.Transform.Translation)
   n5 = &RigUnit_GetTransform(Item=(Type=Bone,Name="hand_r"), bInitial=False)
   Remap = /Script/RigVM.RigVMFunction_MathVectorRemap(
    SourceMaximum=(X=14,Y=14,Z=1)
    SourceMinimum=(X=28,Y=28,Z=0)
    TargetMaximum=(X=-20,Y=-20,Z=1)
    TargetMinimum=(X=0,Y=0,Z=0)
    Value=(X=0,Y=0,Z=0)
    bClamp=true
    TargetMaximum.X=@n3.Result
    TargetMaximum.Y=@n3.Result
    Value.X=@RigVMFunction_MathVectorDistance.Result
    Value.Y=&RigVMFunction_MathVectorDistance(A=@n2.Transform.Translation, B=@n5.Transform.Translation).Result
   ) {
    is=struct RigVMUnitNode
    template="Remap::Execute(in Value,in SourceMinimum,in SourceMaximum,in TargetMinimum,in TargetMaximum,in bClamp,out Result)"
   }
   n6 = &RigVMFunction_MathVectorRemap(
    SourceMaximum=(X=10,Y=10,Z=1)
    SourceMinimum=(X=20,Y=20,Z=0)
    TargetMaximum=(X=0,Y=0,Z=0)
    TargetMinimum=(X=0,Y=0,Z=0)
    Value=(X=0,Y=0,Z=0)
    bClamp=true
    TargetMaximum.X=@n3.Result
    TargetMaximum.Y=@n3.Result
    Value.X=&RigVMFunction_MathVectorDistance(A=@n4.Transform.Translation, B=@RigVMFunction_MathTransformTransformVector.Result).Result
    Value.Y=&RigVMFunction_MathVectorDistance(A=@n5.Transform.Translation, B=@RigVMFunction_MathTransformTransformVector.Result).Result
   ) {template="Remap::Execute(in Value,in SourceMinimum,in SourceMaximum,in TargetMinimum,in TargetMaximum,in bClamp,out Result)"}
   Maximum = RigVMAggregateNode(A=@Remap.Result.X, B=@Remap.Result.Y, C=@n6.Result.X, D=@n6.Result.Y)
   AlphaInterp = &RigVMFunction_AlphaInterp(InterpSpeedDecreasing=10, InterpSpeedIncreasing=10, Scale=1, bClampResult=False, Value=@Maximum.Result)
   Multiply = &RigVMFunction_MathDoubleMul(B=-25, A=@AlphaInterp.Result)
   > &RigUnit_ModifyTransforms(
    ItemToModify=((Item=(Type=Control,Name="dyn_joystick_ctrl"),Transform=(Rotation=(X=0,Y=0,Z=0,W=1),Translation=(X=0,Y=0,Z=0),Scale3D=(X=1,Y=1,Z=1))))
    Mode=OverrideLocal
    Weight=1
    WeightMinimum=0
    ItemToModify.0.Transform.Translation.Z=@Multiply.Result
   )
   > &RigUnit_ParentConstraint(
    AdvancedSettings=(InterpolationType=Average,RotationOrderForFilter=XZY)
    Child=(Type=Bone,Name="base_aim_joystick")
    Parents=((Item=(Type=Control,Name="dyn_joystick_ctrl"),Weight=1))
    Weight=1
   )
   SpringInterp = /Script/ControlRig.RigUnit_SpringInterpVectorV2(
    CriticalDamping=1
    Force=(X=0,Y=0,Z=0)
    Strength=5
    TargetVelocityAmount=0
    bInitializeFromTarget=false
    bUseCurrentInput=false
    Target=&RigUnit_GetTransform(Item=(Type=Bone,Name="base_aim_joystick"), bInitial=False).Transform.Translation
    Current=&RigUnit_GetTransform(Item=(Type=Bone,Name="dyn_aim_joystick"), bInitial=False).Transform.Translation
   ) {
    is=struct RigVMUnitNode
    template="SpringInterp::Execute(in Target,in Strength,in CriticalDamping,in Force,in bUseCurrentInput,in Current,in TargetVelocityAmount,in bInitializeFromTarget,out Result,out Velocity)"
   }
   > &RigUnit_ModifyTransforms(
    ItemToModify=((Item=(Type=Bone,Name="dyn_aim_joystick"),Transform=(Rotation=(X=0,Y=0,Z=0,W=1),Translation=(X=0,Y=0,Z=0),Scale3D=(X=1,Y=1,Z=1))))
    Mode=OverrideGlobal
    WeightMinimum=0
    ItemToModify.0.Transform.Translation=@SpringInterp.Result
    Weight=RigVMVariableNode() {variable=is Playing Emote}.Value
   )
   > &RigUnit_ParentConstraint(
    AdvancedSettings=(InterpolationType=Average,RotationOrderForFilter=XZY)
    Child=(Type=Bone,Name="dyn_aim_joystick")
    Filter=(TranslationFilter=(bX=false,bY=false,bZ=false),RotationFilter=(bX=True,bY=True,bZ=True),ScaleFilter=(bX=false,bY=false,bZ=false))
    Parents=((Item=(Type=Bone,Name="clavicle_l"),Weight=1),(Item=(Type=Bone,Name="clavicle_r"),Weight=1))
    Weight=0.75
   )
   > &RigUnit_AimConstraintLocalSpaceOffset(
    Child=(Type=Bone,Name="aim_joystick")
    Filter=(bX=True,bY=True,bZ=True)
    Parents=((Item=(Type=Bone,Name="dyn_aim_joystick"),Weight=1))
    Weight=1
    WorldUp=(Target=(X=0,Y=0,Z=1),Kind=Direction,Space=(Type=Bone,Name="dyn_aim_joystick"))
   )
   > &RigUnit_ParentConstraint(
    AdvancedSettings=(InterpolationType=Average,RotationOrderForFilter=XZY)
    Child=(Type=Bone,Name="extra_base_arm_l")
    Filter=(TranslationFilter=(bX=false,bY=false,bZ=false),RotationFilter=(bX=true,bY=true,bZ=true),ScaleFilter=(bX=false,bY=false,bZ=false))
    Parents=((Item=(Type=Bone,Name="aim_joystick"),Weight=1))
    Weight=1
   )
   > &RigUnit_ParentConstraint(
    AdvancedSettings=(InterpolationType=Average,RotationOrderForFilter=XZY)
    Child=(Type=Bone,Name="extra_base_arm_r")
    Filter=(TranslationFilter=(bX=false,bY=false,bZ=false),RotationFilter=(bX=true,bY=true,bZ=true),ScaleFilter=(bX=false,bY=false,bZ=false))
    Parents=((Item=(Type=Bone,Name="aim_joystick"),Weight=1))
    Weight=1
   )
   > &RigUnit_ParentConstraint(
    AdvancedSettings=(InterpolationType=Average,RotationOrderForFilter=XZY)
    Child=(Type=Bone,Name="base_joystick")
    Filter=(TranslationFilter=(bX=true,bY=True,bZ=True),RotationFilter=(bX=True,bY=True,bZ=True),ScaleFilter=(bX=false,bY=false,bZ=false))
    Parents=((Item=(Type=Bone,Name="extra_base_arm_l"),Weight=1),(Item=(Type=Bone,Name="extra_base_arm_r"),Weight=1))
    Weight=1
   )
   n7 = RigVMVariableNode() {variable=clavicle_rot_l}
   n8 = &RigVMFunction_MathVectorRemap(
    SourceMaximum=(X=70,Y=70,Z=1)
    SourceMinimum=(X=-70,Y=-70,Z=0)
    TargetMaximum=(X=-45,Y=-25,Z=1)
    TargetMinimum=(X=45,Y=25,Z=0)
    Value=(X=0,Y=0,Z=0)
    bClamp=true
    Value.X=@n7.Value.Y
    Value.Y=@n7.Value.Y
   ) {template="Remap::Execute(in Value,in SourceMinimum,in SourceMaximum,in TargetMinimum,in TargetMaximum,in bClamp,out Result)"}
   n9 = RigVMVariableNode() {variable=clavicle_rot_r}
   n10 = &RigVMFunction_MathVectorRemap(
    SourceMaximum=(X=70,Y=70,Z=1)
    SourceMinimum=(X=-70,Y=-70,Z=0)
    TargetMaximum=(X=45,Y=25,Z=1)
    TargetMinimum=(X=-45,Y=-25,Z=0)
    Value=(X=0,Y=0,Z=0)
    bClamp=true
    Value.X=@n9.Value.Y
    Value.Y=@n9.Value.Y
   ) {template="Remap::Execute(in Value,in SourceMinimum,in SourceMaximum,in TargetMinimum,in TargetMaximum,in bClamp,out Result)"}
   RigVMFunction_MathQuaternionFromEuler = &RigVMFunction_MathQuaternionFromEuler(Euler=(X=0,Y=0,Z=0), Euler.X=&RigVMFunction_MathDoubleAdd(A=@n8.Result.X, B=@n10.Result.Y).Result)
   > n11 = &RigUnit_ModifyTransforms(
    ItemToModify="((Item=(Type=Bone,Name=\"extra_elbow_l\"),Transform=(Rotation=(X=0.3827,Y=0,Z=0,W=0.9239),Translation=(X=0,Y=0,Z=0),Scale3D=(X=1,Y=1,Z=1))),(Item=(Type=Bone,Name=\"extra_elbow_r\"),Transform=(Rotation=(X=0,Y=0,Z=0,W=1),Translation=(X=0,Y=0,Z=0),Scale3D=(X=1,Y= ...(truncated)"
    Weight=1
    WeightMinimum=0
    ItemToModify.0.Transform.Rotation=@RigVMFunction_MathQuaternionFromEuler.Result
    ItemToModify.1.Transform.Rotation=&RigVMFunction_MathQuaternionFromEuler(Euler=(X=0,Y=0,Z=0), Euler.X=&RigVMFunction_MathDoubleAdd(A=@n8.Result.Y, B=@n10.Result.X).Result).Result
   )
   RigVMFunctionReturnNode(ExecutePin=@n11.ExecutePin)
   n12 = &RigUnit_SphericalPoseReader(
    ActiveRegionScaleFactors=(PositiveWidth=1,NegativeWidth=1,PositiveHeight=1,NegativeHeight=1)
    ActiveRegionSize=0.1
    Debug=(bDrawDebug=True,bDraw2D=False,bDrawLocalAxes=False,DebugScale=25,DebugSegments=20,DebugThickness=0.25)
    DriverAxis=(X=0,Y=1,Z=0)
    DriverItem=(Type=Bone,Name="head")
    ExecutePin=()
    FalloffRegionScaleFactors=(PositiveWidth=1,NegativeWidth=1,PositiveHeight=1,NegativeHeight=1)
    FalloffSize=0.2
    FlipHeightScaling=False
    FlipWidthScaling=False
    OptionalParentItem=(Type=Bone,Name="None")
    RotationOffset=(X=95,Y=0,Z=-5)
   )
   > &RigUnit_ModifyTransforms(
    ItemToModify=((Item=(Type=Bone,Name="extra_base_arm_r"),Transform=(Rotation=(X=0.0816,Y=-0.0605,Z=-0.005,W=0.9948),Translation=(X=0.9228,Y=-0.1305,Z=0.7902),Scale3D=(X=1,Y=1,Z=1))))
    WeightMinimum=0
    Weight=&RigVMFunction_MathDoubleMul(A=@n12.OutputParam, B=RigVMVariableNode() {variable=is Targeting}.Value).Result
   )
  ]}
 CollapseNode_ContainedGraph {nodes [
   n1 = RigVMFunctionEntryNode()
   &RigUnit_ParentConstraint(
    AdvancedSettings=(InterpolationType=Average,RotationOrderForFilter=XZY)
    Child=(Type=Control,Name="wrist_l_ctrl")
    Parents=((Item=(Type=Bone,Name="hand_l"),Weight=1))
    Weight=1
    ExecutePin=@n1.ExecutePin
   )
   > &RigUnit_ParentConstraint(
    AdvancedSettings=(InterpolationType=Average,RotationOrderForFilter=XZY)
    Child=(Type=Control,Name="elbow_l_ctrl")
    Parents=((Item=(Type=Bone,Name="upperarm_l"),Weight=1))
    Weight=1
   )
   > &RigUnit_ParentConstraint(
    AdvancedSettings=(InterpolationType=Average,RotationOrderForFilter=XZY)
    Child=(Type=Control,Name="wrist_r_ctrl")
    Parents=((Item=(Type=Bone,Name="hand_r"),Weight=1))
    Weight=1
   )
   > &RigUnit_ParentConstraint(
    AdvancedSettings=(InterpolationType=Average,RotationOrderForFilter=XZY)
    Child=(Type=Control,Name="elbow_r_ctrl")
    Parents=((Item=(Type=Bone,Name="upperarm_r"),Weight=1))
    Weight=1
   )
   > /Script/ControlRig.RigUnit_PositionConstraintLocalSpaceOffset(
    Child=(Type=Control,Name="dyn_elbow_l_ctrl")
    Filter=(bX=True,bY=True,bZ=True)
    Parents=((Item=(Type=Control,Name="elbow_l_ctrl"),Weight=1))
    Weight=1
    bMaintainOffset=True
   ) {is=struct RigVMUnitNode}
   > &RigUnit_PositionConstraintLocalSpaceOffset(
    Child=(Type=Control,Name="dyn_elbow_r_ctrl")
    Filter=(bX=True,bY=True,bZ=True)
    Parents=((Item=(Type=Control,Name="elbow_r_ctrl"),Weight=1))
    Weight=1
    bMaintainOffset=True
   )
   n2 = &RigUnit_GetTransform(Item=(Type=Bone,Name="spine_02"), bInitial=False)
   > RigVMVariableNode(Value=&RigVMFunction_MathVectorDistance(
     A=@n2.Transform.Translation
     B=&RigUnit_GetTransform(Item=(Type=Control,Name="elbow_r_ctrl"), bInitial=False).Transform.Translation
    ).Result) {variable=elbow_r_distance_spine_02}
   RigVMFunction_MathVectorDistance = &RigVMFunction_MathVectorDistance(
    B=@n2.Transform.Translation
    A=&RigUnit_GetTransform(Item=(Type=Control,Name="elbow_l_ctrl"), bInitial=False).Transform.Translation
   )
   > n3 = RigVMVariableNode(Value=@RigVMFunction_MathVectorDistance.Result) {variable=elbow_l_distance_spine_02}
   RigVMFunctionReturnNode(ExecutePin=@n3.ExecuteContext)
  ]}
 CollapseNode_ContainedGraph {nodes [
   n1 = RigVMFunctionEntryNode()
   Remap = &RigVMFunction_MathDoubleRemap(SourceMaximum=14, SourceMinimum=36, TargetMaximum=1, TargetMinimum=0, Value=RigVMVariableNode() {variable=elbow_l_distance_spine_02}.Value)
   AlphaInterp = &RigVMFunction_AlphaInterp(InterpSpeedDecreasing=10, InterpSpeedIncreasing=10, Scale=1, bClampResult=False, bMapRange=true, Value=@Remap.Result)
   Multiply = &RigVMFunction_MathDoubleMul(B=40, A=@AlphaInterp.Result)
   n2 = &RigVMFunction_AlphaInterp(
    InterpSpeedDecreasing=10
    InterpSpeedIncreasing=10
    Scale=1
    bClampResult=False
    bMapRange=true
    Value=&RigVMFunction_MathDoubleRemap(SourceMaximum=14, SourceMinimum=36, TargetMaximum=1, TargetMinimum=0, Value=RigVMVariableNode() {variable=elbow_r_distance_spine_02}.Value).Result
   )
   &RigUnit_ModifyTransforms(
    ItemToModify="((Item=(Type=Control,Name=\"dyn_elbow_l_ctrl\"),Transform=(Rotation=(X=0,Y=0,Z=0,W=1),Translation=(X=0,Y=0,Z=0),Scale3D=(X=1,Y=1,Z=1))),(Item=(Type=Control,Name=\"dyn_elbow_r_ctrl\"),Transform=(Rotation=(X=0,Y=0,Z=0,W=1),Translation=(X=0,Y=0,Z=0),Scale3D=(X=1, ...(truncated)"
    Weight=1
    WeightMinimum=0
    ExecutePin=@n1.ExecutePin
    ItemToModify.0.Transform.Translation.Z=@Multiply.Result
    ItemToModify.1.Transform.Translation.Z=&RigVMFunction_MathDoubleMul(B=-32, A=@n2.Result).Result
   )
   > /Script/ControlRig.RigUnit_TwoBoneIKSimplePerItem(
    DebugSettings=(bEnabled=False,Scale=10,WorldOffset=(Rotation=(X=0,Y=0,Z=0,W=1),Translation=(X=0,Y=0,Z=0),Scale3D=(X=1,Y=1,Z=1)))
    EffectorItem=(Type=Bone,Name="hand_l")
    ItemA=(Type=Bone,Name="upperarm_l")
    ItemALength=0
    ItemB=(Type=Bone,Name="lowerarm_l")
    ItemBLength=0
    PoleVectorKind=Location
    PoleVectorSpace=(Type=Bone,Name="None")
    PrimaryAxis=(X=1,Y=0,Z=0)
    SecondaryAxis=(X=0,Y=0,Z=0)
    SecondaryAxisWeight=1
    StretchMaximumRatio=1.25
    StretchStartRatio=0.75
    Weight=1
    bEnableStretch=False
    bPropagateToChildren=True
    Effector=&RigUnit_GetTransform(Item=(Type=Control,Name="wrist_l_ctrl"), bInitial=False).Transform
    PoleVector=&RigUnit_GetTransform(Item=(Type=Control,Name="dyn_elbow_l_ctrl"), bInitial=False).Transform.Translation
   ) {is=struct RigVMUnitNode}
   > n3 = &RigUnit_TwoBoneIKSimplePerItem(
    DebugSettings=(bEnabled=False,Scale=10,WorldOffset=(Rotation=(X=0,Y=0,Z=0,W=1),Translation=(X=0,Y=0,Z=0),Scale3D=(X=1,Y=1,Z=1)))
    EffectorItem=(Type=Bone,Name="hand_r")
    ItemA=(Type=Bone,Name="upperarm_r")
    ItemALength=0
    ItemB=(Type=Bone,Name="lowerarm_r")
    ItemBLength=0
    PoleVectorKind=Location
    PoleVectorSpace=(Type=Bone,Name="None")
    PrimaryAxis=(X=-1,Y=0,Z=0)
    SecondaryAxis=(X=0,Y=0,Z=0)
    SecondaryAxisWeight=1
    StretchMaximumRatio=1.25
    StretchStartRatio=0.75
    Weight=1
    bEnableStretch=False
    bPropagateToChildren=True
    Effector=&RigUnit_GetTransform(Item=(Type=Control,Name="wrist_r_ctrl"), bInitial=False).Transform
    PoleVector=&RigUnit_GetTransform(Item=(Type=Control,Name="dyn_elbow_r_ctrl"), bInitial=False).Transform.Translation
   )
   RigVMFunctionReturnNode(ExecutePin=@n3.ExecutePin)
  ]}
 CollapseNode_ContainedGraph {nodes [
   n1 = RigVMFunctionEntryNode()
   &RigUnit_ModifyTransforms(
    ItemToModify="((Item=(Type=Bone,Name=\"extra_thumb_01_l\"),Transform=(Rotation=(X=0,Y=0,Z=-0.165,W=0.9863),Translation=(X=0,Y=0,Z=0),Scale3D=(X=1,Y=1,Z=1))),(Item=(Type=Bone,Name=\"extra_thumb_02_l\"),Transform=(Rotation=(X=0,Y=0,Z=0.3007,W=0.9537),Translation=(X=0,Y=0,Z=0) ...(truncated)"
    WeightMinimum=0
    Weight=RigVMDispatchNode(BoneName=root, Default=0, Name=deformRig_l_upperarm_up_alpha) {template="DISPATCH_RigDispatch_GetAnimAttribute(in Name,in BoneName,in Default,out Value,out Found)"}.Value
    ExecutePin=@n1.ExecutePin
   )
   Add = &RigVMFunction_MathDoubleAdd(
    A=RigVMDispatchNode(BoneName=root, Default=0, Name=deformRig_l_upperarm_bk_alpha) {template="DISPATCH_RigDispatch_GetAnimAttribute(in Name,in BoneName,in Default,out Value,out Found)"}.Value
    B=RigVMDispatchNode(BoneName=root, Default=0, Name=deformRig_l_elbow_alpha) {template="DISPATCH_RigDispatch_GetAnimAttribute(in Name,in BoneName,in Default,out Value,out Found)"}.Value
   )
   > &RigUnit_ModifyTransforms(
    ItemToModify="((Item=(Type=Bone,Name=\"extra_thumb_01_l\"),Transform=(Rotation=(X=-0.0011,Y=-0.061,Z=-0.0174,W=0.998),Translation=(X=0,Y=0,Z=0),Scale3D=(X=1,Y=1,Z=1))),(Item=(Type=Bone,Name=\"extra_thumb_02_l\"),Transform=(Rotation=(X=0,Y=0,Z=0.0175,W=0.9998),Translation=(X ...(truncated)"
    WeightMinimum=0
    Weight=@Add.Result
   )
   > &RigUnit_ModifyTransforms(
    ItemToModify="((Item=(Type=Bone,Name=\"extra_thumb_01_l\"),Transform=(Rotation=(X=-0.0057,Y=0.1304,Z=0.0432,W=0.9905),Translation=(X=0,Y=0,Z=0),Scale3D=(X=1,Y=1,Z=1))),(Item=(Type=Bone,Name=\"extra_thumb_02_l\"),Transform=(Rotation=(X=0,Y=0,Z=-0.1305,W=0.9914),Translation=( ...(truncated)"
    WeightMinimum=0
    Weight=RigVMDispatchNode(BoneName=root, Default=0, Name=deformRig_l_upperarm_fr_alpha) {template="DISPATCH_RigDispatch_GetAnimAttribute(in Name,in BoneName,in Default,out Value,out Found)"}.Value
   )
   > &RigUnit_ModifyTransforms(
    ItemToModify="((Item=(Type=Bone,Name=\"extra_thumb_01_r\"),Transform=(Rotation=(X=0,Y=0,Z=-0.1908,W=0.9816),Translation=(X=0,Y=0,Z=0),Scale3D=(X=1,Y=1,Z=1))),(Item=(Type=Bone,Name=\"extra_thumb_02_r\"),Transform=(Rotation=(X=0,Y=0,Z=0.3007,W=0.9537),Translation=(X=0,Y=0,Z=0 ...(truncated)"
    WeightMinimum=0
    Weight=RigVMDispatchNode(BoneName=root, Default=0, Name=deformRig_r_upperarm_up_alpha) {template="DISPATCH_RigDispatch_GetAnimAttribute(in Name,in BoneName,in Default,out Value,out Found)"}.Value
   )
   > &RigUnit_ModifyTransforms(
    ItemToModify="((Item=(Type=Bone,Name=\"extra_thumb_01_r\"),Transform=(Rotation=(X=-0.0136,Y=-0.0861,Z=-0.1558,W=0.9839),Translation=(X=0,Y=0,Z=0),Scale3D=(X=1,Y=1,Z=1))),(Item=(Type=Bone,Name=\"extra_thumb_02_r\"),Transform=(Rotation=(X=0,Y=0,Z=0.1305,W=0.9914),Translation= ...(truncated)"
    WeightMinimum=0
    Weight=&RigVMFunction_MathDoubleAdd(
     A=RigVMDispatchNode(BoneName=root, Default=0, Name=deformRig_r_upperarm_bk_alpha) {template="DISPATCH_RigDispatch_GetAnimAttribute(in Name,in BoneName,in Default,out Value,out Found)"}.Value
     B=RigVMDispatchNode(BoneName=root, Default=0, Name=deformRig_r_elbow_alpha) {template="DISPATCH_RigDispatch_GetAnimAttribute(in Name,in BoneName,in Default,out Value,out Found)"}.Value
    ).Result
   )
   > n2 = &RigUnit_ModifyTransforms(
    ItemToModify="((Item=(Type=Bone,Name=\"extra_thumb_01_r\"),Transform=(Rotation=(X=0.0015,Y=-0.0871,Z=0.0174,W=0.996),Translation=(X=0,Y=0,Z=0),Scale3D=(X=1,Y=1,Z=1))),(Item=(Type=Bone,Name=\"extra_thumb_02_r\"),Transform=(Rotation=(X=0,Y=0,Z=-0.1305,W=0.9914),Translation=(X ...(truncated)"
    WeightMinimum=0
    Weight=RigVMDispatchNode(BoneName=root, Default=0, Name=deformRig_r_upperarm_fr_alpha) {template="DISPATCH_RigDispatch_GetAnimAttribute(in Name,in BoneName,in Default,out Value,out Found)"}.Value
   )
   RigVMFunctionReturnNode(ExecutePin=@n2.ExecutePin)
   &RigUnit_ModifyTransforms(
    ExecutePin=()
    ItemToModify="((Item=(Type=Bone,Name=\"extra_thumb_01_l\"),Transform=(Rotation=(X=0,Y=0,Z=-0.2588,W=0.9659),Translation=(X=0,Y=0,Z=0),Scale3D=(X=1,Y=1,Z=1))),(Item=(Type=Bone,Name=\"extra_thumb_02_l\"),Transform=(Rotation=(X=0,Y=0,Z=0.3007,W=0.9537),Translation=(X=0,Y=0,Z=0 ...(truncated)"
    Weight=1
    WeightMinimum=0
   )
   > &RigUnit_ModifyTransforms(
    ItemToModify="((Item=(Type=Bone,Name=\"extra_thumb_01_l\"),Transform=(Rotation=(X=-0.0151,Y=-0.173,Z=-0.0858,W=0.9811),Translation=(X=0,Y=0,Z=0),Scale3D=(X=1,Y=1,Z=1))),(Item=(Type=Bone,Name=\"extra_thumb_02_l\"),Transform=(Rotation=(X=0,Y=0,Z=0.1045,W=0.9945),Translation=( ...(truncated)"
    Weight=1
    WeightMinimum=0
   )
   > &RigUnit_ModifyTransforms(
    ItemToModify="((Item=(Type=Bone,Name=\"extra_thumb_01_l\"),Transform=(Rotation=(X=-0.0151,Y=0.173,Z=0.0858,W=0.9811),Translation=(X=0,Y=0,Z=0),Scale3D=(X=1,Y=1,Z=1))),(Item=(Type=Bone,Name=\"extra_thumb_02_l\"),Transform=(Rotation=(X=0,Y=0,Z=-0.1305,W=0.9914),Translation=(X ...(truncated)"
    Weight=0
    WeightMinimum=0
   )
   > &RigUnit_ModifyTransforms(
    ItemToModify="((Item=(Type=Bone,Name=\"extra_thumb_01_r\"),Transform=(Rotation=(X=0,Y=0,Z=-0.2588,W=0.9659),Translation=(X=0,Y=0,Z=0),Scale3D=(X=1,Y=1,Z=1))),(Item=(Type=Bone,Name=\"extra_thumb_02_r\"),Transform=(Rotation=(X=0,Y=0,Z=0.3007,W=0.9537),Translation=(X=0,Y=0,Z=0 ...(truncated)"
    Weight=1
    WeightMinimum=0
   )
   > &RigUnit_ModifyTransforms(
    ItemToModify="((Item=(Type=Bone,Name=\"extra_thumb_01_r\"),Transform=(Rotation=(X=0.0076,Y=0.1735,Z=-0.043,W=0.9839),Translation=(X=0,Y=0,Z=0),Scale3D=(X=1,Y=1,Z=1))),(Item=(Type=Bone,Name=\"extra_thumb_02_r\"),Transform=(Rotation=(X=0,Y=0,Z=0.1305,W=0.9914),Translation=(X= ...(truncated)"
    Weight=0
    WeightMinimum=0
   )
   > &RigUnit_ModifyTransforms(
    ItemToModify="((Item=(Type=Bone,Name=\"extra_thumb_01_r\"),Transform=(Rotation=(X=0.0151,Y=-0.173,Z=0.0858,W=0.9811),Translation=(X=0,Y=0,Z=0),Scale3D=(X=1,Y=1,Z=1))),(Item=(Type=Bone,Name=\"extra_thumb_02_r\"),Transform=(Rotation=(X=0,Y=0,Z=-0.1305,W=0.9914),Translation=(X ...(truncated)"
    Weight=0
    WeightMinimum=0
   )
  ]}
 CollapseNode_ContainedGraph {nodes [
   RigVMFunctionEntryNode()
   > Control Stack at Item_4 = RigVMFunctionReferenceNode(
    Add Bottom Null=False
    Add Null=true
    Add Secondary Control=False
    Control Suffix=_ctrl
    Item=(Name="base_aim_joystick",Type=Bone)
    Name=main_joystick
    Null Suffix=_null
    Parent=(Name="spine_04",Type=Bone)
    Secondary Settings="(Shape=(bVisible=True,Name=\"Default\",Color=(R=1,G=0,B=0,A=1),Transform=(Rotation=(X=0,Y=0,Z=0,W=1),Translation=(X=0,Y=0,Z=0),Scale3D=(X=1,Y=1,Z=1))),Proxy=(bIsProxy=False,ShapeVisibility=BasedOnSelection),DisplayName=\"None\",InitialSpace=LocalSpace,bUsePref ...(truncated)"
    Settings="(Shape=(bVisible=True,Name=\"Box_Thick\",Color=(R=1,G=0,B=0,A=1),Transform=(Rotation=(X=0,Y=0,Z=0,W=1),Translation=(X=0,Y=0,Z=0),Scale3D=(X=0.75,Y=0.75,Z=0.75))),Proxy=(bIsProxy=False,ShapeVisibility=BasedOnSelection),DisplayName=\"None\",InitialSpace=LocalSpa ...(truncated)"
   )
   > Control Stack at Item_5 = RigVMFunctionReferenceNode(
    Add Bottom Null=False
    Add Null=true
    Add Secondary Control=False
    Control Suffix=_ctrl
    Item=(Name="base_aim_joystick",Type=Bone)
    Name=dyn_joystick
    Null Suffix=_null
    Parent=(Name="main_joystick_ctrl",Type=Control)
    Secondary Settings="(Shape=(bVisible=True,Name=\"Default\",Color=(R=1,G=0,B=0,A=1),Transform=(Rotation=(X=0,Y=0,Z=0,W=1),Translation=(X=0,Y=0,Z=0),Scale3D=(X=1,Y=1,Z=1))),Proxy=(bIsProxy=False,ShapeVisibility=BasedOnSelection),DisplayName=\"None\",InitialSpace=LocalSpace,bUsePref ...(truncated)"
    Settings="(Shape=(bVisible=True,Name=\"Pyramid_Thick\",Color=(R=0,G=0,B=1,A=1),Transform=(Rotation=(X=0,Y=0,Z=0,W=1),Translation=(X=0,Y=0,Z=0),Scale3D=(X=1,Y=1,Z=1))),Proxy=(bIsProxy=False,ShapeVisibility=BasedOnSelection),DisplayName=\"None\",InitialSpace=LocalSpace,bU ...(truncated)"
   )
   > Control Stack at Item_9 = RigVMFunctionReferenceNode(
    Add Bottom Null=False
    Add Null=true
    Add Secondary Control=False
    Control Suffix=_ctrl
    Item=(Name="clavicle_l",Type=Bone)
    Name=ref_clavicle_l
    Null Suffix=_null
    Parent=(Name="spine_05",Type=Bone)
    Secondary Settings="(Shape=(bVisible=True,Name=\"Default\",Color=(R=1,G=0,B=0,A=1),Transform=(Rotation=(X=0,Y=0,Z=0,W=1),Translation=(X=0,Y=0,Z=0),Scale3D=(X=1,Y=1,Z=1))),Proxy=(bIsProxy=False,ShapeVisibility=BasedOnSelection),DisplayName=\"None\",InitialSpace=LocalSpace,bUsePref ...(truncated)"
    Settings="(Shape=(bVisible=True,Name=\"Sphere_Thick\",Color=(R=1,G=1,B=1,A=1),Transform=(Rotation=(X=0,Y=0,Z=0,W=1),Translation=(X=0,Y=0,Z=0),Scale3D=(X=1,Y=1,Z=1))),Proxy=(bIsProxy=False,ShapeVisibility=BasedOnSelection),DisplayName=\"None\",InitialSpace=LocalSpace,bUs ...(truncated)"
   )
   > Control Stack at Item_8 = RigVMFunctionReferenceNode(
    Add Bottom Null=False
    Add Null=true
    Add Secondary Control=False
    Control Suffix=_ctrl
    Item=(Name="clavicle_r",Type=Bone)
    Name=ref_clavicle_r
    Null Suffix=_null
    Parent=(Name="spine_05",Type=Bone)
    Secondary Settings="(Shape=(bVisible=True,Name=\"Default\",Color=(R=1,G=0,B=0,A=1),Transform=(Rotation=(X=0,Y=0,Z=0,W=1),Translation=(X=0,Y=0,Z=0),Scale3D=(X=1,Y=1,Z=1))),Proxy=(bIsProxy=False,ShapeVisibility=BasedOnSelection),DisplayName=\"None\",InitialSpace=LocalSpace,bUsePref ...(truncated)"
    Settings="(Shape=(bVisible=True,Name=\"Sphere_Thick\",Color=(R=1,G=1,B=1,A=1),Transform=(Rotation=(X=0,Y=0,Z=0,W=1),Translation=(X=0,Y=0,Z=0),Scale3D=(X=1,Y=1,Z=1))),Proxy=(bIsProxy=False,ShapeVisibility=BasedOnSelection),DisplayName=\"None\",InitialSpace=LocalSpace,bUs ...(truncated)"
   )
   > RigVMFunctionReturnNode()
  ]}
 CollapseNode_ContainedGraph {nodes [
   RigVMFunctionEntryNode()
   > Control Stack at Item = RigVMFunctionReferenceNode(
    Add Bottom Null=False
    Add Null=true
    Add Secondary Control=False
    Control Suffix=_ctrl
    Item=(Name="hand_l",Type=Bone)
    Name=wrist_l
    Null Suffix=_null
    Parent=(Name="clavicle_l",Type=Bone)
    Secondary Settings="(Shape=(bVisible=True,Name=\"Default\",Color=(R=1,G=0,B=0,A=1),Transform=(Rotation=(X=0,Y=0,Z=0,W=1),Translation=(X=0,Y=0,Z=0),Scale3D=(X=1,Y=1,Z=1))),Proxy=(bIsProxy=False,ShapeVisibility=BasedOnSelection),DisplayName=\"None\",InitialSpace=LocalSpace,bUsePref ...(truncated)"
    Settings="(Shape=(bVisible=True,Name=\"Arrow4_Thin\",Color=(R=0,G=1,B=0,A=1),Transform=(Rotation=(X=0,Y=-0.7071,Z=0,W=0.7071),Translation=(X=0,Y=0,Z=0),Scale3D=(X=4,Y=4,Z=4))),Proxy=(bIsProxy=False,ShapeVisibility=BasedOnSelection),DisplayName=\"None\",InitialSpace=Loca ...(truncated)"
   )
   > Control Stack at Item_1 = RigVMFunctionReferenceNode(
    Add Bottom Null=False
    Add Null=true
    Add Secondary Control=False
    Control Suffix=_ctrl
    Item=(Name="lowerarm_l",Type=Bone)
    Name=elbow_l
    Null Suffix=_null
    Parent=(Name="clavicle_l",Type=Bone)
    Secondary Settings="(Shape=(bVisible=True,Name=\"Default\",Color=(R=1,G=0,B=0,A=1),Transform=(Rotation=(X=0,Y=0,Z=0,W=1),Translation=(X=0,Y=0,Z=0),Scale3D=(X=1,Y=1,Z=1))),Proxy=(bIsProxy=False,ShapeVisibility=BasedOnSelection),DisplayName=\"None\",InitialSpace=LocalSpace,bUsePref ...(truncated)"
    Settings="(Shape=(bVisible=True,Name=\"Box_Thick\",Color=(R=0,G=1,B=0,A=1),Transform=(Rotation=(X=0,Y=0,Z=0,W=1),Translation=(X=0,Y=0,Z=0),Scale3D=(X=1,Y=1,Z=1))),Proxy=(bIsProxy=False,ShapeVisibility=BasedOnSelection),DisplayName=\"None\",InitialSpace=LocalSpace,bUsePr ...(truncated)"
   )
   > Control Stack at Item_2 = RigVMFunctionReferenceNode(
    Add Bottom Null=False
    Add Null=true
    Add Secondary Control=False
    Control Suffix=_ctrl
    Item=(Name="hand_r",Type=Bone)
    Name=wrist_r
    Null Suffix=_null
    Parent=(Name="clavicle_r",Type=Bone)
    Secondary Settings="(Shape=(bVisible=True,Name=\"Default\",Color=(R=1,G=0,B=0,A=1),Transform=(Rotation=(X=0,Y=0,Z=0,W=1),Translation=(X=0,Y=0,Z=0),Scale3D=(X=1,Y=1,Z=1))),Proxy=(bIsProxy=False,ShapeVisibility=BasedOnSelection),DisplayName=\"None\",InitialSpace=LocalSpace,bUsePref ...(truncated)"
    Settings="(Shape=(bVisible=True,Name=\"Arrow4_Thin\",Color=(R=0,G=1,B=0,A=1),Transform=(Rotation=(X=0,Y=-0.7071,Z=0,W=0.7071),Translation=(X=0,Y=0,Z=0),Scale3D=(X=4,Y=4,Z=4))),Proxy=(bIsProxy=False,ShapeVisibility=BasedOnSelection),DisplayName=\"None\",InitialSpace=Loca ...(truncated)"
   )
   > Control Stack at Item_3 = RigVMFunctionReferenceNode(
    Add Bottom Null=False
    Add Null=true
    Add Secondary Control=False
    Control Suffix=_ctrl
    Item=(Name="lowerarm_r",Type=Bone)
    Name=elbow_r
    Null Suffix=_null
    Parent=(Name="clavicle_r",Type=Bone)
    Secondary Settings="(Shape=(bVisible=True,Name=\"Default\",Color=(R=1,G=0,B=0,A=1),Transform=(Rotation=(X=0,Y=0,Z=0,W=1),Translation=(X=0,Y=0,Z=0),Scale3D=(X=1,Y=1,Z=1))),Proxy=(bIsProxy=False,ShapeVisibility=BasedOnSelection),DisplayName=\"None\",InitialSpace=LocalSpace,bUsePref ...(truncated)"
    Settings="(Shape=(bVisible=True,Name=\"Box_Thick\",Color=(R=0,G=1,B=0,A=1),Transform=(Rotation=(X=0,Y=0,Z=0,W=1),Translation=(X=0,Y=0,Z=0),Scale3D=(X=1,Y=1,Z=1))),Proxy=(bIsProxy=False,ShapeVisibility=BasedOnSelection),DisplayName=\"None\",InitialSpace=LocalSpace,bUsePr ...(truncated)"
   )
   > Control Stack at Item_7 = RigVMFunctionReferenceNode(
    Add Bottom Null=False
    Add Null=true
    Add Secondary Control=False
    Control Suffix=_ctrl
    Item=(Name="lowerarm_l",Type=Bone)
    Name=dyn_elbow_l
    Null Suffix=_null
    Parent=(Name="spine_04",Type=Bone)
    Secondary Settings="(Shape=(bVisible=True,Name=\"Default\",Color=(R=1,G=0,B=0,A=1),Transform=(Rotation=(X=0,Y=0,Z=0,W=1),Translation=(X=0,Y=0,Z=0),Scale3D=(X=1,Y=1,Z=1))),Proxy=(bIsProxy=False,ShapeVisibility=BasedOnSelection),DisplayName=\"None\",InitialSpace=LocalSpace,bUsePref ...(truncated)"
    Settings="(Shape=(bVisible=True,Name=\"Sphere_Thick\",Color=(R=1,G=1,B=0,A=1),Transform=(Rotation=(X=0,Y=0,Z=0,W=1),Translation=(X=0,Y=0,Z=0),Scale3D=(X=1,Y=1,Z=1))),Proxy=(bIsProxy=False,ShapeVisibility=BasedOnSelection),DisplayName=\"None\",InitialSpace=LocalSpace,bUs ...(truncated)"
   )
   > Control Stack at Item_6 = RigVMFunctionReferenceNode(
    Add Bottom Null=False
    Add Null=true
    Add Secondary Control=False
    Control Suffix=_ctrl
    Item=(Name="lowerarm_r",Type=Bone)
    Name=dyn_elbow_r
    Null Suffix=_null
    Parent=(Name="spine_04",Type=Bone)
    Secondary Settings="(Shape=(bVisible=True,Name=\"Default\",Color=(R=1,G=0,B=0,A=1),Transform=(Rotation=(X=0,Y=0,Z=0,W=1),Translation=(X=0,Y=0,Z=0),Scale3D=(X=1,Y=1,Z=1))),Proxy=(bIsProxy=False,ShapeVisibility=BasedOnSelection),DisplayName=\"None\",InitialSpace=LocalSpace,bUsePref ...(truncated)"
    Settings="(Shape=(bVisible=True,Name=\"Sphere_Thick\",Color=(R=1,G=1,B=0,A=1),Transform=(Rotation=(X=0,Y=0,Z=0,W=1),Translation=(X=0,Y=0,Z=0),Scale3D=(X=1,Y=1,Z=1))),Proxy=(bIsProxy=False,ShapeVisibility=BasedOnSelection),DisplayName=\"None\",InitialSpace=LocalSpace,bUs ...(truncated)"
   )
   > RigVMFunctionReturnNode()
  ]}
 CollapseNode_ContainedGraph {nodes [
   n1 = RigVMFunctionEntryNode()
   &RigUnit_ParentConstraint(
    AdvancedSettings=(InterpolationType=Average,RotationOrderForFilter=XZY)
    Child=(Type=Bone,Name="base_hood_back_c")
    Filter=(TranslationFilter=(bX=true,bY=true,bZ=true),RotationFilter=(bX=True,bY=True,bZ=True),ScaleFilter=(bX=True,bY=True,bZ=True))
    Parents=((Item=(Type=Bone,Name="spine_05"),Weight=0.8),(Item=(Type=Bone,Name="clavicle_l"),Weight=0.1),(Item=(Type=Bone,Name="clavicle_r"),Weight=0.1))
    Weight=1
    ExecutePin=@n1.ExecutePin
   )
   > &RigUnit_ParentConstraint(
    AdvancedSettings=(InterpolationType=Average,RotationOrderForFilter=XZY)
    Child=(Type=Bone,Name="base_hood_front_l")
    Filter=(TranslationFilter=(bX=true,bY=true,bZ=true),RotationFilter=(bX=True,bY=True,bZ=True),ScaleFilter=(bX=True,bY=True,bZ=True))
    Parents=((Item=(Type=Bone,Name="spine_05"),Weight=0.7),(Item=(Type=Bone,Name="clavicle_l"),Weight=0.3))
    Weight=1
   )
   > &RigUnit_ParentConstraint(
    AdvancedSettings=(InterpolationType=Average,RotationOrderForFilter=XZY)
    Child=(Type=Bone,Name="base_hood_side_l")
    Filter=(TranslationFilter=(bX=true,bY=true,bZ=true),RotationFilter=(bX=True,bY=True,bZ=True),ScaleFilter=(bX=True,bY=True,bZ=True))
    Parents=((Item=(Type=Bone,Name="spine_05"),Weight=0.7),(Item=(Type=Bone,Name="clavicle_l"),Weight=0.3))
    Weight=1
   )
   > &RigUnit_ParentConstraint(
    AdvancedSettings=(InterpolationType=Average,RotationOrderForFilter=XZY)
    Child=(Type=Bone,Name="base_hood_back_side_l")
    Filter=(TranslationFilter=(bX=true,bY=true,bZ=true),RotationFilter=(bX=True,bY=True,bZ=True),ScaleFilter=(bX=True,bY=True,bZ=True))
    Parents=((Item=(Type=Bone,Name="spine_05"),Weight=0.75),(Item=(Type=Bone,Name="clavicle_l"),Weight=0.15))
    Weight=1
   )
   > &RigUnit_ParentConstraint(
    AdvancedSettings=(InterpolationType=Average,RotationOrderForFilter=XZY)
    Child=(Type=Bone,Name="base_hood_front_r")
    Filter=(TranslationFilter=(bX=true,bY=true,bZ=true),RotationFilter=(bX=True,bY=True,bZ=True),ScaleFilter=(bX=True,bY=True,bZ=True))
    Parents=((Item=(Type=Bone,Name="spine_05"),Weight=0.7),(Item=(Type=Bone,Name="clavicle_r"),Weight=0.3))
    Weight=1
   )
   > &RigUnit_ParentConstraint(
    AdvancedSettings=(InterpolationType=Average,RotationOrderForFilter=XZY)
    Child=(Type=Bone,Name="base_hood_side_r")
    Filter=(TranslationFilter=(bX=true,bY=true,bZ=true),RotationFilter=(bX=True,bY=True,bZ=True),ScaleFilter=(bX=True,bY=True,bZ=True))
    Parents=((Item=(Type=Bone,Name="spine_05"),Weight=0.7),(Item=(Type=Bone,Name="clavicle_r"),Weight=0.3))
    Weight=1
   )
   > n2 = &RigUnit_ParentConstraint(
    AdvancedSettings=(InterpolationType=Average,RotationOrderForFilter=XZY)
    Child=(Type=Bone,Name="base_hood_back_side_r")
    Filter=(TranslationFilter=(bX=true,bY=true,bZ=true),RotationFilter=(bX=True,bY=True,bZ=True),ScaleFilter=(bX=True,bY=True,bZ=True))
    Parents=((Item=(Type=Bone,Name="spine_05"),Weight=0.75),(Item=(Type=Bone,Name="clavicle_r"),Weight=0.15))
    Weight=1
   )
   RigVMFunctionReturnNode(ExecutePin=@n2.ExecutePin)
  ]}
 CollapseNode_ContainedGraph {nodes [
   RigVMFunctionEntryNode()
   RigVMFunction_MathQuaternionSwingTwist = &RigVMFunction_MathQuaternionSwingTwist(
    TwistAxis=(X=1,Y=0,Z=0)
    Input=&RigUnit_GetRelativeTransformForItem(Child=(Type=Bone,Name="head"), Parent=(Type=Control,Name="ref_head_ctrl"), bChildInitial=False, bParentInitial=False).RelativeTransform.Rotation
   )
   RigVMFunction_MathQuaternionToEuler = &RigVMFunction_MathQuaternionToEuler(RotationOrder=ZYX, Value=@RigVMFunction_MathQuaternionSwingTwist.Swing)
   n1 = &RigVMFunction_MathDoubleRemap(SourceMaximum=60, SourceMinimum=20, TargetMaximum=1, TargetMinimum=0, Value=@RigVMFunction_MathQuaternionToEuler.Result.Y)
   n2 = &RigVMFunction_MathQuaternionToEuler(
    RotationOrder=ZYX
    Value=&RigVMFunction_MathQuaternionSwingTwist(
     TwistAxis=(X=1,Y=0,Z=0)
     Input=&RigUnit_GetRelativeTransformForItem(Child=(Type=Bone,Name="neck_02"), Parent=(Type=Control,Name="ref_neck_ctrl"), bChildInitial=False, bParentInitial=False).RelativeTransform.Rotation
    ).Swing
   )
   n3 = &RigVMFunction_MathDoubleRemap(SourceMaximum=40, SourceMinimum=0, TargetMaximum=1, TargetMinimum=0, Value=@n2.Result.Y)
   n4 = &RigVMFunction_MathDoubleRemap(SourceMaximum=-70, SourceMinimum=-40, TargetMaximum=1, TargetMinimum=0, Value=RigVMVariableNode() {variable=clavicle_rot_l}.Value.Y)
   Add_3 = RigVMAggregateNode(B=@n1.Result, C=@n3.Result, A=@n4.Result)
   > RigVMVariableNode(Value=@Add_3.Result) {variable=head_side_l}
   Add = RigVMAggregateNode(
    B=&RigVMFunction_MathDoubleRemap(SourceMaximum=-60, SourceMinimum=-20, TargetMaximum=1, TargetMinimum=0, Value=@RigVMFunction_MathQuaternionToEuler.Result.Y).Result
    C=&RigVMFunction_MathDoubleRemap(SourceMaximum=-40, SourceMinimum=0, TargetMaximum=1, TargetMinimum=0, Value=@n2.Result.Y).Result
    A=&RigVMFunction_MathDoubleRemap(SourceMaximum=-70, SourceMinimum=0, TargetMaximum=1, TargetMinimum=0, Value=RigVMVariableNode() {variable=clavicle_rot_r}.Value.Y).Result
   )
   > RigVMVariableNode(Value=@Add.Result) {variable=head_side_r}
   n5 = &RigVMFunction_MathDoubleRemap(SourceMaximum=-65, SourceMinimum=0, TargetMaximum=1, TargetMinimum=0, Value=@n2.Result.Z)
   RigVMFunction_AnimEasing = /Script/RigVM.RigVMFunction_AnimEasing(
    SourceMaximum=1
    SourceMinimum=0
    TargetMaximum=1
    TargetMinimum=0
    Type=CubicEaseIn
    Value=&RigVMFunction_MathDoubleRemap(SourceMaximum=-120, SourceMinimum=0, TargetMaximum=1, TargetMinimum=0, Value=@RigVMFunction_MathQuaternionToEuler.Result.Z).Result
   ) {is=struct RigVMUnitNode}
   > RigVMVariableNode(Value=&RigVMFunction_MathDoubleAdd(B=@n5.Result, A=@RigVMFunction_AnimEasing.Result).Result) {variable=head_back_c}
   > n6 = &RigUnit_SphericalPoseReader(
    ActiveRegionScaleFactors=(PositiveWidth=1,NegativeWidth=0.5,PositiveHeight=0.75,NegativeHeight=0.75)
    ActiveRegionSize=0.1
    Debug=(bDrawDebug=false,bDraw2D=False,bDrawLocalAxes=False,DebugScale=25,DebugSegments=20,DebugThickness=0.25)
    DriverAxis=(X=0,Y=1,Z=0)
    DriverItem=(Type=Bone,Name="head")
    FalloffRegionScaleFactors=(PositiveWidth=1,NegativeWidth=0.5,PositiveHeight=1,NegativeHeight=1)
    FalloffSize=0.2
    FlipHeightScaling=False
    FlipWidthScaling=False
    OptionalParentItem=(Type=Bone,Name="None")
    RotationOffset=(X=90,Y=0,Z=-30)
   )
   > RigVMVariableNode(Value=@n6.OutputParam) {variable=head_SPR_look_up_alpha}
   > n7 = &RigUnit_SphericalPoseReader(
    ActiveRegionScaleFactors=(PositiveWidth=1,NegativeWidth=1,PositiveHeight=1,NegativeHeight=1)
    ActiveRegionSize=0.025
    Debug=(bDrawDebug=false,bDraw2D=False,bDrawLocalAxes=False,DebugScale=25,DebugSegments=20,DebugThickness=0.25)
    DriverAxis=(X=-1,Y=0,Z=0)
    DriverItem=(Type=Bone,Name="upperarm_r")
    FalloffRegionScaleFactors=(PositiveWidth=1,NegativeWidth=1,PositiveHeight=1,NegativeHeight=1)
    FalloffSize=0.1
    FlipHeightScaling=False
    FlipWidthScaling=False
    OptionalParentItem=(Type=Bone,Name="None")
    RotationOffset=(X=-115,Y=0,Z=25)
   )
   > RigVMVariableNode(Value=@n7.OutputParam) {variable=upperarm_r_SPR_aim_up_two hands alpha}
   Multiply = &RigVMFunction_MathDoubleMul(A=RigVMVariableNode() {variable=upperarm_r_SPR_aim_up_two hands alpha}.Value, B=RigVMVariableNode() {variable=head_SPR_look_up_alpha}.Value)
   > RigVMVariableNode(Value=@Multiply.Result) {variable=look_up_and_left_two_hands_multiplied_alpha}
   > &RigUnit_ModifyTransforms(
    ItemToModify="((Item=(Type=Bone,Name=\"base_hood_side_l\"),Transform=(Rotation=(X=-0.1736,Y=0,Z=0,W=0.9848),Translation=(X=0,Y=0,Z=0),Scale3D=(X=1,Y=1,Z=1))),(Item=(Type=Bone,Name=\"base_hood_back_side_l\"),Transform=(Rotation=(X=-0.1736,Y=0,Z=0,W=0.9848),Translation=(X=0,Y ...(truncated)"
    WeightMinimum=0
    Weight=RigVMVariableNode() {variable=look_up_and_left_two_hands_multiplied_alpha}.Value
   )
   n8 = &RigVMFunction_AlphaInterp(InterpSpeedDecreasing=5, InterpSpeedIncreasing=20, Scale=1, bClampResult=False, Value=RigVMVariableNode() {variable=is Targeting}.Value)
   > &RigUnit_ModifyTransforms(
    ItemToModify="((Item=(Type=Bone,Name=\"upperarm_r_shoulderMover_0\"),Transform=(Rotation=(X=0.1044,Y=0.326,Z=-0.0711,W=-0.9369),Translation=(X=0.8396,Y=2.1153,Z=0.3123),Scale3D=(X=1,Y=1,Z=1))),(Item=(Type=Bone,Name=\"Dyn_hair_Front_L_02\"),Transform=(Rotation=(X=0,Y=0,Z=0,W ...(truncated)"
    WeightMinimum=0
    Weight=&RigVMFunction_MathDoubleMul(
     B=@n8.Result
     A=RigVMDispatchNode(BoneName=root, Default=0, Name=deformRig_r_upperarm_fr_alpha) {template="DISPATCH_RigDispatch_GetAnimAttribute(in Name,in BoneName,in Default,out Value,out Found)"}.Value
    ).Result
   )
   AlphaInterp = &RigVMFunction_AlphaInterp(InterpSpeedDecreasing=5, InterpSpeedIncreasing=20, Scale=1, bClampResult=False, Value=RigVMVariableNode() {variable=is Targeting}.Value)
   > &RigUnit_ModifyTransforms(
    ItemToModify="((Item=(Type=Bone,Name=\"upperarm_l_shoulderMover_0\"),Transform=(Rotation=(X=0.1437,Y=0.1637,Z=0.0394,W=-0.9752),Translation=(X=-1.8665,Y=-7.3857,Z=1.0385),Scale3D=(X=1,Y=1,Z=1))),(Item=(Type=Bone,Name=\"Dyn_hair_Front_L_02\"),Transform=(Rotation=(X=0,Y=0,Z=0 ...(truncated)"
    WeightMinimum=0
    Weight=&RigVMFunction_MathDoubleMul(
     B=@AlphaInterp.Result
     A=RigVMDispatchNode(BoneName=root, Default=0, Name=deformRig_l_upperarm_fr_alpha) {template="DISPATCH_RigDispatch_GetAnimAttribute(in Name,in BoneName,in Default,out Value,out Found)"}.Value
    ).Result
   )
   > n9 = &RigUnit_SphericalPoseReader(
    ActiveRegionScaleFactors=(PositiveWidth=3,NegativeWidth=3,PositiveHeight=1,NegativeHeight=2.5)
    ActiveRegionSize=0.1
    Debug=(bDrawDebug=false,bDraw2D=False,bDrawLocalAxes=False,DebugScale=25,DebugSegments=20,DebugThickness=0.25)
    DriverAxis=(X=-1,Y=0,Z=0)
    DriverItem=(Type=Bone,Name="upperarm_r")
    FalloffRegionScaleFactors=(PositiveWidth=1,NegativeWidth=1,PositiveHeight=1,NegativeHeight=1)
    FalloffSize=0.4
    FlipHeightScaling=False
    FlipWidthScaling=False
    OptionalParentItem=(Type=Bone,Name="None")
    RotationOffset=(X=-60,Y=20,Z=20)
   )
   > n10 = &RigUnit_SphericalPoseReader(
    ActiveRegionScaleFactors=(PositiveWidth=3,NegativeWidth=3,PositiveHeight=3,NegativeHeight=3.8)
    ActiveRegionSize=0.1
    Debug=(bDrawDebug=false,bDraw2D=False,bDrawLocalAxes=False,DebugScale=25,DebugSegments=20,DebugThickness=0.25)
    DriverAxis=(X=1,Y=0,Z=0)
    DriverItem=(Type=Bone,Name="upperarm_l")
    FalloffRegionScaleFactors=(PositiveWidth=1,NegativeWidth=1,PositiveHeight=2,NegativeHeight=2)
    FalloffSize=0.4
    FlipHeightScaling=False
    FlipWidthScaling=False
    OptionalParentItem=(Type=Bone,Name="None")
    RotationOffset=(X=-0.4617,Y=-21.1077,Z=-76.4455)
   )
   > n11 = &RigUnit_SphericalPoseReader(
    ActiveRegionScaleFactors=(PositiveWidth=2,NegativeWidth=2,PositiveHeight=2,NegativeHeight=2)
    ActiveRegionSize=0.1
    Debug=(bDrawDebug=false,bDraw2D=False,bDrawLocalAxes=False,DebugScale=25,DebugSegments=20,DebugThickness=0.25)
    DriverAxis=(X=1,Y=0,Z=0)
    DriverItem=(Type=Bone,Name="upperarm_l")
    FalloffRegionScaleFactors=(PositiveWidth=1,NegativeWidth=1,PositiveHeight=1,NegativeHeight=1)
    FalloffSize=0.3
    FlipHeightScaling=False
    FlipWidthScaling=False
    OptionalParentItem=(Type=Bone,Name="None")
    RotationOffset=(X=84.7921,Y=-21.1077,Z=-76.4355)
   )
   > &RigUnit_ModifyTransforms(
    ItemToModify="((Item=(Type=Bone,Name=\"upperarm_l_shoulderMover_0\"),Transform=(Rotation=(X=-0.0008,Y=-0.1512,Z=-0.0071,W=-0.9885),Translation=(X=-1.5617,Y=-3.884,Z=2.7605),Scale3D=(X=1,Y=1,Z=1))),(Item=(Type=Bone,Name=\"extra_base_arm_r\"),Transform=(Rotation=(X=0.0572,Y=0 ...(truncated)"
    WeightMinimum=0
    Weight=&RigVMFunction_MathDoubleMul(A=@n10.OutputParam, B=@n9.OutputParam).Result
   )
   n12 = &RigVMFunction_AlphaInterp(
    InterpSpeedDecreasing=10
    InterpSpeedIncreasing=10
    Scale=1
    bClampResult=true
    Value=&RigVMFunction_MathDoubleMul(A=@n11.OutputParam, B=@n9.OutputParam).Result
   )
   > &RigUnit_ModifyTransforms(
    ItemToModify=((Item=(Type=Bone,Name="extra_base_arm_r"),Transform=(Rotation=(X=0.0816,Y=-0.0605,Z=-0.005,W=0.9948),Translation=(X=0.9228,Y=-0.1305,Z=0.7902),Scale3D=(X=1,Y=1,Z=1))))
    WeightMinimum=0
    Weight=@n12.Result
   )
   > RigVMFunctionReturnNode()
   Maximum = &RigVMFunction_MathDoubleMax(B=@n5.Result, A=@RigVMFunction_AnimEasing.Result)
   Maximum_2 = RigVMAggregateNode(B=@n1.Result, C=@n3.Result, A=@n4.Result)
  ]}
 CollapseNode_ContainedGraph {nodes [
   RigVMFunctionEntryNode()
   > &RigUnit_ParentConstraint(
    AdvancedSettings=(InterpolationType=Average,RotationOrderForFilter=XZY)
    Child=(Type=Bone,Name="base_shirt_f_med")
    Parents=((Item=(Type=Bone,Name="thigh_l"),Weight=0.5),(Item=(Type=Bone,Name="thigh_r"),Weight=0.5))
    Weight=1
   )
   > &RigUnit_ParentConstraint(
    AdvancedSettings=(InterpolationType=Average,RotationOrderForFilter=XZY)
    Child=(Type=Bone,Name="base_shirt_b_med")
    Parents=((Item=(Type=Bone,Name="thigh_l"),Weight=0.5),(Item=(Type=Bone,Name="thigh_r"),Weight=0.5))
    Weight=1
   )
   n1 = RigVMDispatchNode(BoneName=root, Default=1, Name=deformRig_r_thigh_up_alpha) {template="DISPATCH_RigDispatch_GetAnimAttribute(in Name,in BoneName,in Default,out Value,out Found)"}
   > &RigUnit_ModifyTransforms(
    ItemToModify=((Item=(Type=Bone,Name="base_shirt_f_med"),Transform=(Rotation=(X=0.1265,Y=0.1512,Z=-0.2864,W=0.9376),Translation=(X=-1.8229,Y=7.1656,Z=5.2312),Scale3D=(X=1,Y=1,Z=1))))
    WeightMinimum=0
    Weight=@n1.Value
   )
   n2 = RigVMDispatchNode(BoneName=root, Default=1, Name=deformRig_l_thigh_up_alpha) {template="DISPATCH_RigDispatch_GetAnimAttribute(in Name,in BoneName,in Default,out Value,out Found)"}
   > &RigUnit_ModifyTransforms(
    ItemToModify=((Item=(Type=Bone,Name="base_shirt_f_med"),Transform=(Rotation=(X=0.1265,Y=-0.1512,Z=0.2864,W=0.9376),Translation=(X=1.9036,Y=8.472,Z=5.7084),Scale3D=(X=1,Y=1,Z=1))))
    WeightMinimum=0
    Weight=@n2.Value
   )
   Multiply = &RigVMFunction_MathDoubleMul(A=@n1.Value, B=@n2.Value)
   > &RigUnit_ModifyTransforms(
    ItemToModify=((Item=(Type=Bone,Name="base_shirt_f_med"),Transform=(Rotation=(X=-0.1063,Y=0.0658,Z=0.015,W=0.992),Translation=(X=-6.5737,Y=-16.5253,Z=-2.0871),Scale3D=(X=1,Y=1,Z=1))))
    WeightMinimum=0
    Weight=@Multiply.Result
   )
   n3 = &RigVMFunction_MathDoubleRemap(SourceMaximum=1, SourceMinimum=0, TargetMaximum=1, TargetMinimum=0, Value=@n1.Value)
   > &RigUnit_ModifyTransforms(
    ItemToModify=((Item=(Type=Bone,Name="base_shirt_b_med"),Transform=(Rotation=(X=-0.021,Y=-0.0916,Z=-0.1917,W=0.9769),Translation=(X=0.8645,Y=-9.1995,Z=4.8763),Scale3D=(X=1,Y=1,Z=1))))
    WeightMinimum=0
    Weight=@n3.Result
   )
   Remap = &RigVMFunction_MathDoubleRemap(SourceMaximum=1, SourceMinimum=0, TargetMaximum=1, TargetMinimum=0, Value=@n2.Value)
   > &RigUnit_ModifyTransforms(
    ItemToModify=((Item=(Type=Bone,Name="base_shirt_b_med"),Transform=(Rotation=(X=-0.0151,Y=0.0928,Z=0.1912,W=0.977),Translation=(X=0.8645,Y=-9.1995,Z=4.8763),Scale3D=(X=1,Y=1,Z=1))))
    WeightMinimum=0
    Weight=@Remap.Result
   )
   > &RigUnit_ModifyTransforms(
    ItemToModify=((Item=(Type=Bone,Name="base_shirt_b_med"),Transform=(Rotation=(X=0,Y=0,Z=0,W=1),Translation=(X=2.6587,Y=16.2481,Z=0),Scale3D=(X=1,Y=1,Z=1))))
    WeightMinimum=0
    Weight=&RigVMFunction_MathDoubleMul(B=@Remap.Result, A=@n3.Result).Result
   )
   > n4 = &RigUnit_SphericalPoseReader(
    ActiveRegionScaleFactors=(PositiveWidth=3.5,NegativeWidth=0.5,PositiveHeight=2,NegativeHeight=3)
    ActiveRegionSize=0.1
    Debug=(bDrawDebug=false,bDraw2D=False,bDrawLocalAxes=False,DebugScale=25,DebugSegments=20,DebugThickness=0.25)
    DriverAxis=(X=-1,Y=0,Z=0)
    DriverItem=(Type=Bone,Name="thigh_l")
    FalloffRegionScaleFactors=(PositiveWidth=1,NegativeWidth=0.4,PositiveHeight=1,NegativeHeight=1)
    FalloffSize=0.45
    FlipHeightScaling=False
    FlipWidthScaling=False
    OptionalParentItem=(Type=Bone,Name="None")
    RotationOffset=(X=-90,Y=0,Z=-70)
   )
   > n5 = &RigUnit_SphericalPoseReader(
    ActiveRegionScaleFactors=(PositiveWidth=3.5,NegativeWidth=0.5,PositiveHeight=2,NegativeHeight=3)
    ActiveRegionSize=0.1
    Debug=(bDrawDebug=false,bDraw2D=False,bDrawLocalAxes=False,DebugScale=25,DebugSegments=20,DebugThickness=0.25)
    DriverAxis=(X=1,Y=0,Z=0)
    DriverItem=(Type=Bone,Name="thigh_r")
    FalloffRegionScaleFactors=(PositiveWidth=1,NegativeWidth=0.4,PositiveHeight=1,NegativeHeight=1)
    FalloffSize=0.45
    FlipHeightScaling=False
    FlipWidthScaling=False
    OptionalParentItem=(Type=Bone,Name="None")
    RotationOffset=(X=90,Y=180,Z=-70)
   )
   > &RigUnit_ModifyTransforms(
    ItemToModify="((Item=(Type=Bone,Name=\"base_shirt_b_med\"),Transform=(Rotation=(X=0,Y=0,Z=0.342,W=0.9397),Translation=(X=0,Y=0,Z=0),Scale3D=(X=1,Y=1,Z=1))),(Item=(Type=Bone,Name=\"base_shirt_b_right\"),Transform=(Rotation=(X=0,Y=0,Z=0,W=1),Translation=(X=0,Y=-2.8372,Z=0),Sc ...(truncated)"
    WeightMinimum=0
    Weight=@n5.OutputParam
   )
   > &RigUnit_ModifyTransforms(
    ItemToModify="((Item=(Type=Bone,Name=\"base_shirt_b_med\"),Transform=(Rotation=(X=0,Y=0,Z=-0.342,W=0.9397),Translation=(X=0,Y=0,Z=0),Scale3D=(X=1,Y=1,Z=1))),(Item=(Type=Bone,Name=\"base_shirt_b_left\"),Transform=(Rotation=(X=0,Y=0,Z=0,W=-1),Translation=(X=0,Y=-4.1795,Z=0),S ...(truncated)"
    WeightMinimum=0
    Weight=@n4.OutputParam
   )
   > &RigUnit_ModifyTransforms(
    ItemToModify=((Item=(Type=Bone,Name="base_shirt_b_med"),Transform=(Rotation=(X=0,Y=0,Z=0,W=1),Translation=(X=0.8505,Y=1.5276,Z=0),Scale3D=(X=1,Y=1,Z=1))))
    WeightMinimum=0
    Weight=&RigVMFunction_MathDoubleMul(B=@n4.OutputParam, A=@n5.OutputParam).Result
   )
   > RigVMFunctionReturnNode()
  ]}
 CollapseNode_ContainedGraph {nodes [
   RigVMFunctionEntryNode()
   RigVMFunction_MathQuaternionToEuler = &RigVMFunction_MathQuaternionToEuler(
    RotationOrder=ZYX
    Value=&RigVMFunction_MathTransformMakeRelative(
     Global=&RigUnit_GetTransform(Item=(Type=Bone,Name="foot_l"), Space=LocalSpace).Transform
     Parent=&RigUnit_GetTransform(Item=(Type=Bone,Name="foot_l"), Space=LocalSpace, bInitial=true).Transform
    ).Local.Rotation
   )
   Remap = &RigVMFunction_MathDoubleRemap(SourceMaximum=-50, SourceMinimum=-14, TargetMaximum=1, TargetMinimum=0, Value=@RigVMFunction_MathQuaternionToEuler.Result.Z)
   > &RigUnit_ModifyTransforms(
    ItemToModify=((Item=(Type=Bone,Name="base_footTongue_L"),Transform=(Rotation=(X=0.4036,Y=0,Z=0,W=-0.9149),Translation=(X=0,Y=1.5373,Z=-1.6838),Scale3D=(X=1,Y=1,Z=1))))
    WeightMinimum=0
    Weight=@Remap.Result
   )
   > &RigUnit_ModifyTransforms(
    ItemToModify=((Item=(Type=Bone,Name="base_footTongue_L"),Transform=(Rotation=(X=-0.2053,Y=0,Z=0,W=-0.9787),Translation=(X=0,Y=-1.8039,Z=1.7982),Scale3D=(X=1,Y=1,Z=1))))
    WeightMinimum=0
    Weight=&RigVMFunction_MathDoubleRemap(SourceMaximum=80, SourceMinimum=30, TargetMaximum=1, TargetMinimum=0, Value=@RigVMFunction_MathQuaternionToEuler.Result.Z).Result
   )
   n1 = &RigVMFunction_MathQuaternionToEuler(
    RotationOrder=ZYX
    Value=&RigVMFunction_MathTransformMakeRelative(
     Global=&RigUnit_GetTransform(Item=(Type=Bone,Name="foot_r"), Space=LocalSpace).Transform
     Parent=&RigUnit_GetTransform(Item=(Type=Bone,Name="foot_r"), Space=LocalSpace, bInitial=true).Transform
    ).Local.Rotation
   )
   > &RigUnit_ModifyTransforms(
    ItemToModify=((Item=(Type=Bone,Name="base_footTongue_R"),Transform=(Rotation=(X=0.4036,Y=0,Z=0,W=-0.9149),Translation=(X=0,Y=1.5373,Z=-1.6838),Scale3D=(X=1,Y=1,Z=1))))
    WeightMinimum=0
    Weight=&RigVMFunction_MathDoubleRemap(SourceMaximum=-50, SourceMinimum=-14, TargetMaximum=1, TargetMinimum=0, Value=@n1.Result.Z).Result
   )
   > &RigUnit_ModifyTransforms(
    ItemToModify=((Item=(Type=Bone,Name="base_footTongue_R"),Transform=(Rotation=(X=-0.2053,Y=0,Z=0,W=-0.9787),Translation=(X=0,Y=-1.8039,Z=1.7982),Scale3D=(X=1,Y=1,Z=1))))
    WeightMinimum=0
    Weight=&RigVMFunction_MathDoubleRemap(SourceMaximum=80, SourceMinimum=30, TargetMaximum=1, TargetMinimum=0, Value=@n1.Result.Z).Result
   )
   > RigVMFunctionReturnNode()
  ]}
 CollapseNode_ContainedGraph {nodes [
   RigVMFunctionEntryNode()
   > n1 = &RigUnit_SphericalPoseReader(
    ActiveRegionScaleFactors=(PositiveWidth=3.5,NegativeWidth=0.5,PositiveHeight=2,NegativeHeight=3)
    ActiveRegionSize=0.1
    Debug=(bDrawDebug=false,bDraw2D=False,bDrawLocalAxes=False,DebugScale=25,DebugSegments=20,DebugThickness=0.25)
    DriverAxis=(X=-1,Y=0,Z=0)
    DriverItem=(Type=Bone,Name="thigh_l")
    FalloffRegionScaleFactors=(PositiveWidth=1,NegativeWidth=1,PositiveHeight=1,NegativeHeight=1)
    FalloffSize=0.45
    FlipHeightScaling=False
    FlipWidthScaling=False
    OptionalParentItem=(Type=Bone,Name="spine_02")
    RotationOffset=(X=90,Y=0,Z=-7)
   )
   > &RigUnit_ModifyTransforms(
    ItemToModify=((Item=(Type=Bone,Name="base_jacket_l"),Transform=(Rotation=(X=0,Y=0,Z=-0.5736,W=0.8192),Translation=(X=-2.2849,Y=2.9244,Z=0),Scale3D=(X=1,Y=1,Z=1))))
    WeightMinimum=0
    Weight=@n1.OutputParam
   )
   > n2 = &RigUnit_SphericalPoseReader(
    ActiveRegionScaleFactors=(PositiveWidth=3,NegativeWidth=0.5,PositiveHeight=3,NegativeHeight=2)
    ActiveRegionSize=0.1
    Debug=(bDrawDebug=false,bDraw2D=False,bDrawLocalAxes=False,DebugScale=25,DebugSegments=20,DebugThickness=0.25)
    DriverAxis=(X=1,Y=0,Z=0)
    DriverItem=(Type=Bone,Name="thigh_r")
    FalloffRegionScaleFactors=(PositiveWidth=1,NegativeWidth=1.2,PositiveHeight=1,NegativeHeight=1)
    FalloffSize=0.45
    FlipHeightScaling=False
    FlipWidthScaling=False
    OptionalParentItem=(Type=Bone,Name="spine_02")
    RotationOffset=(X=-90,Y=180,Z=-7)
   )
   > &RigUnit_ModifyTransforms(
    ItemToModify=((Item=(Type=Bone,Name="base_jacket_r"),Transform=(Rotation=(X=0,Y=0,Z=-0.5736,W=0.8192),Translation=(X=-2.2849,Y=2.9244,Z=0),Scale3D=(X=1,Y=1,Z=1))))
    WeightMinimum=0
    Weight=@n2.OutputParam
   )
   Remap = &RigVMFunction_MathDoubleRemap(SourceMaximum=1, SourceMinimum=0, TargetMaximum=0, TargetMinimum=1, Value=RigVMVariableNode() {variable="is Crouching "}.Value)
   > &RigUnit_ModifyTransforms(
    ItemToModify=((Item=(Type=Bone,Name="base_jacket_r"),Transform=(Rotation=(X=0,Y=0,Z=-0.0698,W=0.9976),Translation=(X=0,Y=0,Z=0),Scale3D=(X=1,Y=1,Z=1))))
    WeightMinimum=0
    Weight=@Remap.Result
   )
   > n3 = &RigUnit_SphericalPoseReader(
    ActiveRegionScaleFactors=(PositiveWidth=3.5,NegativeWidth=0.5,PositiveHeight=3,NegativeHeight=2)
    ActiveRegionSize=0.1
    Debug=(bDrawDebug=false,bDraw2D=False,bDrawLocalAxes=False,DebugScale=25,DebugSegments=20,DebugThickness=0.25)
    DriverAxis=(X=-1,Y=0,Z=0)
    DriverItem=(Type=Bone,Name="thigh_l")
    FalloffRegionScaleFactors=(PositiveWidth=1,NegativeWidth=1,PositiveHeight=1,NegativeHeight=1)
    FalloffSize=0.45
    FlipHeightScaling=False
    FlipWidthScaling=False
    OptionalParentItem=(Type=Bone,Name="spine_02")
    RotationOffset=(X=180,Y=7,Z=0)
   )
   > &RigUnit_ModifyTransforms(
    ItemToModify=((Item=(Type=Bone,Name="base_jacket_l"),Transform=(Rotation=(X=-0.2291,Y=0.13,Z=-0.2932,W=0.9191),Translation=(X=-3.8,Y=3.2,Z=2),Scale3D=(X=1,Y=1,Z=1))))
    WeightMinimum=0
    Weight=@n3.OutputParam
   )
   > n4 = &RigUnit_SphericalPoseReader(
    ActiveRegionScaleFactors=(PositiveWidth=3.5,NegativeWidth=0.5,PositiveHeight=2,NegativeHeight=3)
    ActiveRegionSize=0.1
    Debug=(bDrawDebug=false,bDraw2D=False,bDrawLocalAxes=False,DebugScale=25,DebugSegments=20,DebugThickness=0.25)
    DriverAxis=(X=1,Y=0,Z=0)
    DriverItem=(Type=Bone,Name="thigh_r")
    FalloffRegionScaleFactors=(PositiveWidth=1,NegativeWidth=1,PositiveHeight=1,NegativeHeight=1)
    FalloffSize=0.45
    FlipHeightScaling=False
    FlipWidthScaling=False
    OptionalParentItem=(Type=Bone,Name="spine_02")
    RotationOffset=(X=0,Y=-7,Z=180)
   )
   > n5 = &RigUnit_ModifyTransforms(
    ItemToModify=((Item=(Type=Bone,Name="base_jacket_r"),Transform=(Rotation=(X=0.2611,Y=-0.0341,Z=-0.3189,W=0.9105),Translation=(X=-3.8,Y=3.2,Z=2),Scale3D=(X=1,Y=1,Z=1))))
    WeightMinimum=0
    Weight=@n4.OutputParam
   )
   RigVMFunction_ControlFlowBranch = &RigVMFunction_ControlFlowBranch(ExecuteContext=@n5.ExecutePin, Condition=RigVMVariableNode() {variable="is Crouch Moving "}.Value)
   RigVMFunctionReturnNode(ExecuteContext=@RigVMFunction_ControlFlowBranch.Completed)
   n6 = &RigUnit_SphericalPoseReader(
    ActiveRegionScaleFactors=(PositiveWidth=3,NegativeWidth=0.5,PositiveHeight=3,NegativeHeight=3)
    ActiveRegionSize=0.1
    Debug=(bDrawDebug=false,bDraw2D=False,bDrawLocalAxes=False,DebugScale=25,DebugSegments=20,DebugThickness=0.25)
    DriverAxis=(X=-1,Y=0,Z=0)
    DriverItem=(Type=Bone,Name="thigh_l")
    FalloffRegionScaleFactors=(PositiveWidth=1,NegativeWidth=0.7,PositiveHeight=1,NegativeHeight=1)
    FalloffSize=0.3
    FlipHeightScaling=False
    FlipWidthScaling=False
    OptionalParentItem=(Type=Bone,Name="spine_02")
    RotationOffset=(X=90,Y=0,Z=-45)
    ExecutePin=@RigVMFunction_ControlFlowBranch.True
   )
   > &RigUnit_ModifyTransforms(
    ItemToModify=((Item=(Type=Bone,Name="base_jacket_l"),Transform=(Rotation=(X=0.2676,Y=-0.0877,Z=-0.2728,W=0.9199),Translation=(X=-4.9313,Y=3.7885,Z=-2.9677),Scale3D=(X=1,Y=1,Z=1))))
    WeightMinimum=0
    Weight=@n6.OutputParam
   )
   > n7 = &RigUnit_SphericalPoseReader(
    ActiveRegionScaleFactors=(PositiveWidth=3,NegativeWidth=0.5,PositiveHeight=3,NegativeHeight=3)
    ActiveRegionSize=0.1
    Debug=(bDrawDebug=False,bDraw2D=False,bDrawLocalAxes=False,DebugScale=25,DebugSegments=20,DebugThickness=0.25)
    DriverAxis=(X=1,Y=0,Z=0)
    DriverItem=(Type=Bone,Name="thigh_r")
    FalloffRegionScaleFactors=(PositiveWidth=1,NegativeWidth=0.7,PositiveHeight=1,NegativeHeight=1)
    FalloffSize=0.3
    FlipHeightScaling=False
    FlipWidthScaling=False
    OptionalParentItem=(Type=Bone,Name="spine_02")
    RotationOffset=(X=-90,Y=180,Z=-55)
   )
   AlphaInterp = &RigVMFunction_AlphaInterp(InterpSpeedDecreasing=10, InterpSpeedIncreasing=20, Scale=1, bClampResult=False, Value=@n7.OutputParam)
   > &RigUnit_ModifyTransforms(
    ItemToModify=((Item=(Type=Bone,Name="base_jacket_r"),Transform=(Rotation=(X=0.3304,Y=0.323,Z=-0.2317,W=0.8561),Translation=(X=-1.7817,Y=-14.7321,Z=9.3808),Scale3D=(X=1,Y=1,Z=1))))
    WeightMinimum=0
    Weight=@AlphaInterp.Result
   )
   n8 = &RigUnit_SphericalPoseReader(
    ActiveRegionScaleFactors=(PositiveWidth=3,NegativeWidth=0.5,PositiveHeight=3,NegativeHeight=3)
    ActiveRegionSize=0.1
    Debug=(bDrawDebug=false,bDraw2D=False,bDrawLocalAxes=False,DebugScale=25,DebugSegments=20,DebugThickness=0.25)
    DriverAxis=(X=1,Y=0,Z=0)
    DriverItem=(Type=Bone,Name="thigh_r")
    ExecutePin=()
    FalloffRegionScaleFactors=(PositiveWidth=1,NegativeWidth=0.7,PositiveHeight=1,NegativeHeight=1)
    FalloffSize=0.3
    FlipHeightScaling=False
    FlipWidthScaling=False
    OptionalParentItem=(Type=Bone,Name="spine_02")
    RotationOffset=(X=-90,Y=180,Z=-45)
   )
   > &RigUnit_ModifyTransforms(
    ItemToModify=((Item=(Type=Bone,Name="base_jacket_r"),Transform=(Rotation=(X=-0.1993,Y=0.2198,Z=-0.1926,W=0.9353),Translation=(X=-3.2562,Y=0.0672,Z=5.0607),Scale3D=(X=1,Y=1,Z=1))))
    WeightMinimum=0
    Weight=@n8.OutputParam
   )
   &RigUnit_ModifyTransforms(
    ExecutePin=()
    ItemToModify=((Item=(Type=Bone,Name="base_jacket_l"),Transform=(Rotation=(X=-0.2611,Y=-0.2932,Z=-0.13,W=0.9105),Translation=(X=-3.2,Y=2,Z=3.8),Scale3D=(X=1,Y=1,Z=1))))
    Weight=1
    WeightMinimum=0
   )
   &RigUnit_ModifyTransforms(
    ExecutePin=()
    ItemToModify=((Item=(Type=Bone,Name="base_jacket_r"),Transform=(Rotation=(X=-0.2611,Y=-0.2932,Z=-0.13,W=0.9105),Translation=(X=3.2,Y=2,Z=-3.8),Scale3D=(X=1,Y=1,Z=1))))
    Weight=1
    WeightMinimum=0
   )
   &RigUnit_ModifyTransforms(
    ExecutePin=()
    ItemToModify=((Item=(Type=Bone,Name="base_jacket_l"),Transform=(Rotation=(X=0.1748,Y=-0.3781,Z=-0.0766,W=0.9059),Translation=(X=-3.3291,Y=-2.0977,Z=5.2494),Scale3D=(X=1,Y=1,Z=1))))
    Weight=1
    WeightMinimum=0
   )
   &RigUnit_ModifyTransforms(
    ExecutePin=()
    ItemToModify=((Item=(Type=Bone,Name="base_jacket_r"),Transform=(Rotation=(X=0.1574,Y=-0.3281,Z=-0.3405,W=0.867),Translation=(X=4.6486,Y=3.3651,Z=-3.1005),Scale3D=(X=1,Y=1,Z=1))))
    Weight=1
    WeightMinimum=0
   )
   &RigUnit_ModifyTransforms(
    ExecutePin=()
    ItemToModify=((Item=(Type=Bone,Name="base_jacket_r"),Transform=(Rotation=(X=-0.1748,Y=0.3781,Z=0.0766,W=-0.9059),Translation=(X=3.3291,Y=-2.0977,Z=-5.2494),Scale3D=(X=1,Y=1,Z=1))))
    Weight=1
    WeightMinimum=0
   )
   &RigUnit_ModifyTransforms(
    ExecutePin=()
    ItemToModify=((Item=(Type=Bone,Name="base_jacket_l"),Transform=(Rotation=(X=0,Y=-0.5736,Z=0,W=0.8192),Translation=(X=-2.9244,Y=0,Z=2.2849),Scale3D=(X=1,Y=1,Z=1))))
    Weight=1
    WeightMinimum=0
   )
   &RigUnit_ModifyTransforms(
    ExecutePin=()
    ItemToModify=((Item=(Type=Bone,Name="base_jacket_r"),Transform=(Rotation=(X=0,Y=-0.5736,Z=0,W=0.8192),Translation=(X=2.9244,Y=0,Z=-2.2849),Scale3D=(X=1,Y=1,Z=1))))
    Weight=1
    WeightMinimum=0
   )
  ]}
 Maximum_1_ContainedGraph {nodes [
   n1 = RigVMFunctionEntryNode()
   Maximum = &RigVMFunction_MathDoubleMax(A=@n1.A, B=@n1.B)
   RigVMFunctionReturnNode(Result=&RigVMFunction_MathDoubleMax(B=@n1.D, A=&RigVMFunction_MathDoubleMax(B=@n1.C, A=@Maximum.Result).Result).Result)
  ]}
 Maximum_2_ContainedGraph {nodes [
   n1 = RigVMFunctionEntryNode()
   RigVMFunctionReturnNode(Result=&RigVMFunction_MathDoubleMax(B=@n1.C, A=&RigVMFunction_MathDoubleMax(A=@n1.A, B=@n1.B).Result).Result)
  ]}
 RigVMFunctionLibrary {links [], nodes []}
 RigVMFunction_Sequence_1_ContainedGraph {nodes [
   n1 = RigVMFunctionEntryNode()
   RigVMFunction_Sequence = &RigVMFunction_Sequence(ExecuteContext=@n1.ExecuteContext)
   n2 = &RigVMFunction_Sequence(ExecuteContext=@RigVMFunction_Sequence.B)
   RigVMFunctionReturnNode(A=@RigVMFunction_Sequence.A, B=@n2.A, C=@n2.B)
  ]}
 RigVMModel {nodes [
   n1 = &RigUnit_PrepareForExecution()
   Human Arms Controls = RigVMCollapseNode(ExecuteContext=@n1.ExecutePin)
   > Robot Arms Controls = RigVMCollapseNode()
   > Control Stack at Item = RigVMFunctionReferenceNode(
    Add Bottom Null=False
    Add Null=true
    Add Secondary Control=False
    Control Suffix=_ctrl
    Item=(Name="head",Type=Bone)
    Name=ref_head
    Null Suffix=_null
    Parent=(Name="neck_02",Type=Bone)
    Secondary Settings="(Shape=(bVisible=True,Name=\"Default\",Color=(R=1,G=0,B=0,A=1),Transform=(Rotation=(X=0,Y=0,Z=0,W=1),Translation=(X=0,Y=0,Z=0),Scale3D=(X=1,Y=1,Z=1))),Proxy=(bIsProxy=False,ShapeVisibility=BasedOnSelection),DisplayName=\"None\",InitialSpace=LocalSpace,bUsePref ...(truncated)"
    Settings="(Shape=(bVisible=True,Name=\"Pyramid_Thin\",Color=(R=1,G=0,B=0,A=1),Transform=(Rotation=(X=0,Y=0,Z=0,W=1),Translation=(X=0,Y=0,Z=0),Scale3D=(X=1,Y=1,Z=1))),Proxy=(bIsProxy=False,ShapeVisibility=BasedOnSelection),DisplayName=\"None\",InitialSpace=LocalSpace,bUs ...(truncated)"
   )
   > Control Stack at Item_1 = RigVMFunctionReferenceNode(
    Add Bottom Null=False
    Add Null=true
    Add Secondary Control=False
    Control Suffix=_ctrl
    Item=(Name="neck_02",Type=Bone)
    Name=ref_neck
    Null Suffix=_null
    Parent=(Name="spine_05",Type=Bone)
    Secondary Settings="(Shape=(bVisible=True,Name=\"Default\",Color=(R=1,G=0,B=0,A=1),Transform=(Rotation=(X=0,Y=0,Z=0,W=1),Translation=(X=0,Y=0,Z=0),Scale3D=(X=1,Y=1,Z=1))),Proxy=(bIsProxy=False,ShapeVisibility=BasedOnSelection),DisplayName=\"None\",InitialSpace=LocalSpace,bUsePref ...(truncated)"
    Settings="(Shape=(bVisible=True,Name=\"Pyramid_Thin\",Color=(R=1,G=0,B=0,A=1),Transform=(Rotation=(X=0,Y=0,Z=0,W=1),Translation=(X=0,Y=0,Z=0),Scale3D=(X=1,Y=1,Z=1))),Proxy=(bIsProxy=False,ShapeVisibility=BasedOnSelection),DisplayName=\"None\",InitialSpace=LocalSpace,bUs ...(truncated)"
   )
   RigUnit_BeginExecution = &RigUnit_BeginExecution()
   FN_DeformRig_FNC = RigVMFunctionReferenceNode(
    DeformRig_ Curve="(deformRig_l_upperarm_up_alpha_72_#1=True,deformRig_r_upperarm_up_alpha_73_#2=True,deformRig_l_upperarm_fr_alpha_74_#3=True,deformRig_r_upperarm_fr_alpha_75_CB51B80C4 ...(truncated)"
    DeformRig_Bool="(upperarm_1_#4=True,elbow_3_#5=True,wrist_7_#6=True,groin_8_#7=True,glute_17_#8=True,knee_9_40C408324CA261 ...(truncated)"
    DeformRig_Debug=(upperarm_debug_18_#4=False,wrist_debug_20_#6=False,lat_debug_26_#9=False,groinGlute_debug_27_#7=False)
    DeformRig_Value="(upperarm_up_18_#10=(Rotation=(X=0.0,Y=-0.0,Z=0.0,W=1.0),Translation=(X=0.0,Y=0.0,Z=8.0),Scale3D=(X=0.70,Y=1.0,Z=1.0)),upperarm_fwd_fr_19_#11=(Rotation=(X=0.0,Y=-0.0,Z=0.0,W=1.0),Translation=(X=0.0, ...(truncated)"
    ExecuteContext=@RigUnit_BeginExecution.ExecutePin
   )
   > RigVMFunction_Sequence = RigVMAggregateNode()
   Robot Arms Initial Setup = RigVMCollapseNode(ExecutePin=@RigVMFunction_Sequence.A)
   n2 = &RigUnit_GetRelativeTransformForItem(Child=(Type=Bone,Name="clavicle_l"), Parent=(Type=Control,Name="ref_clavicle_l_ctrl"), bChildInitial=False, bParentInitial=False)
   RigVMVariableNode(
    Value=&RigVMFunction_MathQuaternionToEuler(RotationOrder=ZYX, Value=@n2.RelativeTransform.Rotation).Result
    ExecuteContext=@"Robot Arms Initial Setup".ExecutePin
   ) {variable=clavicle_rot_l}
   n3 = &RigUnit_GetRelativeTransformForItem(Child=(Type=Bone,Name="clavicle_r"), Parent=(Type=Control,Name="ref_clavicle_r_ctrl"), bChildInitial=False, bParentInitial=False)
   > RigVMVariableNode(Value=&RigVMFunction_MathQuaternionToEuler(RotationOrder=ZYX, Value=@n3.RelativeTransform.Rotation).Result) {variable=clavicle_rot_r}
   Human Arms Setup = RigVMCollapseNode(ExecutePin=@RigVMFunction_Sequence.B)
   Skirt Setup = RigVMCollapseNode(ExecuteContext=@"Human Arms Setup".ExecutePin)
   > Ankles Setup = RigVMCollapseNode()
   > Jacket Setup = RigVMCollapseNode()
   RigVMFunction_ControlFlowBranch = &RigVMFunction_ControlFlowBranch(ExecuteContext=@"Jacket Setup".ExecuteContext, Condition=RigVMVariableNode() {variable=ikRobotArms}.Value)
   Robot Arms IK Setup = RigVMCollapseNode(ExecutePin=@RigVMFunction_ControlFlowBranch.True)
   Fingers Setup = RigVMCollapseNode(ExecutePin=@RigVMFunction_Sequence.C)
   Neck Collar Constraints = RigVMCollapseNode(ExecutePin=@"Fingers Setup".ExecutePin)
   Neck Collar Get Variables = RigVMCollapseNode(ExecuteContext=@"Neck Collar Constraints".ExecutePin)
   > &RigUnit_ModifyTransforms(
    ItemToModify="((Item=(Type=Bone,Name=\"base_hood_side_l\"),Transform=(Rotation=(X=0,Y=0,Z=0.3827,W=0.9239),Translation=(X=0,Y=0,Z=0),Scale3D=(X=1,Y=1,Z=1))),(Item=(Type=Bone,Name=\"base_hood_front_l\"),Transform=(Rotation=(X=0,Y=0,Z=0.3827,W=0.9239),Translation=(X=0,Y=0,Z=0 ...(truncated)"
    WeightMinimum=0
    Weight=RigVMVariableNode() {variable=head_side_l}.Value
   )
   > &RigUnit_ModifyTransforms(
    ItemToModify="((Item=(Type=Bone,Name=\"base_hood_side_r\"),Transform=(Rotation=(X=0,Y=0,Z=-0.3827,W=0.9239),Translation=(X=0,Y=0,Z=0),Scale3D=(X=1,Y=1,Z=1))),(Item=(Type=Bone,Name=\"base_hood_front_r\"),Transform=(Rotation=(X=0,Y=0,Z=-0.3827,W=0.9239),Translation=(X=0,Y=0,Z ...(truncated)"
    WeightMinimum=0
    Weight=RigVMVariableNode() {variable=head_side_r}.Value
   )
   n4 = RigVMVariableNode() {variable="is Crouching "}
   Interpolate = /Script/RigVM.RigVMFunction_MathQuaternionSlerp(A=(X=-0.2346,Y=0.1094,Z=-0.4082,W=0.8754), B=(X=-0.212,Y=0.1485,Z=-0.554,W=0.7912), T=@n4.Value) {is=struct RigVMUnitNode, template="Interpolate::Execute(in A,in B,in T,out Result)"}
   > n5 = &RigUnit_ModifyTransforms(
    ItemToModify="((Item=(Type=Bone,Name=\"base_hood_back_c\"),Transform=(Rotation=(X=0,Y=0,Z=-0.5736,W=0.8192),Translation=(X=0,Y=0,Z=0),Scale3D=(X=1,Y=1,Z=1))),(Item=(Type=Bone,Name=\"base_hood_back_side_r\"),Transform=(Rotation=(X=-0.212,Y=0.1485,Z=-0.554,W=0.7912),Translati ...(truncated)"
    WeightMinimum=0
    ItemToModify.1.Transform.Rotation=@Interpolate.Result
    ItemToModify.2.Transform.Rotation=&RigVMFunction_MathQuaternionSlerp(A=(X=0.2346,Y=0.1094,Z=0.4082,W=0.8754), B=(X=0.212,Y=0.1485,Z=0.554,W=0.7912), T=@n4.Value) {template="Interpolate::Execute(in A,in B,in T,out Result)"}.Result
    ItemToModify.0.Transform.Rotation=&RigVMFunction_MathQuaternionSlerp(A=(X=0,Y=0,Z=-0.4226,W=0.9063), B=(X=0,Y=0,Z=-0.5736,W=0.8192), T=@n4.Value) {template="Interpolate::Execute(in A,in B,in T,out Result)"}.Result
    Weight=RigVMVariableNode() {variable=head_back_c}.Value
   )
   n6 = &RigVMFunction_ControlFlowBranch(ExecuteContext=@n5.ExecutePin, Condition=RigVMVariableNode() {variable=is Parachute Open}.Value)
   &RigUnit_ModifyTransforms(
    ItemToModify="((Item=(Type=Bone,Name=\"base_hood_side_r\"),Transform=(Rotation=(X=0,Y=0,Z=0.1736,W=0.9848),Translation=(X=0,Y=0,Z=0),Scale3D=(X=1,Y=1,Z=1))),(Item=(Type=Bone,Name=\"base_hood_front_r\"),Transform=(Rotation=(X=0,Y=0,Z=0.1305,W=0.9914),Translation=(X=0,Y=0,Z=0 ...(truncated)"
    Weight=1
    WeightMinimum=0
    ExecutePin=@n6.True
   )
   > &RigUnit_ModifyTransforms(
    ItemToModify="((Item=(Type=Bone,Name=\"upperarm_r_shoulderMover_0\"),Transform=(Rotation=(X=0,Y=0,Z=0,W=1),Translation=(X=0,Y=4,Z=-3),Scale3D=(X=1,Y=1,Z=1))),(Item=(Type=Bone,Name=\"upperarm_l_shoulderMover_0\"),Transform=(Rotation=(X=0,Y=0,Z=0,W=1),Translation=(X=0,Y=-4,Z= ...(truncated)"
    Weight=1
    WeightMinimum=0
   )
   &RigUnit_HierarchyImportFromSkeleton(ExecutePin=(), NameSpace=None, bIncludeCurves=False, bIncludeMeshSockets=False, bIncludeVirtualBones=True)
   addFKCtrl = RigVMFunctionReferenceNode(
    Custom_FKCtrl=()
    ExecuteContext=()
    Options_FKCtrl="(pelvis_1_#12=false,spine_3_#13=false,clavicle_5_#14=true,arm_7_#15=true,leg_9_#16=true,head_11_46DC15EF465CD29 ...(truncated)"
    Shape_Settings="(pelvis_2_#17=(bVisible=True,Name=\"Hexagon_Thin\",Color=(R=0.0117,G=0.8281,B=1,A=1),Transform=(Rotation=(X=0,Y=-0.7071,Z=0,W=0.7071),Translation=(X=2,Y=0,Z=0),Scale3D=(X=5,Y=5,Z=7))),spine_4_#18=(bVi ...(truncated)"
   )
   setFKCtrl = RigVMFunctionReferenceNode(ExecuteContext=(), setFK_Ctrl=true)
  ]}
]
hierarchy [
 root {loc=<0 0 0>, rot=<0 0 0 1>}
 pelvis {parent=root, loc=<0 -0.648 94.7699>, rot=<0.0127 0.707 -0.0127 -0.707>}
 spine_01 {parent=pelvis, loc=<5.0975 -0.6388 0>, rot=<0 0 0.034 -0.9994>}
 spine_02 {parent=spine_01, loc=<7.252 -0.035 0>, rot=<0 0 0.0238 -0.9997>}
 spine_03 {parent=spine_02, loc=<6.0106 0 0>, rot=<0 0 0.0149 -0.9999>}
 spine_04 {parent=spine_03, loc=<8.1023 0 0>, rot=<0 0 -0.0132 -0.9999>}
 spine_05 {parent=spine_04, loc=<10.0142 0 0>, rot=<0 0 -0.0015 -1>}
 clavicle_l {parent=spine_05, loc=<4.1651 -0.3023 -5.0942>, rot=<0.0465 -0.7289 -0.009 -0.6829>}
 upperarm_l {parent=clavicle_l, loc=<11.2117 -1.1486 -0.0332>, rot=<0.0179 -0.2982 -0.028 -0.9539>}
 lowerarm_l {parent=upperarm_l, loc=<26.5013 -0.0835 0.2605>, rot=<0.0036 0.0161 -0.2298 -0.9731>}
 hand_l {parent=lowerarm_l, loc=<23.1356 -1.1606 -0.2612>, rot=<-0.6704 -0.0028 0.0567 -0.7398>}
 weapon_l {parent=hand_l, loc=<-0.1974 -3.4111 0.3811>, rot=<0 0 0 1>}
 thumb_01_l {parent=hand_l, loc=<1.3118 -0.3191 -2.459>, rot=<0.6546 -0.4042 0.0664 -0.6354>}
 thumb_02_l {parent=thumb_01_l, loc=<5.9336 -0.0448 0.0401>, rot=<-0.0037 0.0213 0.298 -0.9543>}
 thumb_03_l {parent=thumb_02_l, loc=<2.2223 0.3682 -0.0907>, rot=<0.0118 -0.013 0.0066 -0.9998>}
 index_metacarpal_l {parent=hand_l, loc=<1.6535 0.0183 -2.96>, rot=<0.0158 -0.0124 -0.0173 -0.9996>}
 index_01_l {parent=index_metacarpal_l, loc=<8.3008 0.0357 -0.0524>, rot=<-0.0044 0.0485 0.0924 -0.9945>}
 index_02_l {parent=index_01_l, loc=<3.3587 -0.3004 0.0452>, rot=<-0.002 0.0027 0.3225 -0.9466>}
 index_03_l {parent=index_02_l, loc=<2.5578 -0.123 -0.0252>, rot=<-0.002 0.0067 0.2558 -0.9667>}
 middle_metacarpal_l {parent=hand_l, loc=<2.1856 -0.1725 -0.5984>, rot=<0.006 0.0168 -0.0329 -0.9993>}
 middle_01_l {parent=middle_metacarpal_l, loc=<7.3275 -0.2447 0.0129>, rot=<-0.0208 0.0042 0.0858 -0.9961>}
 middle_02_l {parent=middle_01_l, loc=<4.0123 0.1401 0.1175>, rot=<-0.0218 0.0258 0.3512 -0.9357>}
 middle_03_l {parent=middle_02_l, loc=<2.6077 0.0534 0.1373>, rot=<0.0082 0.0188 0.1708 -0.9851>}
 ring_metacarpal_l {parent=hand_l, loc=<2.2074 -0.3625 1.3638>, rot=<-0.0284 0.0722 -0.0458 -0.9959>}
 ring_01_l {parent=ring_metacarpal_l, loc=<7.0295 -0.6197 -0.1446>, rot=<-0.0489 -0.0252 0.1892 -0.9804>}
 ring_02_l {parent=ring_01_l, loc=<3.6174 0.4647 -0.1027>, rot=<0 0 0.2806 -0.9598>}
 ring_03_l {parent=ring_02_l, loc=<2.4152 0.114 0.0199>, rot=<0 0 0.1189 -0.9929>}
 pinky_metacarpal_l {parent=hand_l, loc=<1.7616 -0.9753 3.3814>, rot=<-0.0397 0.0814 -0.069 -0.9935>}
 pinky_01_l {parent=pinky_metacarpal_l, loc=<7.1709 -0.4831 -0.1094>, rot=<-0.0654 -0.0035 0.2242 -0.9723>}
 pinky_02_l {parent=pinky_01_l, loc=<2.8274 0.0893 -0.1043>, rot=<0.0037 0.0011 0.2854 -0.9584>}
 pinky_03_l {parent=pinky_02_l, loc=<2.2046 0.0307 -0.0078>, rot=<-0.0002 0.0098 0.1445 -0.9895>}
 hand_attach_l {parent=hand_l, loc=<8.5 -3 0>, rot=<0.7071 0.7071 0 0>}
 deform_wrist_dn_base_l {parent=hand_l, loc=<-0.7813 -1.7011 0.9082>, rot=<0 0 0 -1>}
 deform_wrist_dn_l {parent=deform_wrist_dn_base_l, loc=<0 0 0>, rot=<0 0 0 1>}
 deform_wrist_up_base_l {parent=hand_l, loc=<-0.1406 2.5627 0.0698>, rot=<0 0 0 -1>}
 deform_wrist_up_l {parent=deform_wrist_up_base_l, loc=<0 0 0>, rot=<0 0 0 1>}
 lowerarm_twist_02_l {parent=lowerarm_l, loc=<11.5353 0 0>, rot=<0 0 0 1>}
 lowerarm_twist_01_l {parent=lowerarm_l, loc=<16.7573 0 0>, rot=<0 0 0 1>}
 deform_elbow_in_base_l {parent=lowerarm_l, loc=<0 4.5494 0>, rot=<0 0 0 1>}
 deform_elbow_in_l {parent=deform_elbow_in_base_l, loc=<0 0 0>, rot=<0 0 0 1>}
 deform_elbow_out_base_l {parent=lowerarm_l, loc=<-2.6147 -4.6123 0>, rot=<0 0 0 1>}
 deform_elbow_out_l {parent=deform_elbow_out_base_l, loc=<0 0 0>, rot=<0 0 0 1>}
 FX_Socket_L {parent=lowerarm_l, type=Null, loc=<7.6769 -2.4616 0.123>, rot=<0.053 0.6416 -0.0396 0.7642>}
 upperarm_twist_01_l {parent=upperarm_l, loc=<7.2135 -0.1134 0>, rot=<0 0 0 1>}
 deform_upperarm_bk_base_l {parent=upperarm_twist_01_l, loc=<-5.5712 -7.2751 -1.2275>, rot=<0 0 0 -1>}
 deform_upperarm_bk_l {parent=deform_upperarm_bk_base_l, loc=<0 0 0>, rot=<0 0 0 1>}
 deform_upperarm_fr_base_l {parent=upperarm_twist_01_l, loc=<-5.3904 6.218 -0.859>, rot=<0 0 0 -1>}
 deform_upperarm_fr_l {parent=deform_upperarm_fr_base_l, loc=<0 0 0>, rot=<0 0 0 1>}
 deform_upperarm_up_base_l {parent=upperarm_twist_01_l, loc=<-8.9945 -0.5206 4.5756>, rot=<0 0 0 -1>}
 deform_upperarm_up_l {parent=deform_upperarm_up_base_l, loc=<0 0 0>, rot=<0 0 0 1>}
 upperarm_twist_02_l {parent=upperarm_l, loc=<19.7998 -0.1134 0>, rot=<0 0 0 1>}
 deform_bicep_base_l {parent=upperarm_twist_02_l, loc=<-1.8082 4.197 0>, rot=<0 0 0 1>}
 deform_bicep_l {parent=deform_bicep_base_l, loc=<0 0 0>, rot=<0 0 0 1>}
 upperarm_l_shoulderMover_0 {parent=upperarm_l, loc=<0 0 0>, rot=<0 0 0 1>}
 upperarm_l_shoulderMover_1 {parent=upperarm_l_shoulderMover_0, loc=<0 0 0>, rot=<0 0 0 1>}
 extra_base_arm_l {parent=clavicle_l, loc=<8.76 0.1948 -4.9912>, rot=<0.0243 0.0231 -0.7243 -0.6887>}
 extra_elbow_l {parent=extra_base_arm_l, loc=<10.1053 0.7625 -3.1108>, rot=<0.9389 0.0809 -0.1297 0.3085>}
 base_hood_front_l {parent=clavicle_l, loc=<2.1732 5.7446 2.2536>, rot=<-0.5157 -0.6224 -0.3295 0.488>}
 dyn_hood_front_l {parent=base_hood_front_l, loc=<0 0 0>, rot=<0 0 0 1>}
 base_hood_side_l {parent=clavicle_l, loc=<3.0586 -1.3909 4.6583>, rot=<0.6316 0.5245 0.2877 -0.4932>}
 dyn_hood_side_l {parent=base_hood_side_l, loc=<0 0 0>, rot=<0 0 0 1>}
 base_hood_back_side_l {parent=clavicle_l, loc=<0.0115 -6.4527 4.2685>, rot=<0.6417 0.1586 0.6539 -0.368>}
 dyn_hood_back_side_l {parent=base_hood_back_side_l, loc=<0 0 0>, rot=<0 0 0 1>}
 neck_01 {parent=spine_05, loc=<11.3775 -0.6396 0>, rot=<0 0 -0.2003 -0.9797>}
 neck_02 {parent=neck_01, loc=<4.5627 0 0>, rot=<0 0 0.0504 -0.9987>}
 head {parent=neck_02, loc=<5.3914 0 0>, rot=<0 0 0.115 -0.9934>}
 clavicle_r {parent=spine_05, loc=<4.1651 -0.3023 5.0942>, rot=<0.7289 0.0465 -0.6829 0.009>}
 upperarm_r {parent=clavicle_r, loc=<-11.2117 1.1486 0.0332>, rot=<0.0179 -0.2982 -0.028 -0.9539>}
 lowerarm_r {parent=upperarm_r, loc=<-26.5013 0.0835 -0.2604>, rot=<0.0036 0.0161 -0.2298 -0.9731>}
 hand_r {parent=lowerarm_r, loc=<-23.1356 1.1606 0.2612>, rot=<-0.6704 -0.0028 0.0567 -0.7398>}
 hand_attach_r {parent=hand_r, loc=<-8.5 3 0>, rot=<0 0 -0.7071 -0.7071>}
 weapon_r {parent=hand_r, loc=<0.1974 3.4111 -0.3811>, rot=<0 0 0 1>}
 thumb_01_r {parent=hand_r, loc=<-1.3118 0.3191 2.459>, rot=<0.6546 -0.4042 0.0664 -0.6354>}
 thumb_02_r {parent=thumb_01_r, loc=<-5.9336 0.0448 -0.0401>, rot=<-0.0037 0.0213 0.298 -0.9543>}
 thumb_03_r {parent=thumb_02_r, loc=<-2.2223 -0.3682 0.0907>, rot=<0.0118 -0.013 0.0066 -0.9998>}
 index_metacarpal_r {parent=hand_r, loc=<-1.6535 -0.0183 2.96>, rot=<0.0158 -0.0124 -0.0173 -0.9996>}
 index_01_r {parent=index_metacarpal_r, loc=<-8.3008 -0.0357 0.0524>, rot=<-0.0044 0.0485 0.0924 -0.9945>}
 index_02_r {parent=index_01_r, loc=<-3.3587 0.3004 -0.0452>, rot=<-0.002 0.0027 0.3225 -0.9466>}
 index_03_r {parent=index_02_r, loc=<-2.5578 0.123 0.0252>, rot=<-0.002 0.0067 0.2558 -0.9667>}
 middle_metacarpal_r {parent=hand_r, loc=<-2.1856 0.1725 0.5984>, rot=<0.006 0.0168 -0.0329 -0.9993>}
 middle_01_r {parent=middle_metacarpal_r, loc=<-7.3275 0.2447 -0.0129>, rot=<-0.0208 0.0042 0.0858 -0.9961>}
 middle_02_r {parent=middle_01_r, loc=<-4.0123 -0.1401 -0.1175>, rot=<-0.0218 0.0258 0.3512 -0.9357>}
 middle_03_r {parent=middle_02_r, loc=<-2.6077 -0.0534 -0.1373>, rot=<0.0082 0.0188 0.1708 -0.9851>}
 ring_metacarpal_r {parent=hand_r, loc=<-2.2074 0.3625 -1.3638>, rot=<-0.0284 0.0722 -0.0458 -0.9959>}
 ring_01_r {parent=ring_metacarpal_r, loc=<-7.0295 0.6197 0.1446>, rot=<-0.0489 -0.0252 0.1892 -0.9804>}
 ring_02_r {parent=ring_01_r, loc=<-3.6174 -0.4647 0.1027>, rot=<0 0 0.2806 -0.9598>}
 ring_03_r {parent=ring_02_r, loc=<-2.4152 -0.114 -0.0199>, rot=<0 0 0.1189 -0.9929>}
 pinky_metacarpal_r {parent=hand_r, loc=<-1.7616 0.9753 -3.3814>, rot=<-0.0397 0.0814 -0.069 -0.9935>}
 pinky_01_r {parent=pinky_metacarpal_r, loc=<-7.1709 0.4831 0.1094>, rot=<-0.0654 -0.0035 0.2242 -0.9723>}
 pinky_02_r {parent=pinky_01_r, loc=<-2.8274 -0.0893 0.1043>, rot=<0.0037 0.0011 0.2854 -0.9584>}
 pinky_03_r {parent=pinky_02_r, loc=<-2.2046 -0.0307 0.0078>, rot=<-0.0002 0.0098 0.1445 -0.9895>}
 deform_wrist_dn_base_r {parent=hand_r, loc=<0.781 1.7015 -0.9083>, rot=<0 0 0 -1>}
 deform_wrist_dn_r {parent=deform_wrist_dn_base_r, loc=<0.0003 -0.0003 0.0001>, rot=<0 0 0 -1>}
 deform_wrist_up_base_r {parent=hand_r, loc=<0.1405 -2.5626 -0.0698>, rot=<0 0 0 -1>}
 deform_wrist_up_r {parent=deform_wrist_up_base_r, loc=<0.0001 -0.0001 0>, rot=<0 0 0 -1>}
 lowerarm_twist_02_r {parent=lowerarm_r, loc=<-11.5353 0 0>, rot=<0 0 0 -1>}
 lowerarm_twist_01_r {parent=lowerarm_r, loc=<-16.7573 0 0>, rot=<0 0 0 -1>}
 deform_elbow_in_base_r {parent=lowerarm_r, loc=<0 -4.5494 -0.0001>, rot=<0 0 0 1>}
 deform_elbow_in_r {parent=deform_elbow_in_base_r, loc=<0 0 0.0001>, rot=<0 0 0 1>}
 deform_elbow_out_base_r {parent=lowerarm_r, loc=<2.6148 4.6123 0>, rot=<0 0 0 1>}
 deform_elbow_out_r {parent=deform_elbow_out_base_r, loc=<-0.0001 0 0>, rot=<0 0 0 1>}
 FX_Socket_R {parent=lowerarm_r, type=Null, loc=<-7.7069 0.3335 0.2645>, rot=<0.0168 0.7772 0.0136 -0.6289>}
 upperarm_twist_01_r {parent=upperarm_r, loc=<-7.2135 0.1134 0>, rot=<0 0 0 -1>}
 deform_upperarm_bk_base_r {parent=upperarm_twist_01_r, loc=<5.5714 7.2751 1.2273>, rot=<0 0 0 -1>}
 deform_upperarm_bk_r {parent=deform_upperarm_bk_base_r, loc=<-0.0001 0 0.0002>, rot=<0 0 0 -1>}
 deform_upperarm_fr_base_r {parent=upperarm_twist_01_r, loc=<5.3903 -6.218 0.8592>, rot=<0 0 0 -1>}
 deform_upperarm_fr_r {parent=deform_upperarm_fr_base_r, loc=<0.0001 0 -0.0001>, rot=<0 0 0 -1>}
 deform_upperarm_up_base_r {parent=upperarm_twist_01_r, loc=<8.9943 0.5206 -4.5754>, rot=<0 0 0 -1>}
 deform_upperarm_up_r {parent=deform_upperarm_up_base_r, loc=<0.0002 0 -0.0002>, rot=<0 0 0 -1>}
 upperarm_twist_02_r {parent=upperarm_r, loc=<-19.7998 0.1134 0>, rot=<0 0 0 -1>}
 deform_bicep_base_r {parent=upperarm_twist_02_r, loc=<1.8083 -4.197 -0.0001>, rot=<0 0 0 1>}
 deform_bicep_r {parent=deform_bicep_base_r, loc=<0 0 0>, rot=<0 0 0 1>}
 upperarm_r_shoulderMover_0 {parent=upperarm_r, loc=<0 0 0>, rot=<0 0 0 1>}
 upperarm_r_shoulderMover_1 {parent=upperarm_r_shoulderMover_0, loc=<0 0 0>, rot=<0 0 0 1>}
 extra_base_arm_r {parent=clavicle_r, loc=<-8.7606 -0.1944 4.9915>, rot=<0.7243 -0.6887 0.0243 -0.0231>}
 extra_elbow_r {parent=extra_base_arm_r, loc=<7.6691 -0.5359 -2.3396>, rot=<0.9323 -0.0808 -0.1277 -0.3287>}
 base_hood_front_r {parent=clavicle_r, loc=<-2.1732 -5.7446 -2.2534>, rot=<-0.3295 -0.488 0.5157 -0.6224>}
 dyn_hood_front_r {parent=base_hood_front_r, loc=<0 0 0>, rot=<0 0 0 1>}
 base_hood_side_r {parent=clavicle_r, loc=<-3.0586 1.3909 -4.6582>, rot=<-0.2877 -0.4932 0.6316 -0.5245>}
 dyn_hood_side_r {parent=base_hood_side_r, loc=<0 0 0>, rot=<0 0 0 1>}
 base_hood_back_side_r {parent=clavicle_r, loc=<-0.0115 6.4527 -4.2689>, rot=<0.6539 0.368 -0.6417 0.1586>}
 dyn_hood_back_side_r {parent=base_hood_back_side_r, loc=<0 0 0>, rot=<0 0 0 1>}
 attach_cape {parent=spine_05, loc=<9.017 -11.714 0>, rot=<0.0616 -0.7044 -0.0616 -0.7044>}
 cape {parent=attach_cape, type=Null, loc=<0 0 0>, rot=<0 0 0 1>}
 attach_backpack {parent=spine_05, loc=<-0.123 -13.473 0>, rot=<-0.0093 -0.707 0.0093 -0.707>}
 Backpack_BR {parent=attach_backpack, type=Null, loc=<0 0 0>, rot=<0 0 0 1>}
 base_hood_back_c {parent=spine_05, loc=<8.9541 -6.8398 0>, rot=<0 0 0.2635 -0.9647>}
 dyn_hood_back_c {parent=base_hood_back_c, loc=<0 0 0>, rot=<0 0 0 1>}
 base_hood_back_low_c {parent=spine_05, loc=<11.585 -4.4422 0>, rot=<0.2777 -0.6503 0.6503 -0.2777>}
 dyn_hood_back_low_c {parent=base_hood_back_low_c, loc=<0 0 0>, rot=<0 0 0 1>}
 deform_pec_aim_base_l {parent=spine_04, loc=<8.0082 12.8872 -2.9192>, rot=<0.0422 -0.4258 0.4306 -0.7947>}
 deform_pec_aim_l {parent=deform_pec_aim_base_l, loc=<0 0 0>, rot=<0 0 0 1>}
 deform_pec_base_l {parent=deform_pec_aim_l, loc=<16.3279 0 0>, rot=<0 0 0 1>}
 deform_pec_l {parent=deform_pec_base_l, loc=<0 0 0>, rot=<0 0 0 1>}
 deform_pec_aim_base_r {parent=spine_04, loc=<8.0082 12.8872 2.9192>, rot=<-0.4258 -0.0422 0.7947 0.4306>}
 deform_pec_aim_r {parent=deform_pec_aim_base_r, loc=<0 0 0>, rot=<0 0 0 1>}
 deform_pec_base_r {parent=deform_pec_aim_r, loc=<-16.3279 0.0001 0.0001>, rot=<0 0 0 -1>}
 deform_pec_r {parent=deform_pec_base_r, loc=<0 0 0>, rot=<0 0 0 1>}
 base_joystick {parent=spine_04, loc=<2.4696 23.7885 -0.0469>, rot=<0.0294 -0.7065 -0.0294 -0.7065>}
 dyn_joystick {parent=base_joystick, loc=<0 0 0>, rot=<0 0 0 1>}
 extra_wrist_l {parent=dyn_joystick, loc=<11.9851 -5.759 -1.8826>, rot=<-0.4055 -0.7252 -0.4995 -0.2456>}
 extra_wrist_end_l {parent=extra_wrist_l, loc=<11.0802 0 0>, rot=<-0.3135 -0.8852 0.3389 -0.0567>}
 extra_thumb_01_l {parent=extra_wrist_l, loc=<2.9827 2.117 -1.5773>, rot=<0.9503 0.1921 -0.1703 0.1763>}
 extra_thumb_02_l {parent=extra_thumb_01_l, loc=<4.4693 -0.5624 0.4139>, rot=<0.0516 -0.1139 -0.0384 -0.9914>}
 extra_thumb_03_l {parent=extra_thumb_02_l, loc=<3.3336 -0.1295 0.0159>, rot=<0.0062 0.0257 0.2327 -0.9722>}
 pinky_01_R1 {parent=extra_wrist_l, loc=<6.9001 -1.5587 0.9045>, rot=<0.6645 -0.4278 -0.2052 -0.5774>}
 pinky_02_R2 {parent=pinky_01_R1, loc=<1.4473 0 0>, rot=<0 0.0867 -0.3236 -0.9422>}
 pinky_03_R3 {parent=pinky_02_R2, loc=<1.9965 0 0>, rot=<-0.0807 0.1259 -0.3463 -0.9261>}
 pinky_04_R {parent=pinky_03_R3, loc=<1.8145 0 0>, rot=<0 0 0 1>}
 ring_01_R4 {parent=extra_wrist_l, loc=<8.1105 0.1868 2.1206>, rot=<0.5962 -0.5022 -0.1332 -0.6121>}
 ring_02_R5 {parent=ring_01_R4, loc=<3.3615 0 0>, rot=<0 0 -0.3418 -0.9398>}
 ring_03_R6 {parent=ring_02_R5, loc=<2.019 0 0>, rot=<-0.0644 0.0208 -0.3067 -0.9494>}
 ring_04_R {parent=ring_03_R6, loc=<2.2348 0 0>, rot=<0 0 0 1>}
 middle_01_R7 {parent=extra_wrist_l, loc=<9.3573 2.0365 1.7168>, rot=<0.8028 -0.2637 -0.0484 -0.5326>}
 middle_02_R8 {parent=middle_01_R7, loc=<2.5632 -0.0607 -0.0083>, rot=<-0.0282 -0.015 -0.3626 -0.9314>}
 middle_03_R9 {parent=middle_02_R8, loc=<2.3237 0.0007 -0.067>, rot=<0.0475 0.0057 -0.2504 -0.9669>}
 middle_04_R {parent=middle_03_R9, loc=<2.9498 0 0.166>, rot=<0.9788 0.0057 0.0275 -0.2029>}
 index_01_R10 {parent=extra_wrist_l, loc=<9.6405 4.9969 0.956>, rot=<0.8877 0.1931 -0.1058 -0.4043>}
 index_02_R11 {parent=index_01_R10, loc=<3.5713 0 0>, rot=<0 0 -0.2279 -0.9737>}
 index_03_R12 {parent=index_02_R11, loc=<2.6194 0 0>, rot=<-0.0167 0.0029 -0.1723 -0.9849>}
 index_04_R {parent=index_03_R12, loc=<2.2106 0 0>, rot=<0 0 0 1>}
 extra_wrist_r {parent=dyn_joystick, loc=<-12.1905 -5.7978 -1.8174>, rot=<0.8514 0.4794 0.1163 0.1782>}
 extra_wrist_end_r {parent=extra_wrist_r, loc=<9.763 0 0>, rot=<-0.0131 -0.023 0.4928 -0.8697>}
 extra_thumb_01_r {parent=extra_wrist_r, loc=<3.0639 2.705 -1.9471>, rot=<-0.7644 -0.2958 -0.059 -0.5698>}
 extra_thumb_02_r {parent=extra_thumb_01_r, loc=<4.3041 0 0>, rot=<-0.0439 0.0525 -0.0116 -0.9976>}
 extra_thumb_03_r {parent=extra_thumb_02_r, loc=<3.218 0 0>, rot=<-0.0085 -0.0015 0.1782 -0.984>}
 pinky_01_L13 {parent=extra_wrist_r, loc=<7.4829 -1.1069 1.023>, rot=<-0.0719 0.1893 -0.4512 -0.8691>}
 pinky_02_L14 {parent=pinky_01_L13, loc=<2.177 0 0>, rot=<-0.0143 -0.0528 -0.3806 -0.9231>}
 pinky_03_L15 {parent=pinky_02_L14, loc=<1.9174 0 0>, rot=<-0.0383 -0.05 -0.2653 -0.9621>}
 pinky_04_L {parent=pinky_03_L15, loc=<2.0706 0 0>, rot=<0 0 0 1>}
 ring_01_L16 {parent=extra_wrist_r, loc=<8.4508 -1.6315 -0.3355>, rot=<-0.0527 0.1704 -0.3884 -0.9041>}
 ring_02_L17 {parent=ring_01_L16, loc=<3.561 0 0>, rot=<0.0413 0.046 -0.5102 -0.8578>}
 ring_03_L18 {parent=ring_02_L17, loc=<1.7934 0 0>, rot=<0.0092 -0.0187 -0.217 -0.976>}
 ring_04_L {parent=ring_03_L18, loc=<2.5995 0 0>, rot=<0 0 0 1>}
 middle_01_L19 {parent=extra_wrist_r, loc=<9.3179 -1.4996 -2.557>, rot=<-0.1445 0.1481 -0.2169 -0.954>}
 middle_02_L20 {parent=middle_01_L19, loc=<4.0671 0 0>, rot=<0 0 -0.3778 -0.9259>}
 middle_03_L21 {parent=middle_02_L20, loc=<1.985 0 0>, rot=<-0.0078 -0.0327 -0.2051 -0.9782>}
 middle_04_L {parent=middle_03_L21, loc=<2.8222 0 0>, rot=<0 0 0 1>}
 index_01_L22 {parent=extra_wrist_r, loc=<9.521 -0.0965 -4.969>, rot=<-0.3875 -0.0373 0.0353 -0.9204>}
 index_02_L23 {parent=index_01_L22, loc=<3.6426 0 0>, rot=<0.0713 -0.0352 -0.2306 -0.9698>}
 index_03_L24 {parent=index_02_L23, loc=<2.8116 0 0>, rot=<0.0403 -0.0127 -0.1684 -0.9848>}
 index_04_L {parent=index_03_L24, loc=<2.0939 0 0>, rot=<0.0172 0 0 -0.9999>}
 aim_joystick {parent=spine_04, loc=<4.4365 0.205 -0.0469>, rot=<0.4788 0.5204 0.5204 0.4788>}
 base_aim_joystick {parent=spine_04, loc=<2.4696 23.7885 -0.0469>, rot=<-0.4788 -0.5204 -0.5204 -0.4788>}
 dyn_aim_joystick {parent=base_aim_joystick, loc=<0 0 0>, rot=<0 0 0 1>}
 deform_lat_l_base {parent=spine_03, loc=<5.6041 -4.8405 -13.6119>, rot=<0 0 0 -1>}
 deform_lat_l {parent=deform_lat_l_base, loc=<0 0 0>, rot=<0 0 0 1>}
 deform_lat_r_base {parent=spine_03, loc=<5.604 -4.8406 13.6119>, rot=<0 0 1 0>}
 deform_lat_r {parent=deform_lat_r_base, loc=<-0.0001 0 0>, rot=<0 0 0 -1>}
 base_jacket_l {parent=spine_01, loc=<4.6725 9.1737 -11.571>, rot=<-0.0624 0.3848 0.9195 0.0504>}
 dyn_jacket_l {parent=base_jacket_l, loc=<0 0 0>, rot=<0 0 0 1>}
 base_jacket_r {parent=spine_01, loc=<4.6726 9.1737 11.571>, rot=<0.0624 -0.3848 0.9195 0.0504>}
 dyn_jacket_r {parent=base_jacket_r, loc=<0 0 0>, rot=<0 0 0 1>}
 thigh_l {parent=pelvis, loc=<-2.5624 -0.0642 -10.3834>, rot=<0.0004 0.0055 0.0155 -0.9999>}
 calf_l {parent=thigh_l, loc=<-39.2091 -0.1238 -0.6795>, rot=<0.0006 -0.0161 -0.0391 -0.9991>}
 foot_l {parent=calf_l, loc=<-43.624 -2.818 -0.713>, rot=<0.044 0.0061 0.0367 -0.9983>}
 ball_l {parent=foot_l, loc=<-4.6464 10.3094 -0.3137>, rot=<0.0001 -0.0001 -0.7063 -0.7079>}
 base_footTongue_L {parent=foot_l, loc=<1.5277 5.7314 0.2743>, rot=<-0.0357 -0.7045 -0.0284 -0.7082>}
 dyn_footTongue_L {parent=base_footTongue_L, loc=<0 0 0>, rot=<0 0 0 1>}
 calf_twist_02_l {parent=calf_l, loc=<-17.5927 -1.7943 -0.2638>, rot=<0.0237 0.0095 -0.0035 -0.9997>}
 calf_twist_01_l {parent=calf_l, loc=<-34.2864 -1.9197 -0.5794>, rot=<0.0237 0.0095 -0.0035 -0.9997>}
 deform_knee_in_base_l {parent=calf_l, loc=<2.8091 -9.2971 0>, rot=<0 0 0 -1>}
 deform_knee_in_l {parent=deform_knee_in_base_l, loc=<0 0 0>, rot=<0 0 0 1>}
 deform_knee_out_base_l {parent=calf_l, loc=<-0.2964 5.0241 0>, rot=<0 0 0 -1>}
 deform_knee_out_l {parent=deform_knee_out_base_l, loc=<0 0 0>, rot=<0 0 0 1>}
 thigh_twist_01_l {parent=thigh_l, loc=<-23.0764 0 0>, rot=<0.0013 0 0 -1>}
 base_keyboard {parent=thigh_twist_01_l, loc=<1.974 0.7996 -4.9306>, rot=<-0.0028 -0.711 0.0006 -0.7032>}
 dyn_keyboard {parent=base_keyboard, loc=<0 0 0>, rot=<0 0 0 1>}
 deform_groin_base_l {parent=thigh_l, loc=<0 8.4477 0>, rot=<0 0 0 -1>}
 deform_groin_l {parent=deform_groin_base_l, loc=<0 0 0>, rot=<0 0 0 1>}
 base_shirt_f_left {parent=thigh_l, loc=<-11.3528 8.7155 -1.9989>, rot=<-0.0019 -0.711 0.0015 -0.7032>}
 dyn_shirt_f_left {parent=base_shirt_f_left, loc=<0 0 0>, rot=<0 0 0 1>}
 base_shirt_b_left {parent=thigh_l, loc=<-11.4671 -7.8185 -0.2674>, rot=<-0.0019 -0.711 0.0015 -0.7032>}
 dyn_shirt_b_left {parent=base_shirt_b_left, loc=<0 0 0>, rot=<0 0 0 1>}
 thigh_r {parent=pelvis, loc=<-2.5624 -0.0642 10.3834>, rot=<0.0055 -0.0004 0.9999 0.0155>}
 calf_r {parent=thigh_r, loc=<39.2091 0.1238 0.6795>, rot=<0.0006 -0.0161 -0.0391 -0.9991>}
 foot_r {parent=calf_r, loc=<43.624 2.818 0.713>, rot=<0.044 0.0061 0.0367 -0.9983>}
 ball_r {parent=foot_r, loc=<4.6464 -10.3094 0.3137>, rot=<0.0001 -0.0001 -0.7063 -0.7079>}
 base_footTongue_R {parent=foot_r, loc=<-1.5305 -6.0116 -0.2998>, rot=<0.7082 0.0284 -0.7045 -0.0357>}
 dyn_footTongue_R {parent=base_footTongue_R, loc=<0 0 0>, rot=<0 0 0 1>}
 calf_twist_02_r {parent=calf_r, loc=<17.5927 1.7943 0.2638>, rot=<0.0237 0.0095 -0.0035 -0.9997>}
 calf_twist_01_r {parent=calf_r, loc=<34.2864 1.9197 0.5794>, rot=<0.0237 0.0095 -0.0035 -0.9997>}
 deform_knee_in_base_r {parent=calf_r, loc=<-2.8091 9.2971 0>, rot=<0 0 0 1>}
 deform_knee_in_r {parent=deform_knee_in_base_r, loc=<0 0 0>, rot=<0 0 0 1>}
 deform_knee_out_base_r {parent=calf_r, loc=<0.2964 -5.0241 0>, rot=<0 0 0 -1>}
 deform_knee_out_r {parent=deform_knee_out_base_r, loc=<0 0 0>, rot=<0 0 0 1>}
 thigh_twist_01_r {parent=thigh_r, loc=<23.0764 0 0>, rot=<0.0013 0 0 -1>}
 base_granade {parent=thigh_twist_01_r, loc=<-3.5767 -1.059 4.9775>, rot=<-0.7032 0.0006 0.711 0.0028>}
 dyn_granade {parent=base_granade, loc=<0 0 0>, rot=<0 0 0 1>}
 base_granadeCable_01 {parent=thigh_twist_01_r, loc=<-14.7734 1.0292 9.7088>, rot=<0.4817 0.5845 -0.2462 -0.6047>}
 dyn_granadeCable_01 {parent=base_granadeCable_01, loc=<0 0 0>, rot=<0 0 0 1>}
 deform_groin_base_r {parent=thigh_r, loc=<0 -8.4477 0>, rot=<0 0 0 -1>}
 deform_groin_r {parent=deform_groin_base_r, loc=<0 0 0>, rot=<0 0 0 -1>}
 base_shirt_f_right {parent=thigh_r, loc=<9.1568 -9.5077 0.9837>, rot=<-0.7032 0.0015 0.711 0.0019>}
 dyn_shirt_f_right {parent=base_shirt_f_right, loc=<0 0 0>, rot=<0 0 0 1>}
 base_shirt_b_right {parent=thigh_r, loc=<9.8185 7.3957 0.9747>, rot=<-0.7032 0.0015 0.711 0.0019>}
 dyn_shirt_b_right {parent=base_shirt_b_right, loc=<0 0 0>, rot=<0 0 0 1>}
 deform_glute_base_l {parent=pelvis, loc=<-13.1199 -13.5027 -12.4424>, rot=<-0.011 0.0565 -0.4248 -0.9035>}
 deform_glute_l {parent=deform_glute_base_l, loc=<-0.0103 -0.0257 0>, rot=<0 0 0 1>}
 deform_glute_base_r {parent=pelvis, loc=<-13.1198 -13.5027 12.4424>, rot=<0.0565 0.011 0.9035 -0.4248>}
 deform_glute_r {parent=deform_glute_base_r, loc=<0 0 0>, rot=<0 0 0 1>}
 base_shirt_f_med {parent=pelvis, loc=<-12.871 9.7768 -0.5998>, rot=<-0.0127 -0.707 0.0127 -0.707>}
 dyn_shirt_f_med {parent=base_shirt_f_med, loc=<0 0 0>, rot=<0 0 0 1>}
 base_shirt_b_med {parent=pelvis, loc=<-14.5686 -8.0402 0.3754>, rot=<-0.0127 -0.707 0.0127 -0.707>}
 dyn_shirt_b_med {parent=base_shirt_b_med, loc=<0 0 0>, rot=<0 0 0 1>}
 attach {parent=root, loc=<0 0 0>, rot=<0 0 0 1>}
 ik_hand_root {parent=root, loc=<0 0 0>, rot=<0 0 0 1>}
 ik_hand_gun {parent=ik_hand_root, loc=<-53.7993 3.4863 105.1719>, rot=<0.6513 0.0939 0.3837 -0.6479>}
 ik_hand_r {parent=ik_hand_gun, loc=<0 0 0>, rot=<0 0 0 1>}
 ik_hand_l {parent=ik_hand_gun, loc=<74.0247 66.6541 40.682>, rot=<0.688 0.6195 0.3781 0>}
 ik_foot_root {parent=root, loc=<0 0 0>, rot=<0 0 0 1>}
 ik_foot_l {parent=ik_foot_root, loc=<11.2777 -7.5397 9.7673>, rot=<0.0357 0.7045 0.0284 -0.7082>}
 ik_foot_r {parent=ik_foot_root, loc=<-11.2777 -7.5397 9.7673>, rot=<0.7082 0.0284 -0.7045 0.0357>}
 root_fp {parent=root, loc=<0 0 0>, rot=<0 0 0 1>}
 camera_root {parent=root_fp, loc=<0 0 0>, rot=<0 0 0 1>}
 camera {parent=camera_root, loc=<0 0 0>, rot=<0 0 0 1>}
 attach_fp {parent=root_fp, loc=<0 0 0>, rot=<0 0 0 1>}
 attach_emote_01 {parent=root, loc=<0 0 0>, rot=<0 0 0 1>}
 attach_emote_02 {parent=root, loc=<0 0 0>, rot=<0 0 0 1>}
 attach_emote_03 {parent=root, loc=<0 0 0>, rot=<0 0 0 1>}
 warp_point_01 {parent=root, loc=<0 0 0>, rot=<0 0 0 1>}
 warp_point_02 {parent=root, loc=<0 0 0>, rot=<0 0 0 1>}
 warp_point_03 {parent=root, loc=<0 0 0>, rot=<0 0 0 1>}
]
variables [
 HoodBackSide_L_Rotation {type=FQuat, default=(X=0,Y=0,Z=0,W=1)}
 clavicle_rot_l {type=FVector, default=(X=0,Y=0,Z=0)}
 clavicle_rot_r {type=FVector, default=(X=0,Y=0,Z=0)}
 elbow_l_distance_spine_02 {type=double, default=0}
 elbow_r_distance_spine_02 {type=double, default=0}
 head_SPR_look_up_alpha {type=double, default=0}
 head_back_c {type=double, default=0}
 head_side_l {type=double, default=0}
 head_side_r {type=double, default=0}
 ikRobotArms {type=bool, default=False}
 "is Crouch Moving " {type=bool, default=False}
 "is Crouching " {type=double, default=0}
 is Parachute Open {type=bool, default=False}
 is Playing Emote {type=double, default=0}
 is Targeting {type=double, default=0}
 isMoving {type=bool, default=False}
 look_up_and_left_two_hands_multiplied_alpha {type=double, default=0}
 upperarm_r_SPR_aim_up_two hands alpha {type=double, default=0}
]
