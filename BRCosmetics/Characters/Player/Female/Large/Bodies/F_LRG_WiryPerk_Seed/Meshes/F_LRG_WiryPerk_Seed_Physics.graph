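# Physics asset for the F_LRG_WiryPerk_Seed large female player cosmetic body, covering only the pelvis, thighs and two jacket bone chains.
# Capsules on the pelvis and thighs give the lower body its volume, while tiny capsules on the base jacket bones anchor each side's dyn_jacket capsule.
# Thighs and jacket bases join the pelvis with free angular motion.
# Each dyn_jacket joint has tight twist and swing2 limits, leaving a hinge-like swing on one axis.
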PhysicsAsset /BRCosmetics/Characters/Player/Female/Large/Bodies/F_LRG_WiryPerk_Seed/Meshes/F_LRG_WiryPerk_Seed_Physics
bodies {
 base_jacket_l [{shapeType=Capsule, length=0.505, radius=0.505, shapeName=base_jacket_l_capsule, rotation=<rot 0 -90 0>}]
 base_jacket_r [{shapeType=Capsule, length=0.505, radius=0.505, shapeName=base_jacket_r_capsule, rotation=<rot 0 -90 0>}]
 dyn_jacket_l [{
   shapeType=Capsule
   length=4.6953
   radius=1.4502
   shapeName=dyn_jacket_l_capsule
   center=<15.5695 0.9148 1.8027>
   rotation=<rot -0.2268 -0.0001 7.6818>
  }]
 dyn_jacket_r [{
   shapeType=Capsule
   length=4.4477
   radius=1.6978
   shapeName=dyn_jacket_l_capsule
   center=<15.2656 1.6555 -1.6915>
   rotation=<rot -0.2268 -0.0001 -12.3601>
  }]
 pelvis [{shapeType=Capsule, length=13.8718, radius=9.2193, shapeName=pelvis_capsule}]
 thigh_l [{
   shapeType=Capsule
   length=12.8176
   radius=8.7278
   shapeName=thigh_l_capsule
   center=<-8.8464 1.1194 2.5532>
   rotation=<rot 86.6465 179.9999 180>
  }]
 thigh_r [{
   shapeType=Capsule
   length=12.8176
   radius=8.7278
   shapeName=thigh_l_capsule
   center=<8.8655 -0.8035 -2.8791>
   rotation=<rot 86.6465 179.9999 180>
  }]
}
constraints [
 {
  bone1Name=thigh_l
  bone2Name=pelvis
  swing1LimitDegrees=45
  swing1Motion=Free
  swing2LimitDegrees=45
  swing2Motion=Free
  twistLimitDegrees=45
  twistMotion=Free
 }
 {
  bone1Name=thigh_r
  bone2Name=pelvis
  swing1LimitDegrees=45
  swing1Motion=Free
  swing2LimitDegrees=45
  swing2Motion=Free
  twistLimitDegrees=45
  twistMotion=Free
 }
 {
  bone1Name=base_jacket_l
  bone2Name=pelvis
  swing1LimitDegrees=45
  swing1Motion=Free
  swing2LimitDegrees=45
  swing2Motion=Free
  twistLimitDegrees=45
  twistMotion=Free
 }
 {bone1Name=dyn_jacket_l, bone2Name=base_jacket_l, swing1LimitDegrees=20, swing2LimitDegrees=1, twistLimitDegrees=1}
 {
  bone1Name=base_jacket_r
  bone2Name=pelvis
  swing1LimitDegrees=45
  swing1Motion=Free
  swing2LimitDegrees=45
  swing2Motion=Free
  twistLimitDegrees=45
  twistMotion=Free
 }
 {bone1Name=dyn_jacket_r, bone2Name=base_jacket_r, swing1LimitDegrees=20, swing2LimitDegrees=1, twistLimitDegrees=1}
]
