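# Post-process style cosmetic AnimBP for the F_LRG_WiryPerk_Seed player body, copying the attached parent's pose and adding secondary motion.
# Control Rig, fed by movement and action flags, plus additive bone offsets form the corrective pose.
# A front-end idle curve blends DefaultSlot over selected bones, then RigidBody and AnimDynamics simulate dangling parts like grenade cable and hood.
# The event graph derives those flags and tracks zipline events to set an acceleration/velocity vector.
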AnimBlueprint /BRCosmetics/Characters/Player/Female/Large/Bodies/F_LRG_WiryPerk_Seed/Meshes/F_LRG_WiryPerk_Seed_AnimBP
target_skeleton=&/F_LRG_WiryPerk_Seed_Skeleton
graphs [
 AnimGraph {nodes [
   Output Pose(Result=Use cached pose 'animDynamics'()) {layerGroup=None}
   n1 = Layered blend per bone(
    BasePose=Use cached pose 'Correctives Pre Dynamics'()
    BlendPoses_0=Slot 'DefaultSlot'()
    BlendWeights_0=Get Custom Idle is Playing Float() -> (Custom Idle is Playing Float:double)
   ) {layerSetup [{branchFilters [
       {boneName=base_aim_joystick}
       {boneName=dyn_aim_joystick}
       {boneName=base_joystick}
       {boneName=dyn_joystick}
       {boneName=extra_wrist_r}
       {boneName=extra_wrist_l}
       {boneName=extra_base_arm_r}
       {boneName=extra_base_arm_l}
       {boneName=upperarm_r_shoulderMover_0}
       {boneName=upperarm_l_shoulderMover_0}
       {boneName=base_shirt_f_right}
       {boneName=base_shirt_b_right}
       {boneName=base_shirt_f_med}
       {boneName=base_shirt_b_med}
       {boneName=base_shirt_f_left}
       {boneName=base_shirt_b_left}
      ]}]}
   n2 = "Transform (Modify) Bone - Bone: upperarm_l_shoulderMover_0"(ComponentPose=Local To Component(LocalPose=@n1), Alpha=1, Translation="0,0,0", Rotation="15,0,15", Scale="1,1,1") {
    rotationMode=BMM_Additive
    rotationSpace=BCS_BoneSpace
    translationMode=BMM_Additive
    translationSpace=BCS_BoneSpace
    alphaBoolBlend {blendOption=Linear}
    boneToModify {boneName=upperarm_l_shoulderMover_0}
    rotation {pitch=15, roll=15}
   }
   n3 = RigidBody!(ComponentPose=@n2, Alpha, OverrideWorldGravity:Vector="0,0,-980", ExternalForce) -> (Pose) {
    bClampLinearTranslationLimitToRefPose=true
    bOverrideWorldGravity=true
    overridePhysicsAsset=&/F_LRG_WiryPerk_Seed_Physics
    simulationSpace=BaseBoneSpace
    alphaBoolBlend {blendOption=Linear}
    baseBoneRef {boneName=root}
    simSpaceSettings {
     dampingAlpha=0
     maxAngularAcceleration=26
     maxAngularVelocity=8
     maxLinearAcceleration=1800
     maxLinearVelocity=700
     velocityScaleZ=0.3
     worldAlpha=1
     externalLinearDragV=<0.5 0.5 0.5>
    }
   }
   n4 = Anim Dynamics - Bone: dyn_granadeCable_01!(ComponentPose=@n3, Alpha=1, ComponentLinearAccScale:Vector="1,1,1", ComponentLinearVelScale:Vector="1,1,1", ExternalForce="0,0,0") -> (Pose) {
    angularDampingOverride=1
    angularSpringConstant=20000
    bAngularSpring=true
    bOverrideAngularDamping=true
    bUseGravityOverride=true
    gravityScale=0.35
    simulationSpace=World
    alphaBoolBlend {blendOption=Linear}
    boundBone {boneName=dyn_granadeCable_01}
    componentLinearAccScale=<1 1 1>
    componentLinearVelScale=<1 1 1>
    gravityOverride {z=-980}
    physicsBodyDefinitions [{
      boundBone {boneName=dyn_granadeCable_01}
      boxExtents=<30 30 30>
      constraintSetup {angularLimitsMax=<15 30 40>, angularLimitsMin=<-15 -30 -40>, angularTarget=<1 0 0>}
      localJointOffset=<4 0 0>
     }]
    relativeSpaceBone {boneName=pelvis}
   }
   n5 = Anim Dynamics - Bone: dyn_hood_side_l(ComponentPose=@n4, Alpha=1, ExternalForce="0,0,0") {
    angularDampingOverride=1
    angularSpringConstant=20000
    bAngularSpring=true
    bOverrideAngularDamping=true
    bUseGravityOverride=true
    gravityScale=0.35
    simulationSpace=RootRelative
    alphaBoolBlend {blendOption=Linear}
    boundBone {boneName=dyn_hood_side_l}
    componentLinearAccScale=<1 1 1>
    componentLinearVelScale=<1 1 1>
    gravityOverride {z=-980}
    physicsBodyDefinitions [{
      boundBone {boneName=dyn_hood_side_l}
      boxExtents=<30 30 30>
      constraintSetup {angularLimitsMax=<0 0 10>, angularLimitsMin=<0 0 -5>, angularTarget=<1 0 0>}
      localJointOffset=<2 0 0>
     }]
    relativeSpaceBone {boneName=spine_05}
   }
   n6 = Anim Dynamics - Bone: dyn_hood_back_side_l(ComponentPose=@n5, Alpha=1, ExternalForce="0,0,0") {
    angularDampingOverride=1
    angularSpringConstant=20000
    bAngularSpring=true
    bOverrideAngularDamping=true
    bUseGravityOverride=true
    gravityScale=0.35
    simulationSpace=RootRelative
    alphaBoolBlend {blendOption=Linear}
    boundBone {boneName=dyn_hood_back_side_l}
    componentLinearAccScale=<1 1 1>
    componentLinearVelScale=<1 1 1>
    gravityOverride {z=-980}
    physicsBodyDefinitions [{
      boundBone {boneName=dyn_hood_back_side_l}
      boxExtents=<30 30 30>
      constraintSetup {angularLimitsMax=<0 0 12>, angularLimitsMin=<0 0 -5>, angularTarget=<1 0 0>}
      localJointOffset=<2 0 0>
     }]
    relativeSpaceBone {boneName=spine_05}
   }
   n7 = Anim Dynamics - Bone: dyn_hood_back_side_r(ComponentPose=@n6, Alpha=1, ExternalForce="0,0,0") {
    angularDampingOverride=1
    angularSpringConstant=20000
    bAngularSpring=true
    bOverrideAngularDamping=true
    bUseGravityOverride=true
    gravityScale=0.35
    simulationSpace=RootRelative
    alphaBoolBlend {blendOption=Linear}
    boundBone {boneName=dyn_hood_back_side_r}
    componentLinearAccScale=<1 1 1>
    componentLinearVelScale=<1 1 1>
    gravityOverride {z=-980}
    physicsBodyDefinitions [{
      boundBone {boneName=dyn_hood_back_side_r}
      boxExtents=<30 30 30>
      constraintSetup {angularLimitsMax=<0 0 5>, angularLimitsMin=<0 0 -12>, angularTarget=<1 0 0>}
      localJointOffset=<2 0 0>
     }]
    relativeSpaceBone {boneName=spine_05}
   }
   n8 = Anim Dynamics - Bone: dyn_hood_side_r(ComponentPose=@n7, Alpha=1, ExternalForce="0,0,0") {
    angularDampingOverride=1
    angularSpringConstant=20000
    bAngularSpring=true
    bOverrideAngularDamping=true
    bUseGravityOverride=true
    gravityScale=0.35
    simulationSpace=RootRelative
    alphaBoolBlend {blendOption=Linear}
    boundBone {boneName=dyn_hood_side_r}
    componentLinearAccScale=<1 1 1>
    componentLinearVelScale=<1 1 1>
    gravityOverride {z=-980}
    physicsBodyDefinitions [{
      boundBone {boneName=dyn_hood_side_r}
      boxExtents=<30 30 30>
      constraintSetup {angularLimitsMax=<0 0 5>, angularLimitsMin=<0 0 -10>, angularTarget=<1 0 0>}
      localJointOffset=<2 0 0>
     }]
    relativeSpaceBone {boneName=spine_05}
   }
   n9 = Anim Dynamics - Bone: dyn_hood_back_c(ComponentPose=@n8, Alpha=1, ExternalForce="0,0,0") {
    angularDampingOverride=1
    angularSpringConstant=20000
    bAngularSpring=true
    bOverrideAngularDamping=true
    bUseGravityOverride=true
    gravityScale=0.35
    simulationSpace=RootRelative
    alphaBoolBlend {blendOption=Linear}
    boundBone {boneName=dyn_hood_back_c}
    componentLinearAccScale=<1 1 1>
    componentLinearVelScale=<1 1 1>
    gravityOverride {z=-980}
    physicsBodyDefinitions [{
      boundBone {boneName=dyn_hood_back_c}
      boxExtents=<30 30 30>
      constraintSetup {angularLimitsMax=<0 0 5>, angularLimitsMin=<0 0 -15>, angularTarget=<1 0 0>}
      localJointOffset=<2 0 0>
     }]
    relativeSpaceBone {boneName=spine_05}
   }
   n10 = Anim Dynamics - Bone: dyn_hood_back_low_c(ComponentPose=@n9, Alpha=1, ExternalForce="0,0,0") {
    angularDampingOverride=1
    angularSpringConstant=20000
    bAngularSpring=true
    bOverrideAngularDamping=true
    bUseGravityOverride=true
    gravityScale=0.35
    simulationSpace=RootRelative
    alphaBoolBlend {blendOption=Linear}
    boundBone {boneName=dyn_hood_back_low_c}
    componentLinearAccScale=<1 1 1>
    componentLinearVelScale=<1 1 1>
    gravityOverride {z=-980}
    physicsBodyDefinitions [{
      boundBone {boneName=dyn_hood_back_low_c}
      boxExtents=<30 30 30>
      constraintSetup {angularLimitsMax=<0 0 15>, angularLimitsMin=<0 -15 -15>, angularTarget=<1 0 0>}
      localJointOffset=<4 0 0>
     }]
    relativeSpaceBone {boneName=spine_05}
   }
   n11 = "Transform (Modify) Bone - Bone: base_shirt_b_right"(ComponentPose=@n10, Alpha=1, Translation="0,-2,0", Rotation="0,0,0", Scale="1,1,1") {
    rotationMode=BMM_Additive
    rotationSpace=BCS_BoneSpace
    translationMode=BMM_Additive
    translationSpace=BCS_BoneSpace
    alphaBoolBlend {blendOption=Linear}
    boneToModify {boneName=base_shirt_b_right}
    translation {y=-2}
   }
   n12 = "Transform (Modify) Bone - Bone: base_jacket_l"(ComponentPose=@n11, Alpha=1, Translation="0,0,0", Rotation="-9.945,0,0", Scale="1,1,1") {
    rotationMode=BMM_Additive
    rotationSpace=BCS_BoneSpace
    translationMode=BMM_Additive
    translationSpace=BCS_BoneSpace
    alphaBoolBlend {blendOption=Linear}
    boneToModify {boneName=base_jacket_l}
    rotation {pitch=-9.945}
   }
   Save cached pose 'animDynamics'(Pose=Component To Local(ComponentPose=@n12)) {cacheName=animDynamics}
   n13 = Control Rig!(
    Source=Copy Pose From Mesh() {bUseAttachedParent=true}
    Alpha
    isMoving:bool=Get movingNotTargeting() -> (movingNotTargeting:bool)
    ikRobotArms:bool=Get DisabledIKRobotArms() -> (DisabledIKRobotArms:bool)
    "is Crouching ":double="To Float (Boolean)"(InBool=Get bIsCrouching())
    is Parachute Open:bool=Get bIsParachuteOpened()
    is Targeting:double="To Float (Boolean)"(InBool=Get bIsTargeting())
    "is Crouch Moving ":bool="Get Crouch moving "() -> ("Crouch moving ":bool)
    is Playing Emote:double=Get is Playing Emote CR() -> (is Playing Emote CR:double)
   ) -> (Pose) {
    alphaBoolBlend {blendOption=Linear}
    controlRigAssetReference {blueprintRigClass=&/F_LRG_WiryPerk_Seed_CtrlRig.F_LRG_WiryPerk_Seed_CtrlRig_C}
   }
   n14 = "Transform (Modify) Bone - Bone: base_jacket_r"(ComponentPose=Local To Component(LocalPose=@n13), Alpha=1, Translation="0,0,0", Rotation="-5,-2,0", Scale="1,1,1") {
    rotationMode=BMM_Additive
    rotationSpace=BCS_BoneSpace
    translationMode=BMM_Additive
    translationSpace=BCS_BoneSpace
    alphaBoolBlend {blendOption=Linear}
    boneToModify {boneName=base_jacket_r}
    rotation {pitch=-5, yaw=-2}
   }
   n15 = "Transform (Modify) Bone - Bone: upperarm_r_shoulderMover_0"(ComponentPose=@n14, Alpha=1, Translation="0,0,0", Rotation="18,0,0", Scale="1,1,1") {
    rotationMode=BMM_Additive
    rotationSpace=BCS_BoneSpace
    translationMode=BMM_Additive
    translationSpace=BCS_BoneSpace
    alphaBoolBlend {blendOption=Linear}
    boneToModify {boneName=upperarm_r_shoulderMover_0}
    rotation {pitch=18}
   }
   n16 = "Transform (Modify) Bone - Bone: upperarm_l_shoulderMover_0"(ComponentPose=@n15, Alpha=1, Translation="0,0,0", Rotation="15,0,0", Scale="1,1,1") {
    rotationMode=BMM_Additive
    rotationSpace=BCS_BoneSpace
    translationMode=BMM_Additive
    translationSpace=BCS_BoneSpace
    alphaBoolBlend {blendOption=Linear}
    boneToModify {boneName=upperarm_l_shoulderMover_0}
    rotation {pitch=15}
   }
   n17 = "Transform (Modify) Bone - Bone: dyn_hood_side_l"(ComponentPose=@n16, Alpha=1, Translation="-1.5,0,0", Rotation="0,0,0", Scale="1,1,1") {
    rotationMode=BMM_Additive
    rotationSpace=BCS_BoneSpace
    translationMode=BMM_Additive
    translationSpace=BCS_BoneSpace
    alphaBoolBlend {blendOption=Linear}
    boneToModify {boneName=dyn_hood_side_l}
    translation {x=-1.5}
   }
   n18 = "Transform (Modify) Bone - Bone: dyn_hood_back_side_l"(ComponentPose=@n17, Alpha=1, Translation="-1,0,0", Rotation="0,0,0", Scale="1,1,1") {
    rotationMode=BMM_Additive
    rotationSpace=BCS_BoneSpace
    translationMode=BMM_Additive
    translationSpace=BCS_BoneSpace
    alphaBoolBlend {blendOption=Linear}
    boneToModify {boneName=dyn_hood_back_side_l}
    translation {x=-1}
   }
   n19 = "Transform (Modify) Bone - Bone: base_shirt_f_left"(ComponentPose=@n18, Alpha=1, Translation="0,1,0", Rotation="0,0,0", Scale="1,1,1") {
    rotationMode=BMM_Additive
    rotationSpace=BCS_BoneSpace
    translationMode=BMM_Additive
    translationSpace=BCS_BoneSpace
    alphaBoolBlend {blendOption=Linear}
    boneToModify {boneName=base_shirt_f_left}
    translation {y=1}
   }
   n20 = "Transform (Modify) Bone - Bone: base_shirt_b_med"(ComponentPose=@n19, Alpha=1, Translation="0,0,0", Rotation="0,15,0", Scale="1,1,1") {
    rotationMode=BMM_Additive
    rotationSpace=BCS_BoneSpace
    translationMode=BMM_Additive
    translationSpace=BCS_BoneSpace
    alphaBoolBlend {blendOption=Linear}
    boneToModify {boneName=base_shirt_b_med}
    rotation {yaw=15}
   }
   n21 = "Transform (Modify) Bone - Bone: base_jacket_r"(ComponentPose=@n20, Alpha=1, Translation="0,0,0", Rotation="0,15,0", Scale="1,1,1") {
    rotationMode=BMM_Additive
    rotationSpace=BCS_BoneSpace
    translationMode=BMM_Additive
    translationSpace=BCS_BoneSpace
    alphaBoolBlend {blendOption=Linear}
    boneToModify {boneName=base_jacket_r}
    rotation {yaw=15}
   }
   Save cached pose 'Correctives Pre Dynamics'(Pose=Component To Local(ComponentPose=@n21)) {cacheName=Correctives Pre Dynamics}
  ]}
 EventGraph {nodes [
   Event BlueprintUpdateAnimation()
   n1 = Get bIsTargeting()
   > Set movingNotTargeting(movingNotTargeting:bool=AND Boolean(A=Get bIsMoving2D(), B=NOT Boolean(A=@n1))) -> (Output_Get:bool)
   n2 = Get bIsLeftArmHidden() -> (bIsLeftArmHidden:bool)
   n3 = Get bIsRightArmHidden() -> (bIsRightArmHidden:bool)
   > Set DisabledIKRobotArms(DisabledIKRobotArms:bool=NOT Boolean(A=AND Boolean(A=@n2, B=@n3))) -> (Output_Get:bool)
   n4 = Get bIsPlayingEmote()
   > Set FrontEnd notEmoting(FrontEnd notEmoting:bool=AND Boolean(A=Get bIsInFrontEnd(), B=NOT Boolean(A=@n4))) -> (Output_Get:bool)
   n5 = Get bIsCrouching()
   n6 = Get bIsMoving2D()
   > "Set Crouch moving "("Crouch moving ":bool=AND Boolean(A=@n5, B=@n6)) -> (Output_Get:bool)
   > Set Crouch Idle(Crouch Idle:bool=AND Boolean(A=@n5, B=NOT Boolean(A=@n6))) -> (Output_Get:bool)
   n7 = Get bIsPlayingEmote()
   n8 = Get bIsPlayingMeleeAnim()
   > Set is Playing Emote CR(is Playing Emote CR:double="To Float (Boolean)"(InBool=NOR Boolean!(A:bool=@n7, B:bool=@n8) -> (ReturnValue:bool))) -> (Output_Get:double)
   n9 = Get CurrentLOD()
   > Set current LOD Equal 1 ?(current LOD Equal 1 ?:bool="integer >= integer"(A=@n9, B=1)) -> (Output_Get:bool)
   > Branch(Condition=Get bIsInFrontEnd())
   > Set Custom Idle is Playing Float(Custom Idle is Playing Float:double=GetCurveValue(self=GetOwnerFrontEndAnimInstance(), CurveName=Frontend_Idle_Bypass)) -> (Output_Get:double)
   > Sequence()
   > n10 = Branch(Condition=Get OnZipline() -> (OnZipline:bool))
   > "Set  ACC/VEL Zipline"(" ACC/VEL Zipline":Vector="0, 0, 0") -> (Output_Get:Vector)
   "Set  ACC/VEL Zipline"(execute=@n10.else, " ACC/VEL Zipline"="1,1,1")
   n11 = Event BlueprintProcessAnimInputEvent()
   > n12 = Branch(Condition=IsNamed(self=@n11.AnimInputEvent, InEventName=aie.Player.Zipline.Entered))
   > Set OnZipline(OnZipline:bool=1) -> (Output_Get:bool)
   Branch(execute=@n12.else, Condition=IsNamed(self=@n11.AnimInputEvent, InEventName=aie.Player.Zipline.Exited))
   > Set OnZipline(OnZipline=false)
   "Not Equal (Integer)"(A=@n9)
  ]}
]
variables [
 " ACC/VEL Zipline" {type=Vector}
 Crouch Idle {type=boolean}
 "Crouch moving " {type=boolean}
 Crouch not Moving {type=boolean}
 current LOD Equal 1 ? {type=boolean}
 Custom Idle is Playing Float {type=number}
 DisabledIKRobotArms {type=boolean}
 FrontEnd notEmoting {type=boolean}
 is Playing Emote CR {type=number}
 is Targeting CR {type=number}
 movingNotTargeting {type=boolean}
 OnZipline {type=boolean}
]
