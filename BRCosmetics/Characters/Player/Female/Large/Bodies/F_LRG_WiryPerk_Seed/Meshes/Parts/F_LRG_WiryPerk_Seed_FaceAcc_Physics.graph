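# Physics asset for the face accessory mesh of the female large WiryPerk Seed character, set up for two dynamic side-hair strands.
# Each strand is a three-bone capsule chain from a base bone on neck_02, with zero twist and swing limits that tighten at the tip.
# Neck, upper-spine and hood capsules supply body volumes around the hair, linked by free angular constraints.
# The left hood-side body has no constraint or right-side counterpart.
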PhysicsAsset /BRCosmetics/Characters/Player/Female/Large/Bodies/F_LRG_WiryPerk_Seed/Meshes/Parts/F_LRG_WiryPerk_Seed_FaceAcc_Physics
bodies {
 base_hood_front_l [{
   shapeType=Capsule
   length=3
   radius=2
   shapeName=base_hood_front_l_capsule
   center=<-0.9747 1.2791 3.565>
   rotation=<rot -1.9623 -82.8619 -74.7076>
  }]
 base_hood_front_r [{
   shapeType=Capsule
   length=3
   radius=2
   shapeName=base_hood_front_r_capsule
   center=<-0.9895 -1.8862 3.5691>
   rotation=<rot 0.6946 -92.52 -74.6005>
  }]
 base_hood_side_l [{
   shapeType=Capsule
   length=10.1114
   radius=6.3799
   shapeName=spine_05_capsule
   center=<-3.4844 8.0092 -0.093>
   rotation=<rot -0.0001 0 9.4698>
  }]
 base_sideHair_L [{shapeType=Capsule, length=0, radius=0.505, shapeName=base_sideHair_L_capsule, rotation=<rot 0 -90 0>}]
 base_sideHair_R [{shapeType=Capsule, length=0, radius=0.505, shapeName=base_sideHair_L_capsule, rotation=<rot 0 -90 0>}]
 dyn_sideHair_L_01 [{shapeType=Capsule, length=3, radius=1.5, shapeName=dyn_sideHair_L_01_capsule, center=<2.5 -0.4165 0>, rotation=<rot 90 180 0>}]
 dyn_sideHair_L_02 [{shapeType=Capsule, length=2, radius=1.35, shapeName=dyn_sideHair_L_02_capsule, center=<2 0 0>, rotation=<rot 90 0 0>}]
 dyn_sideHair_L_03 [{shapeType=Capsule, length=4, radius=1, shapeName=dyn_sideHair_L_03_capsule, center=<3.0431 0 0>, rotation=<rot 90 0 0>}]
 dyn_sideHair_R_01 [{shapeType=Capsule, length=3, radius=1.5, shapeName=dyn_sideHair_R_01_capsule, center=<2.5 0.4165 0>, rotation=<rot 90 0 0>}]
 dyn_sideHair_R_02 [
  {shapeType=Capsule, length=2, radius=1.5, shapeName=dyn_sideHair_R_02_capsule, center=<2 0 0>, rotation=<rot 90 0 0>}
  {
   shapeType=Capsule
   length=3.7395
   radius=0.5
   shapeName=dyn_sideHair_R_02_capsule
   center=<2.6011 -1.6338 1.1923>
   rotation=<rot -53.2859 -10.0725 -27.1927>
  }
 ]
 dyn_sideHair_R_03 [{shapeType=Capsule, length=4, radius=1, shapeName=dyn_sideHair_R_03_capsule, center=<3.0431 0 0>, rotation=<rot 90 0 0>}]
 neck_01 [{shapeType=Capsule, length=0.2906, radius=3.7094, shapeName=neck_01_capsule, center=<-0.7017 1.0738 0.0486>, rotation=<rot 0 0 90>}]
 neck_02 [{shapeType=Capsule, length=1.9559, radius=3.2902, shapeName=neck_02_capsule, center=<0 0 0.1437>, rotation=<rot 0 0 90>}]
 spine_05 [
  {shapeType=Capsule, length=7.0475, radius=9.4438, shapeName=spine_05_capsule, center=<0.7059 -2.9785 0>}
  {
   shapeType=Capsule
   length=12.0938
   radius=6.0314
   shapeName=spine_05_capsule
   center=<8.827 -0.6507 -12.2194>
   rotation=<rot 0 23.2549 96.1288>
  }
  {
   shapeType=Capsule
   length=12.0938
   radius=6.0314
   shapeName=spine_05_capsule
   center=<8.8039 -0.5969 12.1953>
   rotation=<rot 0 23.2427 82.1723>
  }
  {shapeType=Capsule, length=7.9316, radius=10.3922, shapeName=spine_05_capsule, center=<15.6344 -13.8659 0>}
 ]
}
constraints [
 {bone1Name=dyn_sideHair_R_02, bone2Name=dyn_sideHair_R_01, swing1LimitDegrees=25, swing2LimitDegrees=15, twistLimitDegrees=0}
 {bone1Name=dyn_sideHair_R_03, bone2Name=dyn_sideHair_R_02, swing1LimitDegrees=20, swing2LimitDegrees=10, twistLimitDegrees=0}
 {bone1Name=dyn_sideHair_L_01, bone2Name=base_sideHair_L, swing1LimitDegrees=25, swing2LimitDegrees=15, twistLimitDegrees=0}
 {bone1Name=dyn_sideHair_L_02, bone2Name=dyn_sideHair_L_01, swing1LimitDegrees=25, swing2LimitDegrees=15, twistLimitDegrees=0}
 {bone1Name=dyn_sideHair_L_03, bone2Name=dyn_sideHair_L_02, swing1LimitDegrees=20, swing2LimitDegrees=10, twistLimitDegrees=0}
 {
  bone1Name=base_sideHair_L
  bone2Name=neck_02
  swing1LimitDegrees=45
  swing1Motion=Free
  swing2LimitDegrees=45
  swing2Motion=Free
  twistLimitDegrees=45
  twistMotion=Free
 }
 {
  bone1Name=neck_02
  bone2Name=neck_01
  swing1LimitDegrees=45
  swing1Motion=Free
  swing2LimitDegrees=45
  swing2Motion=Free
  twistLimitDegrees=45
  twistMotion=Free
 }
 {
  bone1Name=neck_01
  bone2Name=spine_05
  swing1LimitDegrees=45
  swing1Motion=Free
  swing2LimitDegrees=45
  swing2Motion=Free
  twistLimitDegrees=45
  twistMotion=Free
 }
 {
  bone1Name=base_hood_front_l
  bone2Name=spine_05
  swing1LimitDegrees=45
  swing1Motion=Free
  swing2LimitDegrees=45
  swing2Motion=Free
  twistLimitDegrees=45
  twistMotion=Free
 }
 {
  bone1Name=base_hood_front_r
  bone2Name=spine_05
  swing1LimitDegrees=45
  swing1Motion=Free
  swing2LimitDegrees=45
  swing2Motion=Free
  twistLimitDegrees=45
  twistMotion=Free
 }
 {bone1Name=dyn_sideHair_R_01, bone2Name=base_sideHair_R, swing1LimitDegrees=25, swing2LimitDegrees=15, twistLimitDegrees=0}
 {
  bone1Name=base_sideHair_R
  bone2Name=neck_02
  swing1LimitDegrees=45
  swing1Motion=Free
  swing2LimitDegrees=45
  swing2Motion=Free
  twistLimitDegrees=45
  twistMotion=Free
 }
]
