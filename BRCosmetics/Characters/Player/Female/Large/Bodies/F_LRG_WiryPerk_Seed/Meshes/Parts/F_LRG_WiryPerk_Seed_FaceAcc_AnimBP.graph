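# Anim Blueprint for the WiryPerk Seed female face-accessory mesh, adding secondary motion to its hair.
# The pose copied from the parent mesh runs through spring-driven AnimDynamics on face and back-hair bones, then a neck-relative RigidBody simulation.
# Bone offsets and weighted neck/clavicle constraints then adjust side and back hair.
# The event graph computes state-based gravity from a parameter asset, plus head-rotation and targeting alphas.
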AnimBlueprint /BRCosmetics/Characters/Player/Female/Large/Bodies/F_LRG_WiryPerk_Seed/Meshes/Parts/F_LRG_WiryPerk_Seed_FaceAcc_AnimBP
target_skeleton=&/F_LRG_WiryPerk_Seed_FaceAcc_Skeleton
graphs [
 AnimGraph {nodes [
   Anim Dynamics - Bone: dyn_faceHair_1_01_R(Alpha=1, ExternalForce="0,0,0") {
    angularDampingOverride=1
    angularSpringConstant=15000
    bAngularSpring=true
    bOverrideAngularDamping=true
    bUseGravityOverride=true
    gravityScale=0.35
    simulationSpace=World
    alphaBoolBlend {blendOption=Linear}
    boundBone {boneName=dyn_faceHair_1_01_R}
    componentAppliedLinearAccClamp=<5000 5000 5000>
    componentLinearAccScale=<0.5 0.5 0.5>
    componentLinearVelScale=<0.5 0.5 0.5>
    gravityOverride {z=-980}
    physicsBodyDefinitions [{
      boundBone {boneName=dyn_faceHair_1_01_R}
      boxExtents=<25 25 25>
      constraintSetup {angularLimitsMax=<0 25 5>, angularLimitsMin=<0 -10 -5>, angularTarget=<1 0 0>}
      localJointOffset=<1 0 0>
     }]
    relativeSpaceBone {boneName=spine_05}
   }
   n1 = Anim Dynamics - Bone: dyn_faceHair_1_00_F(ComponentPose=Local To Component(LocalPose=Copy Pose From Mesh() {bUseAttachedParent=true}), Alpha=1, ExternalForce="0,0,0") {
    angularDampingOverride=1
    angularSpringConstant=10000
    bAngularSpring=true
    bOverrideAngularDamping=true
    bUseGravityOverride=true
    gravityScale=0.35
    simulationSpace=World
    alphaBoolBlend {blendOption=Linear}
    boundBone {boneName=dyn_faceHair_1_00_F}
    componentAppliedLinearAccClamp=<5000 5000 5000>
    componentLinearAccScale=<0.5 0.5 0.5>
    componentLinearVelScale=<0.5 0.5 0.5>
    gravityOverride {z=-980}
    physicsBodyDefinitions [{
      boundBone {boneName=dyn_faceHair_1_00_F}
      boxExtents=<30 30 30>
      constraintSetup {angularLimitsMax=<0 25 25>, angularLimitsMin=<0 -10 -25>, angularTarget=<1 0 0>}
      localJointOffset=<2 0 0>
     }]
    relativeSpaceBone {boneName=spine_05}
    simSpaceSettings {maxAngularAcceleration=35, maxAngularVelocity=7, simSpaceAngularAlpha=1}
   }
   n2 = Anim Dynamics - Bone: dyn_faceHair_2_00_F(ComponentPose=@n1, Alpha=1, ExternalForce="0,0,0") {
    angularDampingOverride=1
    angularSpringConstant=10000
    bAngularSpring=true
    bOverrideAngularDamping=true
    bUseGravityOverride=true
    gravityScale=0.35
    simulationSpace=World
    alphaBoolBlend {blendOption=Linear}
    boundBone {boneName=dyn_faceHair_2_00_F}
    componentAppliedLinearAccClamp=<5000 5000 5000>
    componentLinearAccScale=<0.5 0.5 0.5>
    componentLinearVelScale=<0.5 0.5 0.5>
    gravityOverride {z=-980}
    physicsBodyDefinitions [{
      boundBone {boneName=dyn_faceHair_2_00_F}
      boxExtents=<30 30 30>
      constraintSetup {angularLimitsMax=<0 25 25>, angularLimitsMin=<0 -7 -25>, angularTarget=<1 0 0>}
      localJointOffset=<2 0 0>
     }]
    relativeSpaceBone {boneName=spine_05}
    simSpaceSettings {maxAngularAcceleration=35, maxAngularVelocity=7, simSpaceAngularAlpha=1}
   }
   n3 = Anim Dynamics - Bone: dyn_faceHair_1_00_R(ComponentPose=@n2, Alpha=1, ExternalForce="0,0,0") {
    angularDampingOverride=1
    angularSpringConstant=15000
    bAngularSpring=true
    bOverrideAngularDamping=true
    bUseGravityOverride=true
    gravityScale=0.35
    simulationSpace=World
    alphaBoolBlend {blendOption=Linear}
    boundBone {boneName=dyn_faceHair_1_00_R}
    componentAppliedLinearAccClamp=<5000 5000 5000>
    componentLinearAccScale=<0.5 0.5 0.5>
    componentLinearVelScale=<0.5 0.5 0.5>
    gravityOverride {z=-980}
    physicsBodyDefinitions [{
      boundBone {boneName=dyn_faceHair_1_00_R}
      boxExtents=<25 25 25>
      constraintSetup {angularLimitsMax=<0 25 2>, angularLimitsMin=<0 -5 -2>, angularTarget=<1 0 0>}
      localJointOffset=<1 0 0>
     }]
    relativeSpaceBone {boneName=spine_05}
    simSpaceSettings {maxAngularAcceleration=35, maxAngularVelocity=7, simSpaceAngularAlpha=1}
   }
   n4 = Anim Dynamics - Bone: dyn_faceHair_2_00_R(ComponentPose=@n3, Alpha=1, ExternalForce="0,0,0") {
    angularDampingOverride=1
    angularSpringConstant=10000
    bAngularSpring=true
    bOverrideAngularDamping=true
    bUseGravityOverride=true
    gravityScale=0.35
    simulationSpace=World
    alphaBoolBlend {blendOption=Linear}
    boundBone {boneName=dyn_faceHair_2_00_R}
    componentAppliedLinearAccClamp=<5000 5000 5000>
    componentLinearAccScale=<0.5 0.5 0.5>
    componentLinearVelScale=<0.5 0.5 0.5>
    gravityOverride {z=-980}
    physicsBodyDefinitions [{
      boundBone {boneName=dyn_faceHair_2_00_R}
      boxExtents=<30 30 30>
      constraintSetup {angularLimitsMax=<0 10 11>, angularLimitsMin=<0 -10 -25>, angularTarget=<1 0 0>}
      localJointOffset=<2 0 0>
     }]
    relativeSpaceBone {boneName=spine_05}
    simSpaceSettings {maxAngularAcceleration=35, maxAngularVelocity=7, simSpaceAngularAlpha=1}
   }
   n5 = Anim Dynamics - Bone: dyn_faceHair_1_00_L(ComponentPose=@n4, Alpha=1, ExternalForce="0,0,0") {
    angularDampingOverride=1
    angularSpringConstant=10000
    bAngularSpring=true
    bOverrideAngularDamping=true
    bUseGravityOverride=true
    gravityScale=0.35
    simulationSpace=World
    alphaBoolBlend {blendOption=Linear}
    boundBone {boneName=dyn_faceHair_1_00_L}
    componentAppliedLinearAccClamp=<5000 5000 5000>
    componentLinearAccScale=<0.5 0.5 0.5>
    componentLinearVelScale=<0.5 0.5 0.5>
    gravityOverride {z=-980}
    physicsBodyDefinitions [{
      boundBone {boneName=dyn_faceHair_1_00_L}
      boxExtents=<25 25 25>
      constraintSetup {angularLimitsMax=<0 25 5>, angularLimitsMin=<0 -2 -5>, angularTarget=<1 0 0>}
      localJointOffset=<1 0 0>
     }]
    relativeSpaceBone {boneName=spine_05}
    simSpaceSettings {maxAngularAcceleration=35, maxAngularVelocity=7, simSpaceAngularAlpha=1}
   }
   n6 = Anim Dynamics - Bone: dyn_faceHair_1_01_L(ComponentPose=@n5, Alpha=1, ExternalForce="0,0,0") {
    angularDampingOverride=1
    angularSpringConstant=10000
    bAngularSpring=true
    bOverrideAngularDamping=true
    bUseGravityOverride=true
    gravityScale=0.35
    simulationSpace=World
    alphaBoolBlend {blendOption=Linear}
    boundBone {boneName=dyn_faceHair_1_01_L}
    componentAppliedLinearAccClamp=<5000 5000 5000>
    componentLinearAccScale=<0.5 0.5 0.5>
    componentLinearVelScale=<0.5 0.5 0.5>
    gravityOverride {z=-980}
    physicsBodyDefinitions [{
      boundBone {boneName=dyn_faceHair_1_01_L}
      boxExtents=<25 25 25>
      constraintSetup {angularLimitsMax=<0 25 5>, angularLimitsMin=<0 -5 -5>, angularTarget=<1 0 0>}
      localJointOffset=<1 0 0>
     }]
    relativeSpaceBone {boneName=spine_05}
    simSpaceSettings {maxAngularAcceleration=35, maxAngularVelocity=7, simSpaceAngularAlpha=1}
   }
   n7 = Anim Dynamics - Bone: dyn_faceHair_2_02_L(ComponentPose=@n6, Alpha=1, ExternalForce="0,0,0") {
    angularDampingOverride=1
    angularSpringConstant=15000
    bAngularSpring=true
    bOverrideAngularDamping=true
    bUseGravityOverride=true
    gravityScale=0.35
    simulationSpace=World
    alphaBoolBlend {blendOption=Linear}
    boundBone {boneName=dyn_faceHair_2_02_L}
    componentAppliedLinearAccClamp=<5000 5000 5000>
    componentLinearAccScale=<0.5 0.5 0.5>
    componentLinearVelScale=<0.5 0.5 0.5>
    gravityOverride {z=-980}
    physicsBodyDefinitions [{
      boundBone {boneName=dyn_faceHair_2_02_L}
      boxExtents=<30 30 30>
      constraintSetup {angularLimitsMax=<0 20 10>, angularLimitsMin=<0 -5 -10>, angularTarget=<1 0 0>}
      localJointOffset=<2 0 0>
     }]
    relativeSpaceBone {boneName=spine_05}
    simSpaceSettings {maxAngularAcceleration=35, maxAngularVelocity=7, simSpaceAngularAlpha=1}
   }
   n8 = Anim Dynamics - Bone: dyn_faceHair_2_01_L(ComponentPose=@n7, Alpha=1, ExternalForce="0,0,0") {
    angularDampingOverride=1
    angularSpringConstant=15000
    bAngularSpring=true
    bOverrideAngularDamping=true
    bUseGravityOverride=true
    gravityScale=0.35
    simulationSpace=World
    alphaBoolBlend {blendOption=Linear}
    boundBone {boneName=dyn_faceHair_2_01_L}
    componentAppliedLinearAccClamp=<5000 5000 5000>
    componentLinearAccScale=<0.5 0.5 0.5>
    componentLinearVelScale=<0.5 0.5 0.5>
    gravityOverride {z=-980}
    physicsBodyDefinitions [{
      boundBone {boneName=dyn_faceHair_2_01_L}
      boxExtents=<30 30 30>
      constraintSetup {angularLimitsMax=<0 25 5>, angularLimitsMin=<0 -10 -5>, angularTarget=<1 0 0>}
      localJointOffset=<2 0 0>
     }]
    relativeSpaceBone {boneName=spine_05}
    simSpaceSettings {maxAngularAcceleration=35, maxAngularVelocity=7, simSpaceAngularAlpha=1}
   }
   n9 = Anim Dynamics - Bone: dyn_faceHair_3_00_R(ComponentPose=@n8, Alpha=1, ExternalForce="0,0,0") {
    angularDampingOverride=1
    angularSpringConstant=10000
    bAngularSpring=true
    bOverrideAngularDamping=true
    bUseGravityOverride=true
    gravityScale=0.35
    simulationSpace=World
    alphaBoolBlend {blendOption=Linear}
    boundBone {boneName=dyn_faceHair_3_00_R}
    componentAppliedLinearAccClamp=<5000 5000 5000>
    componentLinearAccScale=<0.5 0.5 0.5>
    componentLinearVelScale=<0.5 0.5 0.5>
    gravityOverride {z=-980}
    physicsBodyDefinitions [{
      boundBone {boneName=dyn_faceHair_3_00_R}
      boxExtents=<30 30 30>
      constraintSetup {angularLimitsMax=<0 10 20>, angularLimitsMin=<0 -10 -25>, angularTarget=<1 0 0>}
      localJointOffset=<2 0 0>
     }]
    relativeSpaceBone {boneName=spine_05}
    simSpaceSettings {maxAngularAcceleration=35, maxAngularVelocity=7, simSpaceAngularAlpha=1}
   }
   n10 = Anim Dynamics - Bone: dyn_backHair_01(ComponentPose=@n9, Alpha=1, ExternalForce="0,0,0") {
    angularDampingOverride=1
    angularSpringConstant=20000
    bAngularSpring=true
    bOverrideAngularDamping=true
    bUseGravityOverride=true
    gravityScale=0.35
    simulationSpace=RootRelative
    alphaBoolBlend {blendOption=Linear}
    boundBone {boneName=dyn_backHair_01}
    componentLinearAccScale=<1 1 1>
    componentLinearVelScale=<1 1 1>
    gravityOverride {z=-980}
    physicsBodyDefinitions [{
      boundBone {boneName=dyn_backHair_01}
      boxExtents=<20 30 30>
      constraintSetup {angularLimitsMax=<0 5 7.5>, angularLimitsMin=<0 -15 -7.5>, angularTarget=<1 0 0>}
      localJointOffset=<6 0 0>
     }]
    simSpaceSettings {maxAngularAcceleration=35, maxAngularVelocity=7, simSpaceAngularAlpha=1}
   }
   n11 = RigidBody!(
    ComponentPose=@n10
    Alpha
    OverrideWorldGravity:Vector="0,0,0"
    ExternalForce
    SimSpaceSettings:Sim Space Settings Structure=(WorldAlpha=1,VelocityScaleZ=1,DampingAlpha=1,MaxLinearVelocity=500,MaxAngularVelocity=3,MaxLinearAcceleration=1000,MaxAngularAcceleration=1,ExternalLinearDragV=(X=1,Y=1,Z=1),ExternalLinearVelocity=(X=0,Y=0,Z=0),ExternalAngularVelocity=(X=0,Y=0,Z=0))
   ) -> (Pose) {
    bClampLinearTranslationLimitToRefPose=true
    bOverrideWorldGravity=true
    overridePhysicsAsset=&/F_LRG_WiryPerk_Seed_FaceAcc_Physics
    simulationSpace=BaseBoneSpace
    alphaBoolBlend {blendOption=Linear}
    baseBoneRef {boneName=neck_01}
    componentAppliedLinearAccClamp=<3000 3000 3000>
    simSpaceSettings {maxAngularAcceleration=1, maxAngularVelocity=3, maxLinearAcceleration=1000, maxLinearVelocity=500, worldAlpha=1, externalLinearDragV=<1 1 1>}
   }
   n12 = "Transform (Modify) Bone - Bone: base_sideHair_L"(ComponentPose=@n11, Alpha=1, Translation="0,0,0", Rotation="12,0,0", Scale="1,1,1") {
    rotationMode=BMM_Additive
    rotationSpace=BCS_BoneSpace
    translationMode=BMM_Additive
    translationSpace=BCS_BoneSpace
    alphaBoolBlend {blendOption=Linear}
    alphaScaleBiasClamp {bInterpResult=true}
    boneToModify {boneName=base_sideHair_L}
    rotation {pitch=12}
   }
   n13 = "Transform (Modify) Bone - Bone: dyn_sideHair_L_02"(ComponentPose=@n12, Alpha=1, Translation="0,0,1", Rotation="15,10,0", Scale="1,1,1") {
    rotationMode=BMM_Additive
    rotationSpace=BCS_BoneSpace
    translationMode=BMM_Additive
    translationSpace=BCS_BoneSpace
    alphaBoolBlend {blendOption=Linear}
    alphaScaleBiasClamp {bInterpResult=true}
    boneToModify {boneName=dyn_sideHair_L_02}
    rotation {pitch=15, yaw=10}
    translation {z=1}
   }
   n14 = "Transform (Modify) Bone - Bone: dyn_sideHair_R_02"(ComponentPose=@n13, Alpha=1, Translation="0,0,0", Rotation="-30,-10,0", Scale="1,1,1") {
    rotationMode=BMM_Additive
    rotationSpace=BCS_BoneSpace
    translationMode=BMM_Additive
    translationSpace=BCS_BoneSpace
    alphaBoolBlend {blendOption=Linear}
    alphaScaleBiasClamp {bInterpResult=true}
    boneToModify {boneName=dyn_sideHair_R_02}
    rotation {pitch=-30, yaw=-10}
   }
   n15 = "Transform (Modify) Bone - Bone: base_backHair"(ComponentPose=@n14, Alpha=1, Translation="0,0,0", Rotation="-86.448,179.805,-179.712", Scale="1.000,1.000,0.278") {
    rotationMode=BMM_Additive
    rotationSpace=BCS_BoneSpace
    scaleMode=BMM_Additive
    scaleSpace=BCS_BoneSpace
    translationMode=BMM_Additive
    translationSpace=BCS_BoneSpace
    alphaBoolBlend {blendOption=Linear}
    alphaScaleBiasClamp {bInterpResult=true}
    boneToModify {boneName=base_backHair}
    rotation=<rot -86.448 179.8048 -179.7118>
    scale {z=0.2781}
   }
   n16 = "Constraint - dyn_sideHair_L_01 (PP)"!(ComponentPose=@n15, Alpha=1, ConstraintWeights_0=0.8, ConstraintWeights_1:float=0.2) -> (Pose) {
    alphaBoolBlend {blendOption=Linear}
    boneToModify {boneName=dyn_sideHair_L_01}
    constraintSetup [{targetBone {boneName=neck_02}}, {targetBone {boneName=clavicle_l}}]
    constraintWeights [1,1]
   }
   n17 = "Constraint - dyn_sideHair_R_01 (PP)"!(ComponentPose=@n16, Alpha=1, ConstraintWeights_0=0.9, ConstraintWeights_1:float=0.1) -> (Pose) {
    alphaBoolBlend {blendOption=Linear}
    boneToModify {boneName=dyn_sideHair_R_01}
    constraintSetup [{targetBone {boneName=neck_02}}, {targetBone {boneName=clavicle_r}}]
    constraintWeights [1,1]
   }
   n18 = "Transform (Modify) Bone - Bone: dyn_backHair_01"!(ComponentPose=@n17, bAlphaBoolEnabled:bool=True, Rotation="0,0,0", Scale="0.6,0.6,1") -> (Pose) {
    alphaInputType=Bool
    rotationMode=BMM_Additive
    scaleMode=BMM_Additive
    alphaBoolBlend {blendInTime=0.2, blendOption=Sinusoidal, blendOutTime=0.2}
    boneToModify {boneName=dyn_backHair_01}
    scale=<0.6 0.6>
   }
   Output Pose(Result=Component To Local(ComponentPose=@n18)) {layerGroup=None}
  ]}
 EventGraph {nodes [
   n1 = "Gravity Override State Function (Nativized)"!(
    execute
    CharacterPartAnimInstance:Custom Character Part Anim Instance Object Reference=Self-Reference()
    GravityOverrideParameterSet:Fort Gravity Override Parameter Set Structure=Get Gravity Override Params() -> (Gravity Override Params:Fort Gravity Override Parameter Set Structure)
    RigidBodyAnimNodeParameterSet:Fort Rigid Body Anim Node Parameter Set Structure=Get RBAN Params() -> (RBAN Params:Fort Rigid Body Anim Node Parameter Set Structure)
    DebugDraw:bool=false
    GameTimeInSeconds:float=GetGameTimeInSeconds()
    DebugJointName:Name=root
    StateLogicLayerName:Name=none
   ) -> (
    then
    OutGravityOverride:Vector
    OutCurrentState:Name=None
    OutGravityOverrideParameters:Fort Gravity Override Parameters Structure
    OutRigidBodyAnimNodeParameters:Fort Rigid Body Anim Node Parameters Structure
   )
   > Set Gravity Override(Gravity Override=@n1)
   > Set RBAN Settings Out(RBAN Settings Out=@n1.OutRigidBodyAnimNodeParameters)
   Event BlueprintUpdateAnimation()
   > n2 = Sequence()
   > n3 = "Gravity Override State Function (Parameter Asset) (Nativized)"(
    CharacterPartAnimInstance=Self-Reference()
    GameTimeInSeconds=GetGameTimeInSeconds()
    Parameters=/BRCosmetics/Characters/Player/Female/Large/Bodies/F_LRG_WiryPerk_Seed/Meshes/M_MED_WiryPerk_Parameters
   )
   > Set Gravity Override(Gravity Override=@n3)
   n4 = Get bRigidBodyEnabled()
   > Set RBAN Off Alpha(RBAN Off Alpha:double="To Float (Boolean)"(InBool=NOT Boolean(A=@n4))) -> (Output_Get:double)
   n5 = GetAlphaFromReferenceTransformDelta(SkeletalMeshComponent=Get SkelMesh(), BoneName=spine_05, BaseName=head, InputRange=(X=0,Y=110), TwistAxis="0,0,1")
   Set Head Rot(execute=@n2.then_1, Head Rot:float=@n5) -> (Output_Get:float)
   n6 = Get bIsTargeting()
   > Set ADS UP(ADS UP:float=float * float(A=@n5, B="To Float (Boolean)"(InBool=@n6)) {is=title K2Node_PromotableOperator}) -> (Output_Get:float)
   Event BlueprintInitializeAnimation()
   > Set SkelMesh(SkelMesh=GetOwnerSkeletalMeshComponent())
  ]}
]
variables [
 ADS UP {type=number}
 Gravity Override {type=Vector}
 Gravity Override Params {
  type=FortGravityOverrideParameterSet
  default=(Skydive_GoingUp=(GravityOverride=(WindFrequency=(X=1.500000,Y=7.000000,Z=8.000000),WindAmplitude=(X=200.000000,Y=-500.000000,Z=400.000000),GravityOverride=(X=0.000000,Y=-300.000000,Z=1500.000000),JointName="root")),Skydive_FloatingDown=(GravityOverride=(WindFrequency=(X=1.500000,Y=7.000000,Z=8.000000),WindAmplitude=(X=200.000000,Y=-500.000000,Z=400.000000),WindOffset=(X=0.000000,Y=-600.000000,Z=400.000000),GravityOverride=(X=0.000000,Y=-700.000000,Z=600.000000),JointName="root")),Skydive_DivingDown=(GravityOverride=(WindFrequency=(X=1.500000,Y=12.000000,Z=10.000000),WindAmplitude=(X=50.000000,Y=300.000000,Z=100.000000),GravityOverride=(X=-1500.000000,Y=0.000000,Z=0.000000),JointName="pelvis")),Skydive_Parachuting=(GravityOverride=(WindFrequency=(X=1.500000,Y=7.000000,Z=8.000000),WindAmplitude=(X=200.000000,Y=-500.000000,Z=400.000000),GravityOverride=(X=0.000000,Y=-800.000000,Z=-980.000000),JointName="root")),OnGround_Standing=(GravityOverride=(GravityOverride=(X=0.000000,Y=0.000000,Z=-980.000000),JointName="root")),OnGround_Moving=(GravityOverride=(GravityOverride=(X=0.000000,Y=0.000000,Z=-980.000000),JointName="root")),OnGround_Moving_HighRotationRate=(GravityOverride=(GravityOverride=(X=0.000000,Y=0.000000,Z=-980.000000),JointName="root")),OnGround_Crouch_Moving=(GravityOverride=(GravityOverride=(X=0.000000,Y=0.000000,Z=-980.000000),JointName="root")),Emote_or_Melee=(GravityOverride=(GravityOverride=(X=0.000000,Y=0.000000,Z=-980.000000),JointName="root")),Swimming=(GravityOverride=(GravityOverride=(X=-800.000000,Y=0.000000,Z=0.000000),JointName="pelvis")),Front_End=(GravityOverride=(GravityOverride=(X=0.000000,Y=0.000000,Z=-980.000000),JointName="root")),DBNO=(GravityOverride=(GravityOverride=(X=-980.000000,Y=400.000000,Z=0.000000),JointName="pelvis")),DBNO_Carried=(GravityOverride=(GravityOverride=(X=-980.000000,Y=500.000000,Z=-500.000000),JointName="pelvis")),OnGround_Falling=(GravityOverride=(GravityOverride=(X=0.000000,Y=0.000000,Z=-980.000000),JointName="root")),Ziplining=(GravityOverride=(GravityOverride=(X=0.000000,Y=0.000000,Z=-980.000000),JointName="root")),Skydive_Surfing=(GravityOverride=(WindFrequency=(X=1.500000,Y=7.000000,Z=8.000000),WindAmplitude=(X=200.000000,Y=-500.000000,Z=400.000000),GravityOverride=(X=500.000000,Y=-300.000000,Z=-980.000000),JointName="root")),Riding_Standing=(GravityOverride=(GravityOverride=(X=0.000000,Y=-350.000000,Z=-980.000000),JointName="root")),Riding_Moving=(GravityOverride=(GravityOverride=(X=0.000000,Y=-450.000000,Z=-450.000000),JointName="root")),Riding_Moving_Targeting=(GravityOverride=(GravityOverride=(X=0.000000,Y=-200.000000,Z=-980.000000),JointName="root")),Tactical_Sprint=(GravityOverride=(GravityOverride=(X=0.000000,Y=-450.000000,Z=-500.000000),JointName="root")),Rail_Grinding=(GravityOverride=(GravityOverride=(X=0.000000,Y=0.000000,Z=-980.000000),JointName="root")),)
 }
 Head Rot {type=number}
 RBAN Off Alpha {type=number}
 RBAN Params {
  type=FortRigidBodyAnimNodeParameterSet
  default=(Skydive_GoingUp=(RigidBodyAnimNode=(SimSpaceSettings=(WorldAlpha=1.000000,MaxLinearVelocity=500.000000,MaxAngularVelocity=3.000000,MaxLinearAcceleration=1000.000000,MaxAngularAcceleration=1.000000,ExternalLinearDragV=(X=1.000000,Y=1.000000,Z=1.000000)))),Skydive_FloatingDown=(RigidBodyAnimNode=(ComponentLinearAccScale=(X=0.000000,Y=0.500000,Z=0.000000),ComponentLinearVelScale=(X=0.000000,Y=5.000000,Z=0.000000),SimSpaceSettings=(WorldAlpha=1.000000,MaxLinearVelocity=500.000000,MaxAngularVelocity=3.000000,MaxLinearAcceleration=1000.000000,MaxAngularAcceleration=1.000000,ExternalLinearDragV=(X=1.000000,Y=1.000000,Z=1.000000)))),Skydive_DivingDown=(RigidBodyAnimNode=(ComponentLinearAccScale=(X=0.100000,Y=0.000000,Z=0.000000),ComponentLinearVelScale=(X=0.100000,Y=0.000000,Z=0.000000),SimSpaceSettings=(WorldAlpha=1.000000,MaxLinearVelocity=500.000000,MaxAngularVelocity=3.000000,MaxLinearAcceleration=1000.000000,MaxAngularAcceleration=1.000000,ExternalLinearDragV=(X=1.000000,Y=1.000000,Z=1.000000)))),Skydive_Parachuting=(RigidBodyAnimNode=(ComponentLinearAccScale=(X=0.000000,Y=0.100000,Z=0.300000),ComponentLinearVelScale=(X=0.000000,Y=0.100000,Z=0.300000),SimSpaceSettings=(WorldAlpha=1.000000,MaxLinearVelocity=500.000000,MaxAngularVelocity=3.000000,MaxLinearAcceleration=1000.000000,MaxAngularAcceleration=1.000000,ExternalLinearDragV=(X=1.000000,Y=1.000000,Z=1.000000)))),OnGround_Standing=(RigidBodyAnimNode=(ComponentLinearAccScale=(X=0.300000,Y=0.000000,Z=0.000000),ComponentLinearVelScale=(X=0.300000,Y=0.000000,Z=0.000000),SimSpaceSettings=(WorldAlpha=1.000000,MaxLinearVelocity=500.000000,MaxAngularVelocity=3.000000,MaxLinearAcceleration=1000.000000,MaxAngularAcceleration=1.000000,ExternalLinearDragV=(X=1.000000,Y=1.000000,Z=1.000000)))),OnGround_Moving=(RigidBodyAnimNode=(ComponentLinearAccScale=(X=0.300000,Y=0.000000,Z=0.000000),ComponentLinearVelScale=(X=0.300000,Y=0.000000,Z=0.000000),SimSpaceSettings=(WorldAlpha=1.000000,MaxLinearVelocity=500.000000,MaxAngularVelocity=3.000000,MaxLinearAcceleration=1000.000000,MaxAngularAcceleration=1.000000,ExternalLinearDragV=(X=1.000000,Y=1.000000,Z=1.000000)))),OnGround_Moving_HighRotationRate=(RigidBodyAnimNode=(ComponentLinearAccScale=(X=0.300000,Y=0.000000,Z=0.000000),ComponentLinearVelScale=(X=0.300000,Y=0.000000,Z=0.000000),SimSpaceSettings=(WorldAlpha=1.000000,MaxLinearVelocity=500.000000,MaxAngularVelocity=3.000000,MaxLinearAcceleration=1000.000000,MaxAngularAcceleration=1.000000,ExternalLinearDragV=(X=1.000000,Y=1.000000,Z=1.000000)))),OnGround_Crouch_Moving=(RigidBodyAnimNode=(SimSpaceSettings=(WorldAlpha=1.000000,MaxLinearVelocity=500.000000,MaxAngularVelocity=3.000000,MaxLinearAcceleration=1000.000000,MaxAngularAcceleration=1.000000,ExternalLinearDragV=(X=1.000000,Y=1.000000,Z=1.000000)))),Emote_or_Melee=(RigidBodyAnimNode=(SimSpaceSettings=(WorldAlpha=1.000000,MaxLinearVelocity=500.000000,MaxAngularVelocity=3.000000,MaxLinearAcceleration=1000.000000,MaxAngularAcceleration=1.000000,ExternalLinearDragV=(X=1.000000,Y=1.000000,Z=1.000000)))),Swimming=(RigidBodyAnimNode=(SimSpaceSettings=(WorldAlpha=1.000000,MaxLinearVelocity=500.000000,MaxAngularVelocity=3.000000,MaxLinearAcceleration=1000.000000,MaxAngularAcceleration=1.000000,ExternalLinearDragV=(X=1.000000,Y=1.000000,Z=1.000000)))),Front_End=(RigidBodyAnimNode=(SimSpaceSettings=(WorldAlpha=0.500000,VelocityScaleZ=0.500000,MaxLinearVelocity=500.000000,MaxAngularVelocity=3.000000,MaxLinearAcceleration=1000.000000,MaxAngularAcceleration=1.000000,ExternalLinearDragV=(X=1.000000,Y=1.000000,Z=1.000000)))),DBNO=(RigidBodyAnimNode=(SimSpaceSettings=(WorldAlpha=1.000000,MaxLinearVelocity=500.000000,MaxAngularVelocity=3.000000,MaxLinearAcceleration=1000.000000,MaxAngularAcceleration=1.000000,ExternalLinearDragV=(X=1.000000,Y=1.000000,Z=1.000000)))),DBNO_Carried=(RigidBodyAnimNode=(SimSpaceSettings=(WorldAlpha=1.000000,MaxLinearVelocity=500.000000,MaxAngularVelocity=3.000000,MaxLinearAcceleration=1000.000000,MaxAngularAcceleration=1.000000,ExternalLinearDragV=(X=1.000000,Y=1.000000,Z=1.000000)))),OnGround_Falling=(RigidBodyAnimNode=(SimSpaceSettings=(WorldAlpha=1.000000,MaxLinearVelocity=500.000000,MaxAngularVelocity=3.000000,MaxLinearAcceleration=1000.000000,MaxAngularAcceleration=1.000000,ExternalLinearDragV=(X=1.000000,Y=1.000000,Z=1.000000)))),Ziplining=(RigidBodyAnimNode=(SimSpaceSettings=(WorldAlpha=1.000000,MaxLinearVelocity=500.000000,MaxAngularVelocity=3.000000,MaxLinearAcceleration=1000.000000,MaxAngularAcceleration=1.000000,ExternalLinearDragV=(X=1.000000,Y=1.000000,Z=1.000000)))),Skydive_Surfing=(RigidBodyAnimNode=(SimSpaceSettings=(WorldAlpha=1.000000,MaxLinearVelocity=500.000000,MaxAngularVelocity=3.000000,MaxLinearAcceleration=1000.000000,MaxAngularAcceleration=1.000000,ExternalLinearDragV=(X=1.000000,Y=1.000000,Z=1.000000)))),Riding_Standing=(RigidBodyAnimNode=(SimSpaceSettings=(WorldAlpha=1.000000,MaxLinearVelocity=500.000000,MaxAngularVelocity=3.000000,MaxLinearAcceleration=1000.000000,MaxAngularAcceleration=1.000000,ExternalLinearDragV=(X=1.000000,Y=1.000000,Z=1.000000)))),Riding_Moving=(RigidBodyAnimNode=(SimSpaceSettings=(WorldAlpha=1.000000,MaxLinearVelocity=500.000000,MaxAngularVelocity=3.000000,MaxLinearAcceleration=1000.000000,MaxAngularAcceleration=1.000000,ExternalLinearDragV=(X=1.000000,Y=1.000000,Z=1.000000)))),Riding_Moving_Targeting=(RigidBodyAnimNode=(SimSpaceSettings=(WorldAlpha=1.000000,MaxLinearVelocity=500.000000,MaxAngularVelocity=3.000000,MaxLinearAcceleration=1000.000000,MaxAngularAcceleration=1.000000,ExternalLinearDragV=(X=1.000000,Y=1.000000,Z=1.000000)))),Tactical_Sprint=(RigidBodyAnimNode=(SimSpaceSettings=(WorldAlpha=1.000000,MaxLinearVelocity=500.000000,MaxAngularVelocity=3.000000,MaxLinearAcceleration=1000.000000,MaxAngularAcceleration=1.000000,ExternalLinearDragV=(X=1.000000,Y=1.000000,Z=1.000000)))),Rail_Grinding=(RigidBodyAnimNode=(SimSpaceSettings=(WorldAlpha=1.000000,MaxLinearVelocity=500.000000,MaxAngularVelocity=3.000000,MaxLinearAcceleration=1000.000000,MaxAngularAcceleration=1.000000,ExternalLinearDragV=(X=1.000000,Y=1.000000,Z=1.000000)))),)
 }
 RBAN Settings Out {type=FortRigidBodyAnimNodeParameters}
 SkelMesh {type=SkeletalMeshComponent}
]
